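# Niagara system for the MatchaSpare emote that, going by its name and emitters, plays an anvil ground impact with debris, cracks and a crater.
# The Debris emitter burst-spawns randomly scaled mesh particles on a ring or disc shape.
# The system combines sprite, mesh and decal renderers with gravity, drag, rotation and collision modules.
# Dynamic inputs include curves, random ranges, time-of-day color and platform-set checks that can vary values per platform.
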NiagaraSystem /BRCosmetics/Animation/Game/MainPlayer/Emotes/MatchaSpare/FX/NS_Emote_MatchaSpare_AnvilGround
dependencies {
 usedDataInterfaces [
  {refPath=&NiagaraDataInterfaceCurve}
  {refPath=&NiagaraDataInterfaceColorCurve}
  {refPath=&NiagaraDataInterfaceVectorCurve}
  {refPath=&NiagaraDataInterfaceMeshRendererInfo}
  {refPath=&NiagaraDataInterfacePlatformSet}
 ]
 usedDynamicInputs [
  {refPath=&Multiply_Float}
  {refPath=&RandomRangeFloat}
  {refPath=&FloatFromCurve}
  {refPath=&ReturnNormalizedExecIndex}
  {refPath=&Lerp_Vector}
  {refPath=&MakeFloatFromInt}
  {refPath=&SpawnGroup}
  {refPath=&FNTimeOfDayColor}
  {refPath=/Niagara/DynamicInputs/CascadeConversion/SplitVector}
  {refPath=&VectorFromCurve}
  {refPath=&VectorFromFloat}
  {refPath=&RandomVector}
  {refPath=&Multiply_VectorByFloat}
  {refPath=&RandomRangeInt}
  {refPath=&IsPlatformSetActive}
  {refPath=&MakeVector}
  {refPath=&RandomRangeVector}
  {refPath=&MakeCustomFloatFromBool}
  {refPath=&MakeCustomIntFromBool}
 ]
 usedModules [
  {refPath=&EmitterState}
  {refPath=&SpawnBurst_Instantaneous}
  {refPath=&InitializeParticle}
  {refPath=&ShapeLocation}
  {refPath=&AddVelocity}
  {refPath=&ParticleState}
  {refPath=&Drag}
  {refPath=&SolveForcesAndVelocity}
  {refPath=&DynamicMaterialParameters}
  {refPath=&ScaleColor}
  {refPath=&ScaleSpriteSize}
  {refPath=&InitialMeshOrientation}
  {refPath=&ScaleMeshSize}
  {refPath=&UpdateMeshOrientation}
  {refPath=&AccelerationForce}
  {refPath=&GravityForce}
  {refPath=&MeshRotationForce}
  {refPath=&Collision}
  {refPath=&self:Ground_Cracks001_0.NiagaraScriptSource_0.NiagaraGraph_0.NiagaraNodeAssignment_3.SetVariables_#1}
  {refPath=&self:Ground_Cracks001_0.NiagaraScriptSource_0.NiagaraGraph_0.NiagaraNodeAssignment_1.SetVariables_#2}
  {refPath=/Niagara/Modules/Spawn/Decal/InitialDecalOrientation}
  {refPath=&self:Ground_Cracks001_0.NiagaraScriptSource_0.NiagaraGraph_0.NiagaraNodeAssignment_2.SetVariables_#3}
  {refPath=&self:Ground_Cracks001_0.NiagaraScriptSource_0.NiagaraGraph_0.NiagaraNodeAssignment_4.SetVariables_#4}
  {refPath=&self:Ground_Crater_0.NiagaraScriptSource_0.NiagaraGraph_0.NiagaraNodeAssignment_3.SetVariables_#1}
  {refPath=&self:Ground_Crater_0.NiagaraScriptSource_0.NiagaraGraph_0.NiagaraNodeAssignment_1.SetVariables_#2}
  {refPath=&self:Ground_Crater_0.NiagaraScriptSource_0.NiagaraGraph_0.NiagaraNodeAssignment_0.SetVariables_#5}
  {refPath=&self:Ground_Crater_0.NiagaraScriptSource_0.NiagaraGraph_0.NiagaraNodeAssignment_2.SetVariables_#3}
  {refPath=&self:Ground_Crater_0.NiagaraScriptSource_0.NiagaraGraph_0.NiagaraNodeAssignment_4.SetVariables_#4}
  {refPath=&SystemState}
  {refPath=&self:SystemSpawnScript.SystemScriptSource.SystemScriptGraph.NiagaraNodeAssignment_0.SetVariables_#6}
 ]
 usedRenderers [
  {refPath=&NiagaraSpriteRendererProperties}
  {refPath=&NiagaraMeshRendererProperties}
  {refPath=&NiagaraDecalRendererProperties}
 ]
}
emitters {
 Debris {
  emitterSpawnScript {scriptName=EmitterSpawnScript, modules []}
  emitterUpdateScript {
   scriptName=EmitterUpdateScript
   modules [
    {
     bIsSetParametersModule=false
     enabled=true
     moduleName=EmitterState
     inputs [
      Life Cycle Mode {
       bIsDynamic=false
       bIsEditable=true
       bIsStaticSwitch=true
       bIsVisible=true
       type {classStructOrEnum {refPath=/Niagara/Enums/ENiagaraEmitterLifeCycleMode}}
      }
      Inactive Response {
       bIsDynamic=false
       bIsEditable=true
       bIsStaticSwitch=true
       bIsVisible=true
       type {classStructOrEnum {refPath=/Niagara/Enums/ENiagaraInactiveMode}}
      }
      Loop Behavior {
       bIsDynamic=false
       bIsEditable=true
       bIsStaticSwitch=true
       bIsVisible=true
       type {classStructOrEnum {refPath=/Niagara/Enums/ENiagara_EmitterStateOptions}}
      }
      Loop Duration Mode {
       bIsDynamic=false
       bIsEditable=true
       bIsStaticSwitch=true
       bIsVisible=true
       type {classStructOrEnum {refPath=/Niagara/Enums/ENiagara_InfiniteLoopDuration}}
      }
      Loop Count {
       bIsDynamic=false
       bIsEditable=false
       bIsStaticSwitch=false
       bIsVisible=false
       type {classStructOrEnum {refPath=/Script/Niagara.NiagaraInt32}}
      }
      Loop Duration {bIsDynamic=false, bIsEditable=true, bIsStaticSwitch=false, bIsVisible=true, type {classStructOrEnum {refPath=/Script/Niagara.NiagaraFloat}}}
      Recalculate Duration Each Loop {bIsDynamic=false, bIsEditable=false, bIsStaticSwitch=false, bIsVisible=false, type {classStructOrEnum {refPath=/Script/Niagara.NiagaraBool}}}
      UseLoopDelay {bIsDynamic=false, bIsEditable=true, bIsStaticSwitch=false, bIsVisible=true, type {classStructOrEnum {refPath=&NiagaraBool}}}
      Loop Delay {bIsDynamic=false, bIsEditable=false, bIsStaticSwitch=false, bIsVisible=true, type {classStructOrEnum {refPath=&NiagaraFloat}}}
      Delay First Loop Only {bIsDynamic=false, bIsEditable=false, bIsStaticSwitch=false, bIsVisible=false, type {classStructOrEnum {refPath=&NiagaraBool}}}
      Scalability Mode {
       bIsDynamic=false
       bIsEditable=true
       bIsStaticSwitch=true
       bIsVisible=true
       type {classStructOrEnum {refPath=/Niagara/Enums/ENiagaraEmitterScalabilityMode_Limited}}
      }
      Enable Distance Culling {bIsDynamic=false, bIsEditable=false, bIsStaticSwitch=true, bIsVisible=false, type {classStructOrEnum {refPath=&NiagaraBool}}}
      MinDistance {
       bIsDynamic=false
       bIsEditable=false
       bIsStaticSwitch=false
       bIsVisible=false
       type {classStructOrEnum {refPath=&NiagaraFloat}}
      }
      Min Distance Response {
       bIsDynamic=false
       bIsEditable=false
       bIsStaticSwitch=false
       bIsVisible=false
       type {classStructOrEnum {refPath=/Script/Niagara.ENiagaraExecutionStateManagement}}
      }
      MaxDistance {
       bIsDynamic=false
       bIsEditable=false
       bIsStaticSwitch=false
       bIsVisible=false
       type {classStructOrEnum {refPath=&NiagaraFloat}}
      }
      Max Distance Response {
       bIsDynamic=false
       bIsEditable=false
       bIsStaticSwitch=false
       bIsVisible=false
       type {classStructOrEnum {refPath=&ENiagaraExecutionStateManagement}}
      }
      Scale Spawn Count {bIsDynamic=false, bIsEditable=false, bIsStaticSwitch=true, bIsVisible=false, type {classStructOrEnum {refPath=&NiagaraBool}}}
      Spawn Count Scale {
       bIsDynamic=false
       bIsEditable=false
       bIsStaticSwitch=false
       bIsVisible=false
       type {classStructOrEnum {refPath=&NiagaraFloat}}
      }
      Spawn Count Scale By Distance {
       bIsDynamic=false
       bIsEditable=false
       bIsStaticSwitch=false
       bIsVisible=false
       type {classStructOrEnum {refPath=&NiagaraDataInterfaceCurve}}
      }
      Enable Visibility Culling {bIsDynamic=false, bIsEditable=false, bIsStaticSwitch=true, bIsVisible=false, type {classStructOrEnum {refPath=&NiagaraBool}}}
      Visibility Cull Response {
       bIsDynamic=false
       bIsEditable=false
       bIsStaticSwitch=false
       bIsVisible=false
       type {classStructOrEnum {refPath=&ENiagaraExecutionStateManagement}}
      }
      Visibility Cull Delay {
       bIsDynamic=false
       bIsEditable=false
       bIsStaticSwitch=false
       bIsVisible=false
       type {classStructOrEnum {refPath=&NiagaraFloat}}
      }
      Reset Age On Awaken {bIsDynamic=false, bIsEditable=false, bIsStaticSwitch=false, bIsVisible=false, type {classStructOrEnum {refPath=&NiagaraBool}}}
     ]
     moduleScript {refPath=&EmitterState}
    }
    {
     bIsSetParametersModule=false
     enabled=true
     moduleName=SpawnBurst_Instantaneous
     inputs [
      Spawn Count {bIsDynamic=true, bIsEditable=true, bIsStaticSwitch=false, bIsVisible=true, type {classStructOrEnum {refPath=&NiagaraInt32}}}
      Spawn Time {bIsDynamic=false, bIsEditable=true, bIsStaticSwitch=false, bIsVisible=true, type {classStructOrEnum {refPath=&NiagaraFloat}}}
      Use Spawn Probability {bIsDynamic=false, bIsEditable=true, bIsStaticSwitch=false, bIsVisible=true, type {classStructOrEnum {refPath=&NiagaraBool}}}
      Spawn Probability {bIsDynamic=false, bIsEditable=false, bIsStaticSwitch=false, bIsVisible=true, type {classStructOrEnum {refPath=&NiagaraFloat}}}
      UseLoopCountLimit {bIsDynamic=false, bIsEditable=true, bIsStaticSwitch=false, bIsVisible=true, type {classStructOrEnum {refPath=&NiagaraBool}}}
      Loop Count Limit {bIsDynamic=false, bIsEditable=false, bIsStaticSwitch=false, bIsVisible=true, type {classStructOrEnum {refPath=&NiagaraInt32}}}
      Spawn Group {bIsDynamic=false, bIsEditable=true, bIsStaticSwitch=false, bIsVisible=true, type {classStructOrEnum {refPath=&NiagaraInt32}}}
      Age {bIsDynamic=false, bIsEditable=true, bIsStaticSwitch=false, bIsVisible=true, type {classStructOrEnum {refPath=&NiagaraFloat}}}
     ]
     moduleScript {refPath=&SpawnBurst_Instantaneous}
    }
   ]
  }
  particleSpawnScript {
   scriptName=ParticleSpawnScript
   modules [
    {
     bIsSetParametersModule=false
     enabled=true
     moduleName=InitializeParticle
     inputs [
      Lifetime Mode {
       bIsDynamic=false
       bIsEditable=true
       bIsStaticSwitch=true
       bIsVisible=true
       type {classStructOrEnum {refPath=/Niagara/Enums/ENiagara_LifetimeMode}}
      }
      Lifetime Min {bIsDynamic=false, bIsEditable=true, bIsStaticSwitch=false, bIsVisible=true, type {classStructOrEnum {refPath=&NiagaraFloat}}}
      Lifetime {
       bIsDynamic=false
       bIsEditable=false
       bIsStaticSwitch=false
       bIsVisible=false
       type {classStructOrEnum {refPath=&NiagaraFloat}}
      }
      Lifetime Max {bIsDynamic=false, bIsEditable=true, bIsStaticSwitch=false, bIsVisible=true, type {classStructOrEnum {refPath=&NiagaraFloat}}}
      Lifetime Randomness Mode {
       bIsDynamic=false
       bIsEditable=true
       bIsStaticSwitch=true
       bIsVisible=true
       type {classStructOrEnum {refPath=/Niagara/Enums/ENiagaraRandomnessMode}}
      }
      Color Mode {
       bIsDynamic=false
       bIsEditable=true
       bIsStaticSwitch=true
       bIsVisible=true
       type {classStructOrEnum {refPath=/Niagara/Enums/ENiagara_ColorInitializationMode}}
      }
      Color Minimum {
       bIsDynamic=false
       bIsEditable=false
       bIsStaticSwitch=false
       bIsVisible=false
       type {classStructOrEnum {refPath=/Script/CoreUObject.LinearColor}}
      }
      Color {
       bIsDynamic=false
       bIsEditable=false
       bIsStaticSwitch=false
       bIsVisible=false
       type {classStructOrEnum {refPath=&LinearColor}}
      }
      AdjustHue {bIsDynamic=false, bIsEditable=false, bIsStaticSwitch=false, bIsVisible=false, type {classStructOrEnum {refPath=&NiagaraBool}}}
      Hue Shift Range {
       bIsDynamic=false
       bIsEditable=false
       bIsStaticSwitch=false
       bIsVisible=false
       type {classStructOrEnum {refPath=/Script/CoreUObject.Vector2f}}
      }
      Color Maximum {
       bIsDynamic=false
       bIsEditable=false
       bIsStaticSwitch=false
       bIsVisible=false
       type {classStructOrEnum {refPath=&LinearColor}}
      }
      AdjustSaturation {bIsDynamic=false, bIsEditable=false, bIsStaticSwitch=false, bIsVisible=false, type {classStructOrEnum {refPath=&NiagaraBool}}}
      Saturation Range {
       bIsDynamic=false
       bIsEditable=false
       bIsStaticSwitch=false
       bIsVisible=false
       type {classStructOrEnum {refPath=&Vector2f}}
      }
      AdjustValue {bIsDynamic=false, bIsEditable=false, bIsStaticSwitch=false, bIsVisible=false, type {classStructOrEnum {refPath=&NiagaraBool}}}
      Value Range {
       bIsDynamic=false
       bIsEditable=false
       bIsStaticSwitch=false
       bIsVisible=false
       type {classStructOrEnum {refPath=&Vector2f}}
      }
      AdjustAlpha {bIsDynamic=false, bIsEditable=false, bIsStaticSwitch=false, bIsVisible=false, type {classStructOrEnum {refPath=&NiagaraBool}}}
      Alpha Scale Range {
       bIsDynamic=false
       bIsEditable=false
       bIsStaticSwitch=false
       bIsVisible=false
       type {classStructOrEnum {refPath=&Vector2f}}
      }
      Color Channel Mode {
       bIsDynamic=false
       bIsEditable=false
       bIsStaticSwitch=true
       bIsVisible=false
       type {classStructOrEnum {refPath=/Niagara/Enums/ENiagaraChannelCorrelation}}
      }
      Color Randomness Mode {
       bIsDynamic=false
       bIsEditable=false
       bIsStaticSwitch=true
       bIsVisible=false
       type {classStructOrEnum {refPath=&ENiagaraRandomnessMode}}
      }
      Position Mode {
       bIsDynamic=false
       bIsEditable=true
       bIsStaticSwitch=true
       bIsVisible=true
       type {classStructOrEnum {refPath=/Niagara/Enums/ENiagara_PositionInitializationMode}}
      }
      Position {
       bIsDynamic=false
       bIsEditable=false
       bIsStaticSwitch=false
       bIsVisible=false
       type {classStructOrEnum {refPath=/Script/Niagara.NiagaraPosition}}
      }
      UsePositionOffset {bIsDynamic=false, bIsEditable=true, bIsStaticSwitch=false, bIsVisible=true, type {classStructOrEnum {refPath=&NiagaraBool}}}
      Position Offset {bIsDynamic=false, bIsEditable=true, bIsStaticSwitch=false, bIsVisible=true, type {classStructOrEnum {refPath=/Script/CoreUObject.Vector3f}}}
      Position Offset Coordinate Space {
       bIsDynamic=false
       bIsEditable=true
       bIsStaticSwitch=false
       bIsVisible=true
       type {classStructOrEnum {refPath=/Script/Niagara.ENiagaraCoordinateSpace}}
      }
      Mass Mode {
       bIsDynamic=false
       bIsEditable=true
       bIsStaticSwitch=true
       bIsVisible=true
       type {classStructOrEnum {refPath=/Niagara/Enums/ENiagara_MassInitializationMode}}
      }
      Mass Min {
       bIsDynamic=false
       bIsEditable=false
       bIsStaticSwitch=false
       bIsVisible=false
       type {classStructOrEnum {refPath=&NiagaraFloat}}
      }
      Mass {
       bIsDynamic=false
       bIsEditable=false
       bIsStaticSwitch=false
       bIsVisible=false
       type {classStructOrEnum {refPath=&NiagaraFloat}}
      }
      Mass Max {
       bIsDynamic=false
       bIsEditable=false
       bIsStaticSwitch=false
       bIsVisible=false
       type {classStructOrEnum {refPath=&NiagaraFloat}}
      }
      Mass Randomness Mode {
       bIsDynamic=false
       bIsEditable=false
       bIsStaticSwitch=true
       bIsVisible=false
       type {classStructOrEnum {refPath=&ENiagaraRandomnessMode}}
      }
      Write Material Random {bIsDynamic=false, bIsEditable=true, bIsStaticSwitch=false, bIsVisible=true, type {classStructOrEnum {refPath=&NiagaraBool}}}
      Material Random {bIsDynamic=false, bIsEditable=true, bIsStaticSwitch=false, bIsVisible=true, type {classStructOrEnum {refPath=&NiagaraFloat}}}
      Sprite Size Mode {
       bIsDynamic=false
       bIsEditable=true
       bIsStaticSwitch=true
       bIsVisible=true
       type {classStructOrEnum {refPath=/Niagara/Enums/ENiagara_SizeScaleMode}}
      }
      Sprite Size {
       bIsDynamic=false
       bIsEditable=false
       bIsStaticSwitch=false
       bIsVisible=false
       type {classStructOrEnum {refPath=&Vector2f}}
      }
      Uniform Sprite Size {
       bIsDynamic=false
       bIsEditable=false
       bIsStaticSwitch=false
       bIsVisible=false
       type {classStructOrEnum {refPath=&NiagaraFloat}}
      }
      Uniform Sprite Size Min {
       bIsDynamic=false
       bIsEditable=false
       bIsStaticSwitch=false
       bIsVisible=false
       type {classStructOrEnum {refPath=&NiagaraFloat}}
      }
      Sprite Size Min {
       bIsDynamic=false
       bIsEditable=false
       bIsStaticSwitch=false
       bIsVisible=false
       type {classStructOrEnum {refPath=&Vector2f}}
      }
      Sprite Size Max {
       bIsDynamic=false
       bIsEditable=false
       bIsStaticSwitch=false
       bIsVisible=false
       type {classStructOrEnum {refPath=&Vector2f}}
      }
      Uniform Sprite Size Max {
       bIsDynamic=false
       bIsEditable=false
       bIsStaticSwitch=false
       bIsVisible=false
       type {classStructOrEnum {refPath=&NiagaraFloat}}
      }
      Sprite Size Randomness Mode {
       bIsDynamic=false
       bIsEditable=false
       bIsStaticSwitch=true
       bIsVisible=false
       type {classStructOrEnum {refPath=&ENiagaraRandomnessMode}}
      }
      Sprite Rotation Mode {
       bIsDynamic=false
       bIsEditable=true
       bIsStaticSwitch=true
       bIsVisible=true
       type {classStructOrEnum {refPath=/Niagara/Enums/ENiagara_SpriteRotationMode}}
      }
      Sprite Rotation Angle Min {
       bIsDynamic=false
       bIsEditable=false
       bIsStaticSwitch=false
       bIsVisible=false
       type {classStructOrEnum {refPath=&NiagaraFloat}}
      }
      Sprite Rotation Angle {
       bIsDynamic=false
       bIsEditable=false
       bIsStaticSwitch=false
       bIsVisible=false
       type {classStructOrEnum {refPath=&NiagaraFloat}}
      }
      Sprite Rotation Angle Max {
       bIsDynamic=false
       bIsEditable=false
       bIsStaticSwitch=false
       bIsVisible=false
       type {classStructOrEnum {refPath=&NiagaraFloat}}
      }
      Sprite Rotation Randomness Mode {
       bIsDynamic=false
       bIsEditable=false
       bIsStaticSwitch=true
       bIsVisible=false
       type {classStructOrEnum {refPath=&ENiagaraRandomnessMode}}
      }
      Sprite UV Mode {
       bIsDynamic=false
       bIsEditable=true
       bIsStaticSwitch=true
       bIsVisible=true
       type {classStructOrEnum {refPath=/Niagara/Enums/ENiagara_UVFlippingMode}}
      }
      Sprite UV Randomness Mode {
       bIsDynamic=false
       bIsEditable=false
       bIsStaticSwitch=true
       bIsVisible=false
       type {classStructOrEnum {refPath=&ENiagaraRandomnessMode}}
      }
      Sprite UV Scale {
       bIsDynamic=false
       bIsEditable=false
       bIsStaticSwitch=false
       bIsVisible=false
       type {classStructOrEnum {refPath=&Vector2f}}
      }
      Mesh Scale Mode {
       bIsDynamic=false
       bIsEditable=true
       bIsStaticSwitch=true
       bIsVisible=true
       type {classStructOrEnum {refPath=&ENiagara_SizeScaleMode}}
      }
      Mesh Uniform Scale Min {bIsDynamic=false, bIsEditable=true, bIsStaticSwitch=false, bIsVisible=true, type {classStructOrEnum {refPath=&NiagaraFloat}}}
      Mesh Uniform Scale {
       bIsDynamic=false
       bIsEditable=false
       bIsStaticSwitch=false
       bIsVisible=false
       type {classStructOrEnum {refPath=&NiagaraFloat}}
      }
      Mesh Scale Min {
       bIsDynamic=false
       bIsEditable=false
       bIsStaticSwitch=false
       bIsVisible=false
       type {classStructOrEnum {refPath=&Vector3f}}
      }
      Mesh Scale {
       bIsDynamic=false
       bIsEditable=false
       bIsStaticSwitch=false
       bIsVisible=false
       type {classStructOrEnum {refPath=&Vector3f}}
      }
      Mesh Scale Max {
       bIsDynamic=false
       bIsEditable=false
       bIsStaticSwitch=false
       bIsVisible=false
       type {classStructOrEnum {refPath=&Vector3f}}
      }
      Mesh Uniform Scale Max {bIsDynamic=false, bIsEditable=true, bIsStaticSwitch=false, bIsVisible=true, type {classStructOrEnum {refPath=&NiagaraFloat}}}
      Mesh Scale Randomness Mode {
       bIsDynamic=false
       bIsEditable=true
       bIsStaticSwitch=true
       bIsVisible=true
       type {classStructOrEnum {refPath=&ENiagaraRandomnessMode}}
      }
      Mesh Renderer Array Visibility Mode {
       bIsDynamic=false
       bIsEditable=true
       bIsStaticSwitch=true
       bIsVisible=true
       type {classStructOrEnum {refPath=/Niagara/Enums/Ribbons/ENiagara_UnsetDirectSetRandom}}
      }
      Mesh Renderer Mesh Index {
       bIsDynamic=false
       bIsEditable=false
       bIsStaticSwitch=false
       bIsVisible=false
       type {classStructOrEnum {refPath=&NiagaraInt32}}
      }
      Wrap Index By Mesh Count {bIsDynamic=false, bIsEditable=false, bIsStaticSwitch=true, bIsVisible=false, type {classStructOrEnum {refPath=&NiagaraBool}}}
      Mesh Renderer Info {
       bIsDynamic=false
       bIsEditable=true
       bIsStaticSwitch=false
       bIsVisible=true
       type {classStructOrEnum {refPath=&NiagaraDataInterfaceMeshRendererInfo}}
      }
      Mesh Renderer Random Seed {bIsDynamic=false, bIsEditable=true, bIsStaticSwitch=false, bIsVisible=true, type {classStructOrEnum {refPath=&NiagaraInt32}}}
      Ribbon Width Mode {
       bIsDynamic=false
       bIsEditable=true
       bIsStaticSwitch=true
       bIsVisible=true
       type {classStructOrEnum {refPath=/Niagara/Enums/Ribbons/ENiagara_UnsetDirectSet}}
      }
      Ribbon Width {
       bIsDynamic=false
       bIsEditable=false
       bIsStaticSwitch=false
       bIsVisible=false
       type {classStructOrEnum {refPath=&NiagaraFloat}}
      }
      Ribbon Facing Mode {
       bIsDynamic=false
       bIsEditable=true
       bIsStaticSwitch=true
       bIsVisible=true
       type {classStructOrEnum {refPath=&ENiagara_UnsetDirectSet}}
      }
      Ribbon Facing Vector {
       bIsDynamic=false
       bIsEditable=false
       bIsStaticSwitch=false
       bIsVisible=false
       type {classStructOrEnum {refPath=&Vector3f}}
      }
      Ribbon Twist Mode {
       bIsDynamic=false
       bIsEditable=true
       bIsStaticSwitch=true
       bIsVisible=true
       type {classStructOrEnum {refPath=&ENiagara_UnsetDirectSet}}
      }
      "Ribbon Twist " {
       bIsDynamic=false
       bIsEditable=false
       bIsStaticSwitch=false
       bIsVisible=false
       type {classStructOrEnum {refPath=&NiagaraFloat}}
      }
     ]
     moduleScript {refPath=&InitializeParticle}
    }
    {
     bIsSetParametersModule=false
     enabled=true
     moduleName=ShapeLocation
     inputs [
      Shape Primitive {
       bIsDynamic=false
       bIsEditable=true
       bIsStaticSwitch=true
       bIsVisible=true
       type {classStructOrEnum {refPath=/Niagara/Enums/Location/ENiagara_LocationShapes}}
      }
      Box / Plane Mode {
       bIsDynamic=false
       bIsEditable=false
       bIsStaticSwitch=true
       bIsVisible=false
       type {classStructOrEnum {refPath=/Niagara/Enums/Location/ENiagara_BoxPlaneMode}}
      }
      Cone Mode {
       bIsDynamic=false
       bIsEditable=false
       bIsStaticSwitch=true
       bIsVisible=false
       type {classStructOrEnum {refPath=/Niagara/Enums/Location/ENiagara_ConeMode}}
      }
      Cylinder Height {
       bIsDynamic=false
       bIsEditable=false
       bIsStaticSwitch=false
       bIsVisible=false
       type {classStructOrEnum {refPath=&NiagaraFloat}}
      }
      Sphere Radius {
       bIsDynamic=false
       bIsEditable=false
       bIsStaticSwitch=false
       bIsVisible=false
       type {classStructOrEnum {refPath=&NiagaraFloat}}
      }
      Ring / Disc Mode {
       bIsDynamic=false
       bIsEditable=true
       bIsStaticSwitch=true
       bIsVisible=true
       type {classStructOrEnum {refPath=/Niagara/Enums/Location/ENiagaraRingDiscMode}}
      }
      Torus Mode {
       bIsDynamic=false
       bIsEditable=false
       bIsStaticSwitch=true
       bIsVisible=false
       type {classStructOrEnum {refPath=/Niagara/Enums/ENiagaraShapeTorusMode}}
      }
      Knot Ring Loops {
       bIsDynamic=false
       bIsEditable=false
       bIsStaticSwitch=false
       bIsVisible=false
       type {classStructOrEnum {refPath=&NiagaraInt32}}
      }
      Knot Helix Loops {
       bIsDynamic=false
       bIsEditable=false
       bIsStaticSwitch=false
       bIsVisible=false
       type {classStructOrEnum {refPath=&NiagaraInt32}}
      }
      Enable Torus Knot Radius {bIsDynamic=false, bIsEditable=false, bIsStaticSwitch=false, bIsVisible=false, type {classStructOrEnum {refPath=&NiagaraBool}}}
      Torus Knot Tube Radius {
       bIsDynamic=false
       bIsEditable=false
       bIsStaticSwitch=false
       bIsVisible=false
       type {classStructOrEnum {refPath=&NiagaraFloat}}
      }
      Cylinder Radius {
       bIsDynamic=false
       bIsEditable=false
       bIsStaticSwitch=false
       bIsVisible=false
       type {classStructOrEnum {refPath=&NiagaraFloat}}
      }
      Cylinder Height Midpoint {
       bIsDynamic=false
       bIsEditable=false
       bIsStaticSwitch=false
       bIsVisible=false
       type {classStructOrEnum {refPath=&NiagaraFloat}}
      }
      Ring Radius {bIsDynamic=false, bIsEditable=true, bIsStaticSwitch=false, bIsVisible=true, type {classStructOrEnum {refPath=&NiagaraFloat}}}
      Box Size {
       bIsDynamic=false
       bIsEditable=false
       bIsStaticSwitch=false
       bIsVisible=false
       type {classStructOrEnum {refPath=&Vector3f}}
      }
      Box Midpoint {
       bIsDynamic=false
       bIsEditable=false
       bIsStaticSwitch=false
       bIsVisible=false
       type {classStructOrEnum {refPath=&Vector3f}}
      }
      Cone Length {
       bIsDynamic=false
       bIsEditable=false
       bIsStaticSwitch=false
       bIsVisible=false
       type {classStructOrEnum {refPath=&NiagaraFloat}}
      }
      Plane Size {
       bIsDynamic=false
       bIsEditable=false
       bIsStaticSwitch=false
       bIsVisible=false
       type {classStructOrEnum {refPath=&Vector2f}}
      }
      Wedge Horizontal Angle {
       bIsDynamic=false
       bIsEditable=false
       bIsStaticSwitch=false
       bIsVisible=false
       type {classStructOrEnum {refPath=&NiagaraFloat}}
      }
      Large Radius {
       bIsDynamic=false
       bIsEditable=false
       bIsStaticSwitch=false
       bIsVisible=false
       type {classStructOrEnum {refPath=&NiagaraFloat}}
      }
      Plane Midpoint {
       bIsDynamic=false
       bIsEditable=false
       bIsStaticSwitch=false
       bIsVisible=false
       type {classStructOrEnum {refPath=&Vector2f}}
      }
      Wedge Vertical Angle {
       bIsDynamic=false
       bIsEditable=false
       bIsStaticSwitch=false
       bIsVisible=false
       type {classStructOrEnum {refPath=&NiagaraFloat}}
      }
      Cone Angle {
       bIsDynamic=false
       bIsEditable=false
       bIsStaticSwitch=false
       bIsVisible=false
       type {classStructOrEnum {refPath=&NiagaraFloat}}
      }
      Handle Radius {
       bIsDynamic=false
       bIsEditable=false
       bIsStaticSwitch=false
       bIsVisible=false
       type {classStructOrEnum {refPath=&NiagaraFloat}}
      }
      Cone Inner Angle {
       bIsDynamic=false
       bIsEditable=false
       bIsStaticSwitch=false
       bIsVisible=false
       type {classStructOrEnum {refPath=&NiagaraFloat}}
      }
      Cone Radial Angle {
       bIsDynamic=false
       bIsEditable=false
       bIsStaticSwitch=false
       bIsVisible=false
       type {classStructOrEnum {refPath=&NiagaraFloat}}
      }
      Cone Angle Type {
       bIsDynamic=false
       bIsEditable=false
       bIsStaticSwitch=true
       bIsVisible=false
       type {classStructOrEnum {refPath=/Niagara/Enums/Angles/ENiagara_AngleInput}}
      }
      Flatten Endcaps {bIsDynamic=false, bIsEditable=false, bIsStaticSwitch=true, bIsVisible=false, type {classStructOrEnum {refPath=&NiagaraBool}}}
      Apply To Particle Position {
       bIsDynamic=false
       bIsEditable=true
       bIsStaticSwitch=true
       bIsVisible=true
       type {classStructOrEnum {refPath=/Niagara/Enums/ENiagara_AttributeSamplingApplyOutput}}
      }
      Cone Surface Distribution {
       bIsDynamic=false
       bIsEditable=false
       bIsStaticSwitch=false
       bIsVisible=false
       type {classStructOrEnum {refPath=&NiagaraFloat}}
      }
      Ring / Disc Distribution Mode {
       bIsDynamic=false
       bIsEditable=true
       bIsStaticSwitch=true
       bIsVisible=true
       type {classStructOrEnum {refPath=/Niagara/Enums/ENiagaraSphereDistributionMode}}
      }
      Sphere Distribution {
       bIsDynamic=false
       bIsEditable=false
       bIsStaticSwitch=true
       bIsVisible=false
       type {classStructOrEnum {refPath=&ENiagaraSphereDistributionMode}}
      }
      Torus Distribution Mode {
       bIsDynamic=false
       bIsEditable=false
       bIsStaticSwitch=true
       bIsVisible=false
       type {classStructOrEnum {refPath=/Niagara/Enums/ENiagaraTorusDistributionMode}}
      }
      Sphere Surface Distribution {
       bIsDynamic=false
       bIsEditable=false
       bIsStaticSwitch=false
       bIsVisible=false
       type {classStructOrEnum {refPath=&NiagaraFloat}}
      }
      Hemisphere Distribution {
       bIsDynamic=false
       bIsEditable=false
       bIsStaticSwitch=false
       bIsVisible=false
       type {classStructOrEnum {refPath=&Vector2f}}
      }
      Cylinder Distribution {
       bIsDynamic=false
       bIsEditable=false
       bIsStaticSwitch=true
       bIsVisible=false
       type {classStructOrEnum {refPath=/Niagara/Enums/Location/ENiagara_CylinderMode}}
      }
      Surface Only {bIsDynamic=false, bIsEditable=false, bIsStaticSwitch=false, bIsVisible=false, type {classStructOrEnum {refPath=&NiagaraBool}}}
      Surface Only Band Thickness {
       bIsDynamic=false
       bIsEditable=false
       bIsStaticSwitch=false
       bIsVisible=false
       type {classStructOrEnum {refPath=&NiagaraFloat}}
      }
      Use Endcaps In Surface Only Mode {bIsDynamic=false, bIsEditable=false, bIsStaticSwitch=true, bIsVisible=false, type {classStructOrEnum {refPath=&NiagaraBool}}}
      Hemicircle X {bIsDynamic=false, bIsEditable=false, bIsStaticSwitch=true, bIsVisible=false, type {classStructOrEnum {refPath=&NiagaraBool}}}
      Hemicircle Y {bIsDynamic=false, bIsEditable=false, bIsStaticSwitch=true, bIsVisible=false, type {classStructOrEnum {refPath=&NiagaraBool}}}
      Disc Coverage {bIsDynamic=false, bIsEditable=true, bIsStaticSwitch=false, bIsVisible=true, type {classStructOrEnum {refPath=&NiagaraFloat}}}
      Hemicircle Internal Cap {bIsDynamic=false, bIsEditable=false, bIsStaticSwitch=true, bIsVisible=false, type {classStructOrEnum {refPath=&NiagaraBool}}}
      U Position {bIsDynamic=true, bIsEditable=false, bIsStaticSwitch=false, bIsVisible=false, type {classStructOrEnum {refPath=&NiagaraFloat}}}
      Enable Lathe Profile {bIsDynamic=false, bIsEditable=false, bIsStaticSwitch=true, bIsVisible=false, type {classStructOrEnum {refPath=&NiagaraBool}}}
      Lathe Profile {
       bIsDynamic=false
       bIsEditable=false
       bIsStaticSwitch=false
       bIsVisible=false
       type {classStructOrEnum {refPath=&NiagaraDataInterfaceCurve}}
      }
      Surface Distribution {
       bIsDynamic=false
       bIsEditable=false
       bIsStaticSwitch=false
       bIsVisible=false
       type {classStructOrEnum {refPath=&NiagaraFloat}}
      }
      Surface Expansion Mode {
       bIsDynamic=false
       bIsEditable=false
       bIsStaticSwitch=false
       bIsVisible=false
       type {classStructOrEnum {refPath=/Niagara/Enums/ENiagaraExpansionMode}}
      }
      U Distribution {bIsDynamic=false, bIsEditable=true, bIsStaticSwitch=false, bIsVisible=true, type {classStructOrEnum {refPath=&NiagaraFloat}}}
      V Position {
       bIsDynamic=false
       bIsEditable=false
       bIsStaticSwitch=false
       bIsVisible=false
       type {classStructOrEnum {refPath=&NiagaraFloat}}
      }
      Radius Position {
       bIsDynamic=false
       bIsEditable=false
       bIsStaticSwitch=false
       bIsVisible=false
       type {classStructOrEnum {refPath=&NiagaraFloat}}
      }
      Sphere Radius Position {
       bIsDynamic=false
       bIsEditable=false
       bIsStaticSwitch=false
       bIsVisible=false
       type {classStructOrEnum {refPath=&NiagaraFloat}}
      }
      V Distribution {
       bIsDynamic=false
       bIsEditable=false
       bIsStaticSwitch=false
       bIsVisible=false
       type {classStructOrEnum {refPath=&NiagaraFloat}}
      }
      Evaluation Type {
       bIsDynamic=false
       bIsEditable=false
       bIsStaticSwitch=true
       bIsVisible=false
       type {classStructOrEnum {refPath=/Niagara/Enums/ENiagaraRandomnessEvaluation}}
      }
      Uniform Distribution {
       bIsDynamic=false
       bIsEditable=false
       bIsStaticSwitch=false
       bIsVisible=false
       type {classStructOrEnum {refPath=&NiagaraFloat}}
      }
      Uniform Spiral Amount {
       bIsDynamic=false
       bIsEditable=false
       bIsStaticSwitch=false
       bIsVisible=false
       type {classStructOrEnum {refPath=&NiagaraFloat}}
      }
      Uniform Spiral Falloff {
       bIsDynamic=false
       bIsEditable=false
       bIsStaticSwitch=false
       bIsVisible=false
       type {classStructOrEnum {refPath=&NiagaraFloat}}
      }
      Uniform Count {
       bIsDynamic=false
       bIsEditable=false
       bIsStaticSwitch=false
       bIsVisible=false
       type {classStructOrEnum {refPath=&NiagaraInt32}}
      }
      Uniform Particle Index {
       bIsDynamic=false
       bIsEditable=false
       bIsStaticSwitch=false
       bIsVisible=false
       type {classStructOrEnum {refPath=&NiagaraInt32}}
      }
      Hemisphere Angle Type {
       bIsDynamic=false
       bIsEditable=false
       bIsStaticSwitch=true
       bIsVisible=false
       type {classStructOrEnum {refPath=&ENiagara_AngleInput}}
      }
      Uniform Endpoint Offset {
       bIsDynamic=false
       bIsEditable=false
       bIsStaticSwitch=false
       bIsVisible=false
       type {classStructOrEnum {refPath=&NiagaraFloat}}
      }
      MaskSpawnGroup {bIsDynamic=false, bIsEditable=true, bIsStaticSwitch=false, bIsVisible=true, type {classStructOrEnum {refPath=&NiagaraBool}}}
      Spawn Group Mask {bIsDynamic=false, bIsEditable=false, bIsStaticSwitch=false, bIsVisible=true, type {classStructOrEnum {refPath=&NiagaraInt32}}}
      Spawn Group {
       bIsDynamic=false
       bIsEditable=false
       bIsStaticSwitch=false
       bIsVisible=false
       type {classStructOrEnum {refPath=&NiagaraInt32}}
      }
      Randomness Mode {
       bIsDynamic=false
       bIsEditable=true
       bIsStaticSwitch=true
       bIsVisible=true
       type {classStructOrEnum {refPath=&ENiagaraRandomnessMode}}
      }
      Override Seed {bIsDynamic=false, bIsEditable=true, bIsStaticSwitch=false, bIsVisible=true, type {classStructOrEnum {refPath=&NiagaraBool}}}
      Random Seed {bIsDynamic=false, bIsEditable=false, bIsStaticSwitch=false, bIsVisible=true, type {classStructOrEnum {refPath=&NiagaraInt32}}}
      Fixed Random Seed {bIsDynamic=false, bIsEditable=false, bIsStaticSwitch=true, bIsVisible=false, type {classStructOrEnum {refPath=&NiagaraBool}}}
      Transform Order {
       bIsDynamic=false
       bIsEditable=true
       bIsStaticSwitch=true
       bIsVisible=true
       type {classStructOrEnum {refPath=/Niagara/Enums/Transforms/ENiagara_TransformOrder}}
      }
      Scale Mode {
       bIsDynamic=false
       bIsEditable=true
       bIsStaticSwitch=true
       bIsVisible=true
       type {classStructOrEnum {refPath=/Niagara/Enums/Transforms/ENiagara_ScaleMode}}
      }
      Non Uniform Scale {bIsDynamic=false, bIsEditable=true, bIsStaticSwitch=false, bIsVisible=true, type {classStructOrEnum {refPath=&Vector3f}}}
      UseOwnerScale {bIsDynamic=false, bIsEditable=true, bIsStaticSwitch=false, bIsVisible=true, type {classStructOrEnum {refPath=&NiagaraBool}}}
      Apply Owner Scale {bIsDynamic=false, bIsEditable=true, bIsStaticSwitch=false, bIsVisible=true, type {classStructOrEnum {refPath=&Vector3f}}}
      Rotation Mode {
       bIsDynamic=false
       bIsEditable=true
       bIsStaticSwitch=true
       bIsVisible=true
       type {classStructOrEnum {refPath=/Niagara/Enums/Transforms/ENiagara_RotationMode}}
      }
      Rotation Axis {
       bIsDynamic=false
       bIsEditable=false
       bIsStaticSwitch=false
       bIsVisible=false
       type {classStructOrEnum {refPath=&Vector3f}}
      }
      Rotation Matrix {
       bIsDynamic=false
       bIsEditable=false
       bIsStaticSwitch=false
       bIsVisible=false
       type {classStructOrEnum {refPath=/Script/Niagara.NiagaraMatrix}}
      }
      Rotation Quaternion {bIsDynamic=false, bIsEditable=false, bIsStaticSwitch=false, bIsVisible=false, type {classStructOrEnum {refPath=/Script/CoreUObject.Quat4f}}}
      Yaw / Pitch / Roll {
       bIsDynamic=false
       bIsEditable=false
       bIsStaticSwitch=false
       bIsVisible=false
       type {classStructOrEnum {refPath=&Vector3f}}
      }
      Rotation Angle {
       bIsDynamic=false
       bIsEditable=false
       bIsStaticSwitch=false
       bIsVisible=false
       type {classStructOrEnum {refPath=&NiagaraFloat}}
      }
      Invert Rotation Quaternion {bIsDynamic=false, bIsEditable=false, bIsStaticSwitch=false, bIsVisible=false, type {classStructOrEnum {refPath=&NiagaraBool}}}
      Additional Quaternion Rotation {bIsDynamic=false, bIsEditable=false, bIsStaticSwitch=false, bIsVisible=false, type {classStructOrEnum {refPath=&NiagaraBool}}}
      Additional Yaw / Pitch / Roll {
       bIsDynamic=false
       bIsEditable=false
       bIsStaticSwitch=false
       bIsVisible=false
       type {classStructOrEnum {refPath=&Vector3f}}
      }
      Rotation Angle Type {
       bIsDynamic=false
       bIsEditable=false
       bIsStaticSwitch=true
       bIsVisible=false
       type {classStructOrEnum {refPath=&ENiagara_AngleInput}}
      }
      Rotation Coordinate Space {
       bIsDynamic=false
       bIsEditable=true
       bIsStaticSwitch=false
       bIsVisible=true
       type {classStructOrEnum {refPath=&ENiagaraCoordinateSpace}}
      }
      Offset Mode {
       bIsDynamic=false
       bIsEditable=true
       bIsStaticSwitch=true
       bIsVisible=true
       type {classStructOrEnum {refPath=/Niagara/Enums/Transforms/ENiagara_OffsetMode}}
      }
      Shape Origin {
       bIsDynamic=false
       bIsEditable=true
       bIsStaticSwitch=false
       bIsVisible=true
       type {classStructOrEnum {refPath=&NiagaraPosition}}
      }
      Offset {
       bIsDynamic=false
       bIsEditable=false
       bIsStaticSwitch=false
       bIsVisible=false
       type {classStructOrEnum {refPath=&Vector3f}}
      }
      Offset Coordinate Space {
       bIsDynamic=false
       bIsEditable=false
       bIsStaticSwitch=false
       bIsVisible=false
       type {classStructOrEnum {refPath=&ENiagaraCoordinateSpace}}
      }
      Custom Transform Matrix {
       bIsDynamic=false
       bIsEditable=false
       bIsStaticSwitch=false
       bIsVisible=false
       type {classStructOrEnum {refPath=&NiagaraMatrix}}
      }
      Transform Method {
       bIsDynamic=false
       bIsEditable=true
       bIsStaticSwitch=true
       bIsVisible=true
       type {classStructOrEnum {refPath=/Niagara/Enums/Transforms/ENiagara_TransformType}}
      }
     ]
     moduleScript {refPath=&ShapeLocation}
    }
    {
     bIsSetParametersModule=false
     enabled=true
     moduleName=InitialMeshOrientation
     inputs [
      Mesh Orientation Mode {
       bIsDynamic=false
       bIsEditable=true
       bIsStaticSwitch=true
       bIsVisible=true
       type {classStructOrEnum {refPath=/Niagara/Enums/Transforms/ENiagara_MeshOrientationOptions}}
      }
      Orientation Quaternion {bIsDynamic=false, bIsEditable=false, bIsStaticSwitch=false, bIsVisible=false, type {classStructOrEnum {refPath=&Quat4f}}}
      Orientation Matrix {
       bIsDynamic=false
       bIsEditable=false
       bIsStaticSwitch=false
       bIsVisible=false
       type {classStructOrEnum {refPath=&NiagaraMatrix}}
      }
      Orientation Vector {
       bIsDynamic=false
       bIsEditable=false
       bIsStaticSwitch=false
       bIsVisible=false
       type {classStructOrEnum {refPath=&Vector3f}}
      }
      Orientation Coordinate Space {
       bIsDynamic=false
       bIsEditable=false
       bIsStaticSwitch=true
       bIsVisible=false
       type {classStructOrEnum {refPath=&ENiagaraCoordinateSpace}}
      }
      Mesh Axis To Orient {
       bIsDynamic=false
       bIsEditable=false
       bIsStaticSwitch=false
       bIsVisible=false
       type {classStructOrEnum {refPath=&Vector3f}}
      }
      Random Seed {bIsDynamic=false, bIsEditable=true, bIsStaticSwitch=false, bIsVisible=true, type {classStructOrEnum {refPath=&NiagaraInt32}}}
      Use Rotation Vector {bIsDynamic=false, bIsEditable=true, bIsStaticSwitch=false, bIsVisible=true, type {classStructOrEnum {refPath=&NiagaraBool}}}
      Rotation {bIsDynamic=false, bIsEditable=false, bIsStaticSwitch=false, bIsVisible=true, type {classStructOrEnum {refPath=&Vector3f}}}
      Rotation Coordinate Space {
       bIsDynamic=false
       bIsEditable=false
       bIsStaticSwitch=true
       bIsVisible=true
       type {classStructOrEnum {refPath=/Niagara/Enums/ENiagaraMeshTransforms}}
      }
     ]
     moduleScript {refPath=&InitialMeshOrientation}
    }
    {
     bIsSetParametersModule=false
     enabled=true
     moduleName=AddVelocity
     inputs [
      Velocity Mode {
       bIsDynamic=false
       bIsEditable=true
       bIsStaticSwitch=true
       bIsVisible=true
       type {classStructOrEnum {refPath=/Niagara/Enums/Utility/ENiagara_VelocityMode}}
      }
      Velocity {
       bIsDynamic=false
       bIsEditable=false
       bIsStaticSwitch=false
       bIsVisible=false
       type {classStructOrEnum {refPath=&Vector3f}}
      }
      Velocity Speed {bIsDynamic=true, bIsEditable=true, bIsStaticSwitch=false, bIsVisible=true, type {classStructOrEnum {refPath=&NiagaraFloat}}}
      Velocity Speed Scale {
       bIsDynamic=false
       bIsEditable=false
       bIsStaticSwitch=false
       bIsVisible=false
       type {classStructOrEnum {refPath=&NiagaraFloat}}
      }
      Module.Use Falloff Distance {bIsDynamic=false, bIsEditable=true, bIsStaticSwitch=false, bIsVisible=true, type {classStructOrEnum {refPath=&NiagaraBool}}}
      Constrain To Radius {bIsDynamic=false, bIsEditable=false, bIsStaticSwitch=false, bIsVisible=true, type {classStructOrEnum {refPath=&NiagaraFloat}}}
      Distribution Along Cone Axis {
       bIsDynamic=false
       bIsEditable=false
       bIsStaticSwitch=false
       bIsVisible=false
       type {classStructOrEnum {refPath=&NiagaraFloat}}
      }
      Radius Falloff Near / Far {
       bIsDynamic=false
       bIsEditable=false
       bIsStaticSwitch=false
       bIsVisible=false
       type {classStructOrEnum {refPath=&Vector2f}}
      }
      Radius Falloff Exponent {
       bIsDynamic=false
       bIsEditable=false
       bIsStaticSwitch=false
       bIsVisible=false
       type {classStructOrEnum {refPath=&NiagaraFloat}}
      }
      Speed Falloff From Cone Axis {
       bIsDynamic=false
       bIsEditable=false
       bIsStaticSwitch=false
       bIsVisible=false
       type {classStructOrEnum {refPath=&NiagaraFloat}}
      }
      Invert Velocity Falloff {bIsDynamic=false, bIsEditable=false, bIsStaticSwitch=false, bIsVisible=false, type {classStructOrEnum {refPath=&NiagaraBool}}}
      Override Seed {bIsDynamic=false, bIsEditable=true, bIsStaticSwitch=false, bIsVisible=true, type {classStructOrEnum {refPath=&NiagaraBool}}}
      Random Seed {bIsDynamic=false, bIsEditable=false, bIsStaticSwitch=false, bIsVisible=true, type {classStructOrEnum {refPath=&NiagaraInt32}}}
      Fixed Random Seed {bIsDynamic=false, bIsEditable=false, bIsStaticSwitch=true, bIsVisible=false, type {classStructOrEnum {refPath=&NiagaraBool}}}
      Randomness Mode {
       bIsDynamic=false
       bIsEditable=true
       bIsStaticSwitch=true
       bIsVisible=true
       type {classStructOrEnum {refPath=&ENiagaraRandomnessMode}}
      }
      Cone Axis {
       bIsDynamic=false
       bIsEditable=false
       bIsStaticSwitch=false
       bIsVisible=false
       type {classStructOrEnum {refPath=&Vector3f}}
      }
      Cone Angle {
       bIsDynamic=false
       bIsEditable=false
       bIsStaticSwitch=false
       bIsVisible=false
       type {classStructOrEnum {refPath=&NiagaraFloat}}
      }
      Inner Cone Angle {
       bIsDynamic=false
       bIsEditable=false
       bIsStaticSwitch=false
       bIsVisible=false
       type {classStructOrEnum {refPath=&NiagaraFloat}}
      }
      Cone Angle Mode {
       bIsDynamic=false
       bIsEditable=false
       bIsStaticSwitch=true
       bIsVisible=false
       type {classStructOrEnum {refPath=&ENiagara_AngleInput}}
      }
      Rotation Mode {
       bIsDynamic=false
       bIsEditable=false
       bIsStaticSwitch=true
       bIsVisible=false
       type {classStructOrEnum {refPath=&ENiagara_RotationMode}}
      }
      Rotation Axis {
       bIsDynamic=false
       bIsEditable=false
       bIsStaticSwitch=false
       bIsVisible=false
       type {classStructOrEnum {refPath=&Vector3f}}
      }
      Yaw / Pitch / Roll {
       bIsDynamic=false
       bIsEditable=false
       bIsStaticSwitch=false
       bIsVisible=false
       type {classStructOrEnum {refPath=&Vector3f}}
      }
      Rotation Quaternion {bIsDynamic=false, bIsEditable=false, bIsStaticSwitch=false, bIsVisible=false, type {classStructOrEnum {refPath=&Quat4f}}}
      Rotation Matrix {
       bIsDynamic=false
       bIsEditable=false
       bIsStaticSwitch=false
       bIsVisible=false
       type {classStructOrEnum {refPath=&NiagaraMatrix}}
      }
      Rotation Angle {
       bIsDynamic=false
       bIsEditable=false
       bIsStaticSwitch=false
       bIsVisible=false
       type {classStructOrEnum {refPath=&NiagaraFloat}}
      }
      Invert Rotation Quaternion {bIsDynamic=false, bIsEditable=false, bIsStaticSwitch=false, bIsVisible=false, type {classStructOrEnum {refPath=&NiagaraBool}}}
      Additional Quaternion Rotation {bIsDynamic=false, bIsEditable=false, bIsStaticSwitch=false, bIsVisible=false, type {classStructOrEnum {refPath=&NiagaraBool}}}
      Additional Yaw / Pitch / Roll {
       bIsDynamic=false
       bIsEditable=false
       bIsStaticSwitch=false
       bIsVisible=false
       type {classStructOrEnum {refPath=&Vector3f}}
      }
      Rotation Angle Type {
       bIsDynamic=false
       bIsEditable=false
       bIsStaticSwitch=true
       bIsVisible=false
       type {classStructOrEnum {refPath=&ENiagara_AngleInput}}
      }
      Rotation Coordinate Space {
       bIsDynamic=false
       bIsEditable=false
       bIsStaticSwitch=false
       bIsVisible=false
       type {classStructOrEnum {refPath=&ENiagaraCoordinateSpace}}
      }
      Velocity Origin {
       bIsDynamic=false
       bIsEditable=true
       bIsStaticSwitch=false
       bIsVisible=true
       type {classStructOrEnum {refPath=&NiagaraPosition}}
      }
      Origin Offset {bIsDynamic=false, bIsEditable=true, bIsStaticSwitch=false, bIsVisible=true, type {classStructOrEnum {refPath=&Vector3f}}}
      Origin Offset Coordinate Space {
       bIsDynamic=false
       bIsEditable=true
       bIsStaticSwitch=false
       bIsVisible=true
       type {classStructOrEnum {refPath=&ENiagaraCoordinateSpace}}
      }
      Default Position {
       bIsDynamic=false
       bIsEditable=true
       bIsStaticSwitch=false
       bIsVisible=true
       type {classStructOrEnum {refPath=&NiagaraPosition}}
      }
     ]
     moduleScript {refPath=&AddVelocity}
    }
   ]
  }
  particleUpdateScript {
   scriptName=ParticleUpdateScript
   modules [
    {
     bIsSetParametersModule=false
     enabled=true
     moduleName=ParticleState
     inputs [
      Kill Particles When Lifetime Has Elapsed {bIsDynamic=false, bIsEditable=true, bIsStaticSwitch=true, bIsVisible=true, type {classStructOrEnum {refPath=&NiagaraBool}}}
      Loop Particles Lifetime {bIsDynamic=false, bIsEditable=false, bIsStaticSwitch=false, bIsVisible=false, type {classStructOrEnum {refPath=&NiagaraBool}}}
      Let Infinitely Lived Particles Die When Emitter Deactivates {bIsDynamic=false, bIsEditable=false, bIsStaticSwitch=false, bIsVisible=false, type {classStructOrEnum {refPath=&NiagaraBool}}}
      Lifetime {bIsDynamic=false, bIsEditable=true, bIsStaticSwitch=false, bIsVisible=true, type {classStructOrEnum {refPath=&NiagaraFloat}}}
      DeltaTime {bIsDynamic=false, bIsEditable=true, bIsStaticSwitch=false, bIsVisible=true, type {classStructOrEnum {refPath=&NiagaraFloat}}}
     ]
     moduleScript {refPath=&ParticleState}
    }
    {
     bIsSetParametersModule=false
     enabled=true
     moduleName=ScaleMeshSize
     inputs [
      Scale Factor {bIsDynamic=true, bIsEditable=true, bIsStaticSwitch=false, bIsVisible=true, type {classStructOrEnum {refPath=&Vector3f}}}
      Initial Mesh Scale {bIsDynamic=false, bIsEditable=true, bIsStaticSwitch=false, bIsVisible=true, type {classStructOrEnum {refPath=&Vector3f}}}
     ]
     moduleScript {refPath=&ScaleMeshSize}
    }
    {
     bIsSetParametersModule=false
     enabled=true
     moduleName=GravityForce
     inputs [
      Gravity {bIsDynamic=false, bIsEditable=true, bIsStaticSwitch=false, bIsVisible=true, type {classStructOrEnum {refPath=&Vector3f}}}
      Coordinate Space {
       bIsDynamic=false
       bIsEditable=true
       bIsStaticSwitch=true
       bIsVisible=true
       type {classStructOrEnum {refPath=&ENiagaraCoordinateSpace}}
      }
     ]
     moduleScript {refPath=&GravityForce}
    }
    {
     bIsSetParametersModule=false
     enabled=true
     moduleName=Drag
     inputs [
      Use Linear Drag {bIsDynamic=false, bIsEditable=true, bIsStaticSwitch=false, bIsVisible=true, type {classStructOrEnum {refPath=&NiagaraBool}}}
      Drag {bIsDynamic=true, bIsEditable=true, bIsStaticSwitch=false, bIsVisible=true, type {classStructOrEnum {refPath=&NiagaraFloat}}}
      Use Rotational Drag {bIsDynamic=false, bIsEditable=true, bIsStaticSwitch=false, bIsVisible=true, type {classStructOrEnum {refPath=&NiagaraBool}}}
      Rotational Drag {bIsDynamic=false, bIsEditable=true, bIsStaticSwitch=false, bIsVisible=true, type {classStructOrEnum {refPath=&NiagaraFloat}}}
      Ignore Mass {bIsDynamic=false, bIsEditable=true, bIsStaticSwitch=false, bIsVisible=true, type {classStructOrEnum {refPath=&NiagaraBool}}}
     ]
     moduleScript {refPath=&Drag}
    }
    {
     bIsSetParametersModule=false
     enabled=true
     moduleName=MeshRotationForce
     inputs [
      Rotation {bIsDynamic=true, bIsEditable=true, bIsStaticSwitch=false, bIsVisible=true, type {classStructOrEnum {refPath=&Vector3f}}}
      Coordinate Space {
       bIsDynamic=false
       bIsEditable=true
       bIsStaticSwitch=true
       bIsVisible=true
       type {classStructOrEnum {refPath=&ENiagaraMeshTransforms}}
      }
      "Lever Radius (cm)" {bIsDynamic=false, bIsEditable=true, bIsStaticSwitch=false, bIsVisible=true, type {classStructOrEnum {refPath=&NiagaraFloat}}}
     ]
     moduleScript {refPath=&MeshRotationForce}
    }
    {
     bIsSetParametersModule=false
     enabled=true
     moduleName=Collision
     inputs [
      Correct Interpenetration {bIsDynamic=false, bIsEditable=true, bIsStaticSwitch=false, bIsVisible=true, type {classStructOrEnum {refPath=&NiagaraBool}}}
      Particle Mass {bIsDynamic=false, bIsEditable=true, bIsStaticSwitch=false, bIsVisible=true, type {classStructOrEnum {refPath=&NiagaraFloat}}}
      PhysicsForce {bIsDynamic=false, bIsEditable=true, bIsStaticSwitch=false, bIsVisible=true, type {classStructOrEnum {refPath=&Vector3f}}}
      CPU Collision Type {
       bIsDynamic=false
       bIsEditable=true
       bIsStaticSwitch=true
       bIsVisible=true
       type {classStructOrEnum {refPath=/Niagara/Enums/ENiagara_CPUCollisionType}}
      }
      GPU Collision Type {
       bIsDynamic=false
       bIsEditable=false
       bIsStaticSwitch=true
       bIsVisible=false
       type {classStructOrEnum {refPath=/Niagara/Enums/ENiagara_GPUCollisionType}}
      }
      Collision Enabled {bIsDynamic=true, bIsEditable=true, bIsStaticSwitch=false, bIsVisible=true, type {classStructOrEnum {refPath=&NiagaraBool}}}
      Ray Trace Query Properties {
       bIsDynamic=false
       bIsEditable=false
       bIsStaticSwitch=false
       bIsVisible=false
       type {classStructOrEnum {refPath=/Script/Niagara.NiagaraDataInterfaceAsyncGpuTrace}}
      }
      Z Depth Query Type {
       bIsDynamic=false
       bIsEditable=false
       bIsStaticSwitch=true
       bIsVisible=false
       type {classStructOrEnum {refPath=/Niagara/Enums/Collision/SceneDepthCollisionQueryMethod}}
      }
      Radius Calculation Type {
       bIsDynamic=false
       bIsEditable=true
       bIsStaticSwitch=true
       bIsVisible=true
       type {classStructOrEnum {refPath=/Niagara/Enums/ENiagaraCollisionRadiusOptions}}
      }
      Particle Radius {
       bIsDynamic=false
       bIsEditable=false
       bIsStaticSwitch=false
       bIsVisible=false
       type {classStructOrEnum {refPath=&NiagaraFloat}}
      }
      Method for Calculating Particle Radius {
       bIsDynamic=false
       bIsEditable=true
       bIsStaticSwitch=true
       bIsVisible=true
       type {classStructOrEnum {refPath=/Niagara/Enums/ENiagaraCalculateRadiusOptions}}
      }
      Mesh Dimensions {bIsDynamic=false, bIsEditable=true, bIsStaticSwitch=false, bIsVisible=true, type {classStructOrEnum {refPath=&Vector3f}}}
      Particle Radius Scale {bIsDynamic=false, bIsEditable=true, bIsStaticSwitch=false, bIsVisible=true, type {classStructOrEnum {refPath=&NiagaraFloat}}}
      Restitution {bIsDynamic=false, bIsEditable=true, bIsStaticSwitch=false, bIsVisible=true, type {classStructOrEnum {refPath=&NiagaraFloat}}}
      Restitution Coefficient Blending {
       bIsDynamic=false
       bIsEditable=true
       bIsStaticSwitch=true
       bIsVisible=true
       type {classStructOrEnum {refPath=/Niagara/Enums/ENiagaraRestitutionMergeType}}
      }
      Module.Randomize Collision Normal Vector {bIsDynamic=false, bIsEditable=true, bIsStaticSwitch=false, bIsVisible=true, type {classStructOrEnum {refPath=&NiagaraBool}}}
      Randomize Collision Normal {bIsDynamic=false, bIsEditable=false, bIsStaticSwitch=false, bIsVisible=true, type {classStructOrEnum {refPath=&NiagaraFloat}}}
      Control Roll On Collision {bIsDynamic=false, bIsEditable=true, bIsStaticSwitch=true, bIsVisible=true, type {classStructOrEnum {refPath=&NiagaraBool}}}
      ScaleRateOfRotation {bIsDynamic=false, bIsEditable=false, bIsStaticSwitch=false, bIsVisible=false, type {classStructOrEnum {refPath=&NiagaraBool}}}
      Radius Scale for Rotation {
       bIsDynamic=false
       bIsEditable=false
       bIsStaticSwitch=false
       bIsVisible=false
       type {classStructOrEnum {refPath=&NiagaraFloat}}
      }
      Simple Friction {bIsDynamic=false, bIsEditable=true, bIsStaticSwitch=true, bIsVisible=true, type {classStructOrEnum {refPath=&NiagaraBool}}}
      CPU Friction Blending {
       bIsDynamic=false
       bIsEditable=true
       bIsStaticSwitch=true
       bIsVisible=true
       type {classStructOrEnum {refPath=/Niagara/Enums/ENiagaraFrictionMergeType}}
      }
      Friction {bIsDynamic=false, bIsEditable=true, bIsStaticSwitch=false, bIsVisible=true, type {classStructOrEnum {refPath=&NiagaraFloat}}}
      Static Friction {
       bIsDynamic=false
       bIsEditable=false
       bIsStaticSwitch=false
       bIsVisible=false
       type {classStructOrEnum {refPath=&NiagaraFloat}}
      }
      Static Friction Engagement Speed {
       bIsDynamic=false
       bIsEditable=false
       bIsStaticSwitch=false
       bIsVisible=false
       type {classStructOrEnum {refPath=&NiagaraFloat}}
      }
      Friction During a Bounce {bIsDynamic=false, bIsEditable=true, bIsStaticSwitch=false, bIsVisible=true, type {classStructOrEnum {refPath=&NiagaraFloat}}}
      Analytical Collision Plane Space {
       bIsDynamic=false
       bIsEditable=false
       bIsStaticSwitch=false
       bIsVisible=false
       type {classStructOrEnum {refPath=&ENiagaraCoordinateSpace}}
      }
      Analytical Collision Normal 1 {
       bIsDynamic=false
       bIsEditable=false
       bIsStaticSwitch=false
       bIsVisible=false
       type {classStructOrEnum {refPath=&Vector3f}}
      }
      Analytical Collision Plane Position 1 {
       bIsDynamic=false
       bIsEditable=false
       bIsStaticSwitch=false
       bIsVisible=false
       type {classStructOrEnum {refPath=&Vector3f}}
      }
      Analytical Collision Normal 2 {
       bIsDynamic=false
       bIsEditable=false
       bIsStaticSwitch=false
       bIsVisible=false
       type {classStructOrEnum {refPath=&Vector3f}}
      }
      Analytical Collision Plane Position 2 {
       bIsDynamic=false
       bIsEditable=false
       bIsStaticSwitch=false
       bIsVisible=false
       type {classStructOrEnum {refPath=&Vector3f}}
      }
      Perform 2 Analytical Collision Queries {bIsDynamic=false, bIsEditable=false, bIsStaticSwitch=true, bIsVisible=false, type {classStructOrEnum {refPath=&NiagaraBool}}}
      Analytical Plane Debug Mode {bIsDynamic=false, bIsEditable=false, bIsStaticSwitch=true, bIsVisible=false, type {classStructOrEnum {refPath=&NiagaraInt32}}}
      Debug Plane 1 Extents {
       bIsDynamic=false
       bIsEditable=false
       bIsStaticSwitch=false
       bIsVisible=false
       type {classStructOrEnum {refPath=&Vector2f}}
      }
      Debug Plane 1 Color {
       bIsDynamic=false
       bIsEditable=false
       bIsStaticSwitch=false
       bIsVisible=false
       type {classStructOrEnum {refPath=&LinearColor}}
      }
      Debug Plane 2 Extents {
       bIsDynamic=false
       bIsEditable=false
       bIsStaticSwitch=false
       bIsVisible=false
       type {classStructOrEnum {refPath=&Vector2f}}
      }
      Debug Plane 2 Color {
       bIsDynamic=false
       bIsEditable=false
       bIsStaticSwitch=false
       bIsVisible=false
       type {classStructOrEnum {refPath=&LinearColor}}
      }
      Debug {
       bIsDynamic=false
       bIsEditable=false
       bIsStaticSwitch=false
       bIsVisible=false
       type {classStructOrEnum {refPath=&NiagaraDataInterfaceDebugDraw}}
      }
      Draw Particle Radius {bIsDynamic=false, bIsEditable=false, bIsStaticSwitch=false, bIsVisible=false, type {classStructOrEnum {refPath=&NiagaraBool}}}
      Allow Offscreen Depth Buffer Collision Particles To Live After a Bounce {bIsDynamic=false, bIsEditable=false, bIsStaticSwitch=false, bIsVisible=false, type {classStructOrEnum {refPath=&NiagaraBool}}}
      Kill Occluded Particles {bIsDynamic=false, bIsEditable=false, bIsStaticSwitch=false, bIsVisible=false, type {classStructOrEnum {refPath=&NiagaraBool}}}
      Max Distance To Collision Plane {
       bIsDynamic=false
       bIsEditable=false
       bIsStaticSwitch=false
       bIsVisible=false
       type {classStructOrEnum {refPath=&NiagaraFloat}}
      }
      Kill Offscreen Particles {bIsDynamic=false, bIsEditable=false, bIsStaticSwitch=false, bIsVisible=false, type {classStructOrEnum {refPath=&NiagaraBool}}}
      Maximum time offscreen {
       bIsDynamic=false
       bIsEditable=false
       bIsStaticSwitch=false
       bIsVisible=false
       type {classStructOrEnum {refPath=&NiagaraFloat}}
      }
      Kill Particles Lodged Within Meshes {bIsDynamic=false, bIsEditable=false, bIsStaticSwitch=false, bIsVisible=false, type {classStructOrEnum {refPath=&NiagaraBool}}}
      CPU Collision Trace Channel {
       bIsDynamic=false
       bIsEditable=true
       bIsStaticSwitch=false
       bIsVisible=true
       type {classStructOrEnum {refPath=/Script/Engine.ECollisionChannel}}
      }
      CPU Trace Vector Length Multiplier {bIsDynamic=false, bIsEditable=true, bIsStaticSwitch=false, bIsVisible=true, type {classStructOrEnum {refPath=&NiagaraFloat}}}
      GPU Trace Vector Length Multiplier {
       bIsDynamic=false
       bIsEditable=false
       bIsStaticSwitch=false
       bIsVisible=false
       type {classStructOrEnum {refPath=&NiagaraFloat}}
      }
      Max GPU Trace Length {
       bIsDynamic=false
       bIsEditable=false
       bIsStaticSwitch=false
       bIsVisible=false
       type {classStructOrEnum {refPath=&NiagaraFloat}}
      }
      Max CPU Trace Length {bIsDynamic=false, bIsEditable=true, bIsStaticSwitch=false, bIsVisible=true, type {classStructOrEnum {refPath=&NiagaraFloat}}}
      Enable Rest State {bIsDynamic=false, bIsEditable=true, bIsStaticSwitch=false, bIsVisible=true, type {classStructOrEnum {refPath=&NiagaraBool}}}
      Manually Enter Rest {bIsDynamic=false, bIsEditable=true, bIsStaticSwitch=false, bIsVisible=true, type {classStructOrEnum {refPath=&NiagaraBool}}}
      Module.EnableMaxCollisionCount {bIsDynamic=false, bIsEditable=true, bIsStaticSwitch=false, bIsVisible=true, type {classStructOrEnum {refPath=&NiagaraBool}}}
      Max Number Of Collisions {bIsDynamic=false, bIsEditable=false, bIsStaticSwitch=false, bIsVisible=true, type {classStructOrEnum {refPath=&NiagaraInt32}}}
      Maximum Penetration Correction Distance {bIsDynamic=false, bIsEditable=true, bIsStaticSwitch=false, bIsVisible=true, type {classStructOrEnum {refPath=&NiagaraFloat}}}
      Percentage of Penetration Before Rest {bIsDynamic=false, bIsEditable=true, bIsStaticSwitch=false, bIsVisible=true, type {classStructOrEnum {refPath=&NiagaraFloat}}}
      Rest State Time Range {bIsDynamic=false, bIsEditable=true, bIsStaticSwitch=false, bIsVisible=true, type {classStructOrEnum {refPath=&NiagaraFloat}}}
      Module.Control Rest State Via Surface Normal {bIsDynamic=false, bIsEditable=true, bIsStaticSwitch=false, bIsVisible=true, type {classStructOrEnum {refPath=&NiagaraBool}}}
      Valid Collision Plane Normal For Rest {bIsDynamic=false, bIsEditable=false, bIsStaticSwitch=false, bIsVisible=true, type {classStructOrEnum {refPath=&Vector3f}}}
      Age Colliding Particles {bIsDynamic=false, bIsEditable=true, bIsStaticSwitch=false, bIsVisible=true, type {classStructOrEnum {refPath=&NiagaraBool}}}
      Advanced Aging Rate {bIsDynamic=false, bIsEditable=false, bIsStaticSwitch=false, bIsVisible=true, type {classStructOrEnum {refPath=&NiagaraFloat}}}
      KillOnCollision {bIsDynamic=false, bIsEditable=true, bIsStaticSwitch=false, bIsVisible=true, type {classStructOrEnum {refPath=&NiagaraBool}}}
      Kill On Collision {bIsDynamic=false, bIsEditable=false, bIsStaticSwitch=false, bIsVisible=true, type {classStructOrEnum {refPath=&NiagaraInt32}}}
      DebugDraw {
       bIsDynamic=false
       bIsEditable=false
       bIsStaticSwitch=false
       bIsVisible=false
       type {classStructOrEnum {refPath=&NiagaraDataInterfaceDebugDraw}}
      }
     ]
     moduleScript {refPath=&Collision}
    }
    {
     bIsSetParametersModule=false
     enabled=true
     moduleName=Collision001
     inputs [
      Correct Interpenetration {bIsDynamic=false, bIsEditable=false, bIsStaticSwitch=false, bIsVisible=false, type {classStructOrEnum {refPath=&NiagaraBool}}}
      Particle Mass {bIsDynamic=false, bIsEditable=true, bIsStaticSwitch=false, bIsVisible=true, type {classStructOrEnum {refPath=&NiagaraFloat}}}
      PhysicsForce {bIsDynamic=false, bIsEditable=true, bIsStaticSwitch=false, bIsVisible=true, type {classStructOrEnum {refPath=&Vector3f}}}
      CPU Collision Type {
       bIsDynamic=false
       bIsEditable=true
       bIsStaticSwitch=true
       bIsVisible=true
       type {classStructOrEnum {refPath=&ENiagara_CPUCollisionType}}
      }
      GPU Collision Type {
       bIsDynamic=false
       bIsEditable=false
       bIsStaticSwitch=true
       bIsVisible=false
       type {classStructOrEnum {refPath=&ENiagara_GPUCollisionType}}
      }
      Collision Enabled {bIsDynamic=true, bIsEditable=true, bIsStaticSwitch=false, bIsVisible=true, type {classStructOrEnum {refPath=&NiagaraBool}}}
      Ray Trace Query Properties {
       bIsDynamic=false
       bIsEditable=false
       bIsStaticSwitch=false
       bIsVisible=false
       type {classStructOrEnum {refPath=&NiagaraDataInterfaceAsyncGpuTrace}}
      }
      Z Depth Query Type {
       bIsDynamic=false
       bIsEditable=false
       bIsStaticSwitch=true
       bIsVisible=false
       type {classStructOrEnum {refPath=&SceneDepthCollisionQueryMethod}}
      }
      Radius Calculation Type {
       bIsDynamic=false
       bIsEditable=true
       bIsStaticSwitch=true
       bIsVisible=true
       type {classStructOrEnum {refPath=&ENiagaraCollisionRadiusOptions}}
      }
      Particle Radius {
       bIsDynamic=false
       bIsEditable=false
       bIsStaticSwitch=false
       bIsVisible=false
       type {classStructOrEnum {refPath=&NiagaraFloat}}
      }
      Method for Calculating Particle Radius {
       bIsDynamic=false
       bIsEditable=true
       bIsStaticSwitch=true
       bIsVisible=true
       type {classStructOrEnum {refPath=&ENiagaraCalculateRadiusOptions}}
      }
      Mesh Dimensions {bIsDynamic=false, bIsEditable=true, bIsStaticSwitch=false, bIsVisible=true, type {classStructOrEnum {refPath=&Vector3f}}}
      Particle Radius Scale {bIsDynamic=false, bIsEditable=true, bIsStaticSwitch=false, bIsVisible=true, type {classStructOrEnum {refPath=&NiagaraFloat}}}
      Restitution {bIsDynamic=false, bIsEditable=true, bIsStaticSwitch=false, bIsVisible=true, type {classStructOrEnum {refPath=&NiagaraFloat}}}
      Restitution Coefficient Blending {
       bIsDynamic=false
       bIsEditable=true
       bIsStaticSwitch=true
       bIsVisible=true
       type {classStructOrEnum {refPath=&ENiagaraRestitutionMergeType}}
      }
      Module.Randomize Collision Normal Vector {bIsDynamic=false, bIsEditable=true, bIsStaticSwitch=false, bIsVisible=true, type {classStructOrEnum {refPath=&NiagaraBool}}}
      Randomize Collision Normal {bIsDynamic=false, bIsEditable=false, bIsStaticSwitch=false, bIsVisible=true, type {classStructOrEnum {refPath=&NiagaraFloat}}}
      Control Roll On Collision {bIsDynamic=false, bIsEditable=true, bIsStaticSwitch=true, bIsVisible=true, type {classStructOrEnum {refPath=&NiagaraBool}}}
      ScaleRateOfRotation {bIsDynamic=false, bIsEditable=false, bIsStaticSwitch=false, bIsVisible=false, type {classStructOrEnum {refPath=&NiagaraBool}}}
      Radius Scale for Rotation {
       bIsDynamic=false
       bIsEditable=false
       bIsStaticSwitch=false
       bIsVisible=false
       type {classStructOrEnum {refPath=&NiagaraFloat}}
      }
      Simple Friction {bIsDynamic=false, bIsEditable=true, bIsStaticSwitch=true, bIsVisible=true, type {classStructOrEnum {refPath=&NiagaraBool}}}
      CPU Friction Blending {
       bIsDynamic=false
       bIsEditable=true
       bIsStaticSwitch=true
       bIsVisible=true
       type {classStructOrEnum {refPath=&ENiagaraFrictionMergeType}}
      }
      Friction {bIsDynamic=false, bIsEditable=true, bIsStaticSwitch=false, bIsVisible=true, type {classStructOrEnum {refPath=&NiagaraFloat}}}
      Static Friction {
       bIsDynamic=false
       bIsEditable=false
       bIsStaticSwitch=false
       bIsVisible=false
       type {classStructOrEnum {refPath=&NiagaraFloat}}
      }
      Static Friction Engagement Speed {
       bIsDynamic=false
       bIsEditable=false
       bIsStaticSwitch=false
       bIsVisible=false
       type {classStructOrEnum {refPath=&NiagaraFloat}}
      }
      Friction During a Bounce {bIsDynamic=false, bIsEditable=true, bIsStaticSwitch=false, bIsVisible=true, type {classStructOrEnum {refPath=&NiagaraFloat}}}
      Analytical Collision Plane Space {
       bIsDynamic=false
       bIsEditable=true
       bIsStaticSwitch=false
       bIsVisible=true
       type {classStructOrEnum {refPath=&ENiagaraCoordinateSpace}}
      }
      Analytical Collision Normal 1 {bIsDynamic=false, bIsEditable=true, bIsStaticSwitch=false, bIsVisible=true, type {classStructOrEnum {refPath=&Vector3f}}}
      Analytical Collision Plane Position 1 {bIsDynamic=false, bIsEditable=true, bIsStaticSwitch=false, bIsVisible=true, type {classStructOrEnum {refPath=&Vector3f}}}
      Analytical Collision Normal 2 {bIsDynamic=false, bIsEditable=true, bIsStaticSwitch=false, bIsVisible=true, type {classStructOrEnum {refPath=&Vector3f}}}
      Analytical Collision Plane Position 2 {bIsDynamic=false, bIsEditable=true, bIsStaticSwitch=false, bIsVisible=true, type {classStructOrEnum {refPath=&Vector3f}}}
      Perform 2 Analytical Collision Queries {bIsDynamic=false, bIsEditable=true, bIsStaticSwitch=true, bIsVisible=true, type {classStructOrEnum {refPath=&NiagaraBool}}}
      Analytical Plane Debug Mode {bIsDynamic=false, bIsEditable=false, bIsStaticSwitch=true, bIsVisible=false, type {classStructOrEnum {refPath=&NiagaraInt32}}}
      Debug Plane 1 Extents {
       bIsDynamic=false
       bIsEditable=false
       bIsStaticSwitch=false
       bIsVisible=false
       type {classStructOrEnum {refPath=&Vector2f}}
      }
      Debug Plane 1 Color {
       bIsDynamic=false
       bIsEditable=false
       bIsStaticSwitch=false
       bIsVisible=false
       type {classStructOrEnum {refPath=&LinearColor}}
      }
      Debug Plane 2 Extents {
       bIsDynamic=false
       bIsEditable=false
       bIsStaticSwitch=false
       bIsVisible=false
       type {classStructOrEnum {refPath=&Vector2f}}
      }
      Debug Plane 2 Color {
       bIsDynamic=false
       bIsEditable=false
       bIsStaticSwitch=false
       bIsVisible=false
       type {classStructOrEnum {refPath=&LinearColor}}
      }
      Debug {
       bIsDynamic=false
       bIsEditable=false
       bIsStaticSwitch=false
       bIsVisible=false
       type {classStructOrEnum {refPath=&NiagaraDataInterfaceDebugDraw}}
      }
      Draw Particle Radius {bIsDynamic=false, bIsEditable=false, bIsStaticSwitch=false, bIsVisible=false, type {classStructOrEnum {refPath=&NiagaraBool}}}
      Allow Offscreen Depth Buffer Collision Particles To Live After a Bounce {bIsDynamic=false, bIsEditable=false, bIsStaticSwitch=false, bIsVisible=false, type {classStructOrEnum {refPath=&NiagaraBool}}}
      Kill Occluded Particles {bIsDynamic=false, bIsEditable=false, bIsStaticSwitch=false, bIsVisible=false, type {classStructOrEnum {refPath=&NiagaraBool}}}
      Max Distance To Collision Plane {
       bIsDynamic=false
       bIsEditable=false
       bIsStaticSwitch=false
       bIsVisible=false
       type {classStructOrEnum {refPath=&NiagaraFloat}}
      }
      Kill Offscreen Particles {bIsDynamic=false, bIsEditable=false, bIsStaticSwitch=false, bIsVisible=false, type {classStructOrEnum {refPath=&NiagaraBool}}}
      Maximum time offscreen {
       bIsDynamic=false
       bIsEditable=false
       bIsStaticSwitch=false
       bIsVisible=false
       type {classStructOrEnum {refPath=&NiagaraFloat}}
      }
      Kill Particles Lodged Within Meshes {bIsDynamic=false, bIsEditable=false, bIsStaticSwitch=false, bIsVisible=false, type {classStructOrEnum {refPath=&NiagaraBool}}}
      CPU Collision Trace Channel {
       bIsDynamic=false
       bIsEditable=false
       bIsStaticSwitch=false
       bIsVisible=false
       type {classStructOrEnum {refPath=&ECollisionChannel}}
      }
      CPU Trace Vector Length Multiplier {
       bIsDynamic=false
       bIsEditable=false
       bIsStaticSwitch=false
       bIsVisible=false
       type {classStructOrEnum {refPath=&NiagaraFloat}}
      }
      GPU Trace Vector Length Multiplier {
       bIsDynamic=false
       bIsEditable=false
       bIsStaticSwitch=false
       bIsVisible=false
       type {classStructOrEnum {refPath=&NiagaraFloat}}
      }
      Max GPU Trace Length {
       bIsDynamic=false
       bIsEditable=false
       bIsStaticSwitch=false
       bIsVisible=false
       type {classStructOrEnum {refPath=&NiagaraFloat}}
      }
      Max CPU Trace Length {
       bIsDynamic=false
       bIsEditable=false
       bIsStaticSwitch=false
       bIsVisible=false
       type {classStructOrEnum {refPath=&NiagaraFloat}}
      }
      Enable Rest State {bIsDynamic=false, bIsEditable=true, bIsStaticSwitch=false, bIsVisible=true, type {classStructOrEnum {refPath=&NiagaraBool}}}
      Manually Enter Rest {bIsDynamic=false, bIsEditable=true, bIsStaticSwitch=false, bIsVisible=true, type {classStructOrEnum {refPath=&NiagaraBool}}}
      Module.EnableMaxCollisionCount {bIsDynamic=false, bIsEditable=true, bIsStaticSwitch=false, bIsVisible=true, type {classStructOrEnum {refPath=&NiagaraBool}}}
      Max Number Of Collisions {bIsDynamic=false, bIsEditable=false, bIsStaticSwitch=false, bIsVisible=true, type {classStructOrEnum {refPath=&NiagaraInt32}}}
      Maximum Penetration Correction Distance {bIsDynamic=false, bIsEditable=true, bIsStaticSwitch=false, bIsVisible=true, type {classStructOrEnum {refPath=&NiagaraFloat}}}
      Percentage of Penetration Before Rest {bIsDynamic=false, bIsEditable=true, bIsStaticSwitch=false, bIsVisible=true, type {classStructOrEnum {refPath=&NiagaraFloat}}}
      Rest State Time Range {bIsDynamic=false, bIsEditable=true, bIsStaticSwitch=false, bIsVisible=true, type {classStructOrEnum {refPath=&NiagaraFloat}}}
      Module.Control Rest State Via Surface Normal {bIsDynamic=false, bIsEditable=true, bIsStaticSwitch=false, bIsVisible=true, type {classStructOrEnum {refPath=&NiagaraBool}}}
      Valid Collision Plane Normal For Rest {bIsDynamic=false, bIsEditable=false, bIsStaticSwitch=false, bIsVisible=true, type {classStructOrEnum {refPath=&Vector3f}}}
      Age Colliding Particles {bIsDynamic=false, bIsEditable=true, bIsStaticSwitch=false, bIsVisible=true, type {classStructOrEnum {refPath=&NiagaraBool}}}
      Advanced Aging Rate {bIsDynamic=false, bIsEditable=false, bIsStaticSwitch=false, bIsVisible=true, type {classStructOrEnum {refPath=&NiagaraFloat}}}
      KillOnCollision {bIsDynamic=false, bIsEditable=true, bIsStaticSwitch=false, bIsVisible=true, type {classStructOrEnum {refPath=&NiagaraBool}}}
      Kill On Collision {bIsDynamic=false, bIsEditable=false, bIsStaticSwitch=false, bIsVisible=true, type {classStructOrEnum {refPath=&NiagaraInt32}}}
      DebugDraw {
       bIsDynamic=false
       bIsEditable=false
       bIsStaticSwitch=false
       bIsVisible=false
       type {classStructOrEnum {refPath=&NiagaraDataInterfaceDebugDraw}}
      }
     ]
     moduleScript {refPath=&Collision}
    }
    {
     bIsSetParametersModule=false
     enabled=true
     moduleName=SolveForcesAndVelocity
     inputs [
      Clamp Velocity {bIsDynamic=false, bIsEditable=true, bIsStaticSwitch=false, bIsVisible=true, type {classStructOrEnum {refPath=&NiagaraBool}}}
      Speed Limit {bIsDynamic=false, bIsEditable=false, bIsStaticSwitch=false, bIsVisible=true, type {classStructOrEnum {refPath=&NiagaraFloat}}}
      Limit Acceleration {bIsDynamic=false, bIsEditable=true, bIsStaticSwitch=false, bIsVisible=true, type {classStructOrEnum {refPath=&NiagaraBool}}}
      Acceleration Limit {bIsDynamic=false, bIsEditable=false, bIsStaticSwitch=false, bIsVisible=true, type {classStructOrEnum {refPath=&NiagaraFloat}}}
      Manually Enable Rotational Solver {bIsDynamic=false, bIsEditable=true, bIsStaticSwitch=false, bIsVisible=true, type {classStructOrEnum {refPath=&NiagaraBool}}}
      Write to Intrinsic Properties {bIsDynamic=false, bIsEditable=true, bIsStaticSwitch=false, bIsVisible=true, type {classStructOrEnum {refPath=&NiagaraBool}}}
      Write to Presolve Properties {bIsDynamic=false, bIsEditable=true, bIsStaticSwitch=true, bIsVisible=true, type {classStructOrEnum {refPath=&NiagaraBool}}}
      DeltaTime {bIsDynamic=false, bIsEditable=true, bIsStaticSwitch=false, bIsVisible=true, type {classStructOrEnum {refPath=&NiagaraFloat}}}
      Mass {bIsDynamic=false, bIsEditable=true, bIsStaticSwitch=false, bIsVisible=true, type {classStructOrEnum {refPath=&NiagaraFloat}}}
      Position {
       bIsDynamic=false
       bIsEditable=true
       bIsStaticSwitch=false
       bIsVisible=true
       type {classStructOrEnum {refPath=&NiagaraPosition}}
      }
      Velocity {bIsDynamic=false, bIsEditable=true, bIsStaticSwitch=false, bIsVisible=true, type {classStructOrEnum {refPath=&Vector3f}}}
      PreviousVelocity {bIsDynamic=false, bIsEditable=true, bIsStaticSwitch=false, bIsVisible=true, type {classStructOrEnum {refPath=&Vector3f}}}
      Force {bIsDynamic=false, bIsEditable=true, bIsStaticSwitch=false, bIsVisible=true, type {classStructOrEnum {refPath=&Vector3f}}}
      Mesh Orientation {bIsDynamic=false, bIsEditable=true, bIsStaticSwitch=false, bIsVisible=true, type {classStructOrEnum {refPath=&Quat4f}}}
      Rotational Velocity {bIsDynamic=false, bIsEditable=true, bIsStaticSwitch=false, bIsVisible=true, type {classStructOrEnum {refPath=&Vector3f}}}
      Rotational Force {bIsDynamic=false, bIsEditable=true, bIsStaticSwitch=false, bIsVisible=true, type {classStructOrEnum {refPath=&Vector3f}}}
      Rotational Inertia {bIsDynamic=false, bIsEditable=true, bIsStaticSwitch=false, bIsVisible=true, type {classStructOrEnum {refPath=&NiagaraFloat}}}
     ]
     moduleScript {refPath=&SolveForcesAndVelocity}
    }
   ]
  }
  rendererClasses [{refPath=&NiagaraMeshRendererProperties}]
  renderers [{rendererIndex=0, rendererClass {refPath=&NiagaraMeshRendererProperties}}]
 }
 Dust {
  emitterSpawnScript {scriptName=EmitterSpawnScript, modules []}
  emitterUpdateScript {
   scriptName=EmitterUpdateScript
   modules [
    {
     bIsSetParametersModule=false
     enabled=true
     moduleName=EmitterState
     inputs [
      Life Cycle Mode {
       bIsDynamic=false
       bIsEditable=true
       bIsStaticSwitch=true
       bIsVisible=true
       type {classStructOrEnum {refPath=&ENiagaraEmitterLifeCycleMode}}
      }
      Inactive Response {
       bIsDynamic=false
       bIsEditable=true
       bIsStaticSwitch=true
       bIsVisible=true
       type {classStructOrEnum {refPath=&ENiagaraInactiveMode}}
      }
      Loop Behavior {
       bIsDynamic=false
       bIsEditable=true
       bIsStaticSwitch=true
       bIsVisible=true
       type {classStructOrEnum {refPath=&ENiagara_EmitterStateOptions}}
      }
      Loop Duration Mode {
       bIsDynamic=false
       bIsEditable=true
       bIsStaticSwitch=true
       bIsVisible=true
       type {classStructOrEnum {refPath=&ENiagara_InfiniteLoopDuration}}
      }
      Loop Count {
       bIsDynamic=false
       bIsEditable=false
       bIsStaticSwitch=false
       bIsVisible=false
       type {classStructOrEnum {refPath=&NiagaraInt32}}
      }
      Loop Duration {bIsDynamic=false, bIsEditable=true, bIsStaticSwitch=false, bIsVisible=true, type {classStructOrEnum {refPath=&NiagaraFloat}}}
      Recalculate Duration Each Loop {bIsDynamic=false, bIsEditable=false, bIsStaticSwitch=false, bIsVisible=false, type {classStructOrEnum {refPath=&NiagaraBool}}}
      UseLoopDelay {bIsDynamic=false, bIsEditable=true, bIsStaticSwitch=false, bIsVisible=true, type {classStructOrEnum {refPath=&NiagaraBool}}}
      Loop Delay {bIsDynamic=false, bIsEditable=false, bIsStaticSwitch=false, bIsVisible=true, type {classStructOrEnum {refPath=&NiagaraFloat}}}
      Delay First Loop Only {bIsDynamic=false, bIsEditable=false, bIsStaticSwitch=false, bIsVisible=false, type {classStructOrEnum {refPath=&NiagaraBool}}}
      Scalability Mode {
       bIsDynamic=false
       bIsEditable=true
       bIsStaticSwitch=true
       bIsVisible=true
       type {classStructOrEnum {refPath=&ENiagaraEmitterScalabilityMode_Limited}}
      }
      Enable Distance Culling {bIsDynamic=false, bIsEditable=false, bIsStaticSwitch=true, bIsVisible=false, type {classStructOrEnum {refPath=&NiagaraBool}}}
      MinDistance {
       bIsDynamic=false
       bIsEditable=false
       bIsStaticSwitch=false
       bIsVisible=false
       type {classStructOrEnum {refPath=&NiagaraFloat}}
      }
      Min Distance Response {
       bIsDynamic=false
       bIsEditable=false
       bIsStaticSwitch=false
       bIsVisible=false
       type {classStructOrEnum {refPath=&ENiagaraExecutionStateManagement}}
      }
      MaxDistance {
       bIsDynamic=false
       bIsEditable=false
       bIsStaticSwitch=false
       bIsVisible=false
       type {classStructOrEnum {refPath=&NiagaraFloat}}
      }
      Max Distance Response {
       bIsDynamic=false
       bIsEditable=false
       bIsStaticSwitch=false
       bIsVisible=false
       type {classStructOrEnum {refPath=&ENiagaraExecutionStateManagement}}
      }
      Scale Spawn Count {bIsDynamic=false, bIsEditable=false, bIsStaticSwitch=true, bIsVisible=false, type {classStructOrEnum {refPath=&NiagaraBool}}}
      Spawn Count Scale {
       bIsDynamic=false
       bIsEditable=false
       bIsStaticSwitch=false
       bIsVisible=false
       type {classStructOrEnum {refPath=&NiagaraFloat}}
      }
      Spawn Count Scale By Distance {
       bIsDynamic=false
       bIsEditable=false
       bIsStaticSwitch=false
       bIsVisible=false
       type {classStructOrEnum {refPath=&NiagaraDataInterfaceCurve}}
      }
      Enable Visibility Culling {bIsDynamic=false, bIsEditable=false, bIsStaticSwitch=true, bIsVisible=false, type {classStructOrEnum {refPath=&NiagaraBool}}}
      Visibility Cull Response {
       bIsDynamic=false
       bIsEditable=false
       bIsStaticSwitch=false
       bIsVisible=false
       type {classStructOrEnum {refPath=&ENiagaraExecutionStateManagement}}
      }
      Visibility Cull Delay {
       bIsDynamic=false
       bIsEditable=false
       bIsStaticSwitch=false
       bIsVisible=false
       type {classStructOrEnum {refPath=&NiagaraFloat}}
      }
      Reset Age On Awaken {bIsDynamic=false, bIsEditable=false, bIsStaticSwitch=false, bIsVisible=false, type {classStructOrEnum {refPath=&NiagaraBool}}}
     ]
     moduleScript {refPath=&EmitterState}
    }
    {
     bIsSetParametersModule=false
     enabled=true
     moduleName=SpawnBurst_Instantaneous
     inputs [
      Spawn Count {bIsDynamic=false, bIsEditable=true, bIsStaticSwitch=false, bIsVisible=true, type {classStructOrEnum {refPath=&NiagaraInt32}}}
      Spawn Time {bIsDynamic=false, bIsEditable=true, bIsStaticSwitch=false, bIsVisible=true, type {classStructOrEnum {refPath=&NiagaraFloat}}}
      Use Spawn Probability {bIsDynamic=false, bIsEditable=true, bIsStaticSwitch=false, bIsVisible=true, type {classStructOrEnum {refPath=&NiagaraBool}}}
      Spawn Probability {bIsDynamic=false, bIsEditable=false, bIsStaticSwitch=false, bIsVisible=true, type {classStructOrEnum {refPath=&NiagaraFloat}}}
      UseLoopCountLimit {bIsDynamic=false, bIsEditable=true, bIsStaticSwitch=false, bIsVisible=true, type {classStructOrEnum {refPath=&NiagaraBool}}}
      Loop Count Limit {bIsDynamic=false, bIsEditable=false, bIsStaticSwitch=false, bIsVisible=true, type {classStructOrEnum {refPath=&NiagaraInt32}}}
      Spawn Group {bIsDynamic=false, bIsEditable=true, bIsStaticSwitch=false, bIsVisible=true, type {classStructOrEnum {refPath=&NiagaraInt32}}}
      Age {bIsDynamic=false, bIsEditable=true, bIsStaticSwitch=false, bIsVisible=true, type {classStructOrEnum {refPath=&NiagaraFloat}}}
     ]
     moduleScript {refPath=&SpawnBurst_Instantaneous}
    }
   ]
  }
  particleSpawnScript {
   scriptName=ParticleSpawnScript
   modules [
    {
     bIsSetParametersModule=false
     enabled=true
     moduleName=InitializeParticle
     inputs [
      Lifetime Mode {
       bIsDynamic=false
       bIsEditable=true
       bIsStaticSwitch=true
       bIsVisible=true
       type {classStructOrEnum {refPath=&ENiagara_LifetimeMode}}
      }
      Lifetime Min {bIsDynamic=false, bIsEditable=true, bIsStaticSwitch=false, bIsVisible=true, type {classStructOrEnum {refPath=&NiagaraFloat}}}
      Lifetime {
       bIsDynamic=false
       bIsEditable=false
       bIsStaticSwitch=false
       bIsVisible=false
       type {classStructOrEnum {refPath=&NiagaraFloat}}
      }
      Lifetime Max {bIsDynamic=false, bIsEditable=true, bIsStaticSwitch=false, bIsVisible=true, type {classStructOrEnum {refPath=&NiagaraFloat}}}
      Lifetime Randomness Mode {
       bIsDynamic=false
       bIsEditable=true
       bIsStaticSwitch=true
       bIsVisible=true
       type {classStructOrEnum {refPath=&ENiagaraRandomnessMode}}
      }
      Color Mode {
       bIsDynamic=false
       bIsEditable=true
       bIsStaticSwitch=true
       bIsVisible=true
       type {classStructOrEnum {refPath=&ENiagara_ColorInitializationMode}}
      }
      Color Minimum {
       bIsDynamic=false
       bIsEditable=false
       bIsStaticSwitch=false
       bIsVisible=false
       type {classStructOrEnum {refPath=&LinearColor}}
      }
      Color {
       bIsDynamic=false
       bIsEditable=true
       bIsStaticSwitch=false
       bIsVisible=true
       type {classStructOrEnum {refPath=&LinearColor}}
      }
      AdjustHue {bIsDynamic=false, bIsEditable=false, bIsStaticSwitch=false, bIsVisible=false, type {classStructOrEnum {refPath=&NiagaraBool}}}
      Hue Shift Range {
       bIsDynamic=false
       bIsEditable=false
       bIsStaticSwitch=false
       bIsVisible=false
       type {classStructOrEnum {refPath=&Vector2f}}
      }
      Color Maximum {
       bIsDynamic=false
       bIsEditable=false
       bIsStaticSwitch=false
       bIsVisible=false
       type {classStructOrEnum {refPath=&LinearColor}}
      }
      AdjustSaturation {bIsDynamic=false, bIsEditable=false, bIsStaticSwitch=false, bIsVisible=false, type {classStructOrEnum {refPath=&NiagaraBool}}}
      Saturation Range {
       bIsDynamic=false
       bIsEditable=false
       bIsStaticSwitch=false
       bIsVisible=false
       type {classStructOrEnum {refPath=&Vector2f}}
      }
      AdjustValue {bIsDynamic=false, bIsEditable=false, bIsStaticSwitch=false, bIsVisible=false, type {classStructOrEnum {refPath=&NiagaraBool}}}
      Value Range {
       bIsDynamic=false
       bIsEditable=false
       bIsStaticSwitch=false
       bIsVisible=false
       type {classStructOrEnum {refPath=&Vector2f}}
      }
      AdjustAlpha {bIsDynamic=false, bIsEditable=false, bIsStaticSwitch=false, bIsVisible=false, type {classStructOrEnum {refPath=&NiagaraBool}}}
      Alpha Scale Range {
       bIsDynamic=false
       bIsEditable=false
       bIsStaticSwitch=false
       bIsVisible=false
       type {classStructOrEnum {refPath=&Vector2f}}
      }
      Color Channel Mode {
       bIsDynamic=false
       bIsEditable=false
       bIsStaticSwitch=true
       bIsVisible=false
       type {classStructOrEnum {refPath=&ENiagaraChannelCorrelation}}
      }
      Color Randomness Mode {
       bIsDynamic=false
       bIsEditable=false
       bIsStaticSwitch=true
       bIsVisible=false
       type {classStructOrEnum {refPath=&ENiagaraRandomnessMode}}
      }
      Position Mode {
       bIsDynamic=false
       bIsEditable=true
       bIsStaticSwitch=true
       bIsVisible=true
       type {classStructOrEnum {refPath=&ENiagara_PositionInitializationMode}}
      }
      Position {
       bIsDynamic=false
       bIsEditable=false
       bIsStaticSwitch=false
       bIsVisible=false
       type {classStructOrEnum {refPath=&NiagaraPosition}}
      }
      UsePositionOffset {bIsDynamic=false, bIsEditable=true, bIsStaticSwitch=false, bIsVisible=true, type {classStructOrEnum {refPath=&NiagaraBool}}}
      Position Offset {bIsDynamic=false, bIsEditable=false, bIsStaticSwitch=false, bIsVisible=true, type {classStructOrEnum {refPath=&Vector3f}}}
      Position Offset Coordinate Space {
       bIsDynamic=false
       bIsEditable=false
       bIsStaticSwitch=false
       bIsVisible=false
       type {classStructOrEnum {refPath=&ENiagaraCoordinateSpace}}
      }
      Mass Mode {
       bIsDynamic=false
       bIsEditable=true
       bIsStaticSwitch=true
       bIsVisible=true
       type {classStructOrEnum {refPath=&ENiagara_MassInitializationMode}}
      }
      Mass Min {
       bIsDynamic=false
       bIsEditable=false
       bIsStaticSwitch=false
       bIsVisible=false
       type {classStructOrEnum {refPath=&NiagaraFloat}}
      }
      Mass {
       bIsDynamic=false
       bIsEditable=false
       bIsStaticSwitch=false
       bIsVisible=false
       type {classStructOrEnum {refPath=&NiagaraFloat}}
      }
      Mass Max {
       bIsDynamic=false
       bIsEditable=false
       bIsStaticSwitch=false
       bIsVisible=false
       type {classStructOrEnum {refPath=&NiagaraFloat}}
      }
      Mass Randomness Mode {
       bIsDynamic=false
       bIsEditable=false
       bIsStaticSwitch=true
       bIsVisible=false
       type {classStructOrEnum {refPath=&ENiagaraRandomnessMode}}
      }
      Write Material Random {bIsDynamic=false, bIsEditable=true, bIsStaticSwitch=false, bIsVisible=true, type {classStructOrEnum {refPath=&NiagaraBool}}}
      Material Random {bIsDynamic=false, bIsEditable=true, bIsStaticSwitch=false, bIsVisible=true, type {classStructOrEnum {refPath=&NiagaraFloat}}}
      Sprite Size Mode {
       bIsDynamic=false
       bIsEditable=true
       bIsStaticSwitch=true
       bIsVisible=true
       type {classStructOrEnum {refPath=&ENiagara_SizeScaleMode}}
      }
      Sprite Size {
       bIsDynamic=false
       bIsEditable=false
       bIsStaticSwitch=false
       bIsVisible=false
       type {classStructOrEnum {refPath=&Vector2f}}
      }
      Uniform Sprite Size {bIsDynamic=true, bIsEditable=false, bIsStaticSwitch=false, bIsVisible=false, type {classStructOrEnum {refPath=&NiagaraFloat}}}
      Uniform Sprite Size Min {bIsDynamic=false, bIsEditable=true, bIsStaticSwitch=false, bIsVisible=true, type {classStructOrEnum {refPath=&NiagaraFloat}}}
      Sprite Size Min {
       bIsDynamic=false
       bIsEditable=false
       bIsStaticSwitch=false
       bIsVisible=false
       type {classStructOrEnum {refPath=&Vector2f}}
      }
      Sprite Size Max {
       bIsDynamic=false
       bIsEditable=false
       bIsStaticSwitch=false
       bIsVisible=false
       type {classStructOrEnum {refPath=&Vector2f}}
      }
      Uniform Sprite Size Max {bIsDynamic=false, bIsEditable=true, bIsStaticSwitch=false, bIsVisible=true, type {classStructOrEnum {refPath=&NiagaraFloat}}}
      Sprite Size Randomness Mode {
       bIsDynamic=false
       bIsEditable=true
       bIsStaticSwitch=true
       bIsVisible=true
       type {classStructOrEnum {refPath=&ENiagaraRandomnessMode}}
      }
      Sprite Rotation Mode {
       bIsDynamic=false
       bIsEditable=true
       bIsStaticSwitch=true
       bIsVisible=true
       type {classStructOrEnum {refPath=&ENiagara_SpriteRotationMode}}
      }
      Sprite Rotation Angle Min {bIsDynamic=false, bIsEditable=true, bIsStaticSwitch=false, bIsVisible=true, type {classStructOrEnum {refPath=&NiagaraFloat}}}
      Sprite Rotation Angle {
       bIsDynamic=false
       bIsEditable=false
       bIsStaticSwitch=false
       bIsVisible=false
       type {classStructOrEnum {refPath=&NiagaraFloat}}
      }
      Sprite Rotation Angle Max {bIsDynamic=false, bIsEditable=true, bIsStaticSwitch=false, bIsVisible=true, type {classStructOrEnum {refPath=&NiagaraFloat}}}
      Sprite Rotation Randomness Mode {
       bIsDynamic=false
       bIsEditable=true
       bIsStaticSwitch=true
       bIsVisible=true
       type {classStructOrEnum {refPath=&ENiagaraRandomnessMode}}
      }
      Sprite UV Mode {
       bIsDynamic=false
       bIsEditable=true
       bIsStaticSwitch=true
       bIsVisible=true
       type {classStructOrEnum {refPath=&ENiagara_UVFlippingMode}}
      }
      Sprite UV Randomness Mode {
       bIsDynamic=false
       bIsEditable=true
       bIsStaticSwitch=true
       bIsVisible=true
       type {classStructOrEnum {refPath=&ENiagaraRandomnessMode}}
      }
      Sprite UV Scale {
       bIsDynamic=false
       bIsEditable=false
       bIsStaticSwitch=false
       bIsVisible=false
       type {classStructOrEnum {refPath=&Vector2f}}
      }
      Mesh Scale Mode {
       bIsDynamic=false
       bIsEditable=true
       bIsStaticSwitch=true
       bIsVisible=true
       type {classStructOrEnum {refPath=&ENiagara_SizeScaleMode}}
      }
      Mesh Uniform Scale Min {
       bIsDynamic=false
       bIsEditable=false
       bIsStaticSwitch=false
       bIsVisible=false
       type {classStructOrEnum {refPath=&NiagaraFloat}}
      }
      Mesh Uniform Scale {
       bIsDynamic=false
       bIsEditable=false
       bIsStaticSwitch=false
       bIsVisible=false
       type {classStructOrEnum {refPath=&NiagaraFloat}}
      }
      Mesh Scale Min {
       bIsDynamic=false
       bIsEditable=false
       bIsStaticSwitch=false
       bIsVisible=false
       type {classStructOrEnum {refPath=&Vector3f}}
      }
      Mesh Scale {
       bIsDynamic=false
       bIsEditable=false
       bIsStaticSwitch=false
       bIsVisible=false
       type {classStructOrEnum {refPath=&Vector3f}}
      }
      Mesh Scale Max {
       bIsDynamic=false
       bIsEditable=false
       bIsStaticSwitch=false
       bIsVisible=false
       type {classStructOrEnum {refPath=&Vector3f}}
      }
      Mesh Uniform Scale Max {
       bIsDynamic=false
       bIsEditable=false
       bIsStaticSwitch=false
       bIsVisible=false
       type {classStructOrEnum {refPath=&NiagaraFloat}}
      }
      Mesh Scale Randomness Mode {
       bIsDynamic=false
       bIsEditable=false
       bIsStaticSwitch=true
       bIsVisible=false
       type {classStructOrEnum {refPath=&ENiagaraRandomnessMode}}
      }
      Mesh Renderer Array Visibility Mode {
       bIsDynamic=false
       bIsEditable=true
       bIsStaticSwitch=true
       bIsVisible=true
       type {classStructOrEnum {refPath=&ENiagara_UnsetDirectSetRandom}}
      }
      Mesh Renderer Mesh Index {
       bIsDynamic=false
       bIsEditable=false
       bIsStaticSwitch=false
       bIsVisible=false
       type {classStructOrEnum {refPath=&NiagaraInt32}}
      }
      Wrap Index By Mesh Count {bIsDynamic=false, bIsEditable=false, bIsStaticSwitch=true, bIsVisible=false, type {classStructOrEnum {refPath=&NiagaraBool}}}
      Mesh Renderer Info {
       bIsDynamic=false
       bIsEditable=false
       bIsStaticSwitch=false
       bIsVisible=false
       type {classStructOrEnum {refPath=&NiagaraDataInterfaceMeshRendererInfo}}
      }
      Mesh Renderer Random Seed {
       bIsDynamic=false
       bIsEditable=false
       bIsStaticSwitch=false
       bIsVisible=false
       type {classStructOrEnum {refPath=&NiagaraInt32}}
      }
      Ribbon Width Mode {
       bIsDynamic=false
       bIsEditable=true
       bIsStaticSwitch=true
       bIsVisible=true
       type {classStructOrEnum {refPath=&ENiagara_UnsetDirectSet}}
      }
      Ribbon Width {
       bIsDynamic=false
       bIsEditable=false
       bIsStaticSwitch=false
       bIsVisible=false
       type {classStructOrEnum {refPath=&NiagaraFloat}}
      }
      Ribbon Facing Mode {
       bIsDynamic=false
       bIsEditable=true
       bIsStaticSwitch=true
       bIsVisible=true
       type {classStructOrEnum {refPath=&ENiagara_UnsetDirectSet}}
      }
      Ribbon Facing Vector {
       bIsDynamic=false
       bIsEditable=false
       bIsStaticSwitch=false
       bIsVisible=false
       type {classStructOrEnum {refPath=&Vector3f}}
      }
      Ribbon Twist Mode {
       bIsDynamic=false
       bIsEditable=true
       bIsStaticSwitch=true
       bIsVisible=true
       type {classStructOrEnum {refPath=&ENiagara_UnsetDirectSet}}
      }
      "Ribbon Twist " {
       bIsDynamic=false
       bIsEditable=false
       bIsStaticSwitch=false
       bIsVisible=false
       type {classStructOrEnum {refPath=&NiagaraFloat}}
      }
     ]
     moduleScript {refPath=&InitializeParticle}
    }
    {
     bIsSetParametersModule=false
     enabled=true
     moduleName=ShapeLocation
     inputs [
      Shape Primitive {
       bIsDynamic=false
       bIsEditable=true
       bIsStaticSwitch=true
       bIsVisible=true
       type {classStructOrEnum {refPath=&ENiagara_LocationShapes}}
      }
      Box / Plane Mode {
       bIsDynamic=false
       bIsEditable=false
       bIsStaticSwitch=true
       bIsVisible=false
       type {classStructOrEnum {refPath=&ENiagara_BoxPlaneMode}}
      }
      Cone Mode {
       bIsDynamic=false
       bIsEditable=false
       bIsStaticSwitch=true
       bIsVisible=false
       type {classStructOrEnum {refPath=&ENiagara_ConeMode}}
      }
      Cylinder Height {
       bIsDynamic=false
       bIsEditable=false
       bIsStaticSwitch=false
       bIsVisible=false
       type {classStructOrEnum {refPath=&NiagaraFloat}}
      }
      Sphere Radius {
       bIsDynamic=false
       bIsEditable=false
       bIsStaticSwitch=false
       bIsVisible=false
       type {classStructOrEnum {refPath=&NiagaraFloat}}
      }
      Ring / Disc Mode {
       bIsDynamic=false
       bIsEditable=true
       bIsStaticSwitch=true
       bIsVisible=true
       type {classStructOrEnum {refPath=&ENiagaraRingDiscMode}}
      }
      Torus Mode {
       bIsDynamic=false
       bIsEditable=false
       bIsStaticSwitch=true
       bIsVisible=false
       type {classStructOrEnum {refPath=&ENiagaraShapeTorusMode}}
      }
      Knot Ring Loops {
       bIsDynamic=false
       bIsEditable=false
       bIsStaticSwitch=false
       bIsVisible=false
       type {classStructOrEnum {refPath=&NiagaraInt32}}
      }
      Knot Helix Loops {
       bIsDynamic=false
       bIsEditable=false
       bIsStaticSwitch=false
       bIsVisible=false
       type {classStructOrEnum {refPath=&NiagaraInt32}}
      }
      Enable Torus Knot Radius {bIsDynamic=false, bIsEditable=false, bIsStaticSwitch=false, bIsVisible=false, type {classStructOrEnum {refPath=&NiagaraBool}}}
      Torus Knot Tube Radius {
       bIsDynamic=false
       bIsEditable=false
       bIsStaticSwitch=false
       bIsVisible=false
       type {classStructOrEnum {refPath=&NiagaraFloat}}
      }
      Cylinder Radius {
       bIsDynamic=false
       bIsEditable=false
       bIsStaticSwitch=false
       bIsVisible=false
       type {classStructOrEnum {refPath=&NiagaraFloat}}
      }
      Cylinder Height Midpoint {
       bIsDynamic=false
       bIsEditable=false
       bIsStaticSwitch=false
       bIsVisible=false
       type {classStructOrEnum {refPath=&NiagaraFloat}}
      }
      Ring Radius {bIsDynamic=false, bIsEditable=true, bIsStaticSwitch=false, bIsVisible=true, type {classStructOrEnum {refPath=&NiagaraFloat}}}
      Box Size {
       bIsDynamic=false
       bIsEditable=false
       bIsStaticSwitch=false
       bIsVisible=false
       type {classStructOrEnum {refPath=&Vector3f}}
      }
      Box Midpoint {
       bIsDynamic=false
       bIsEditable=false
       bIsStaticSwitch=false
       bIsVisible=false
       type {classStructOrEnum {refPath=&Vector3f}}
      }
      Cone Length {
       bIsDynamic=false
       bIsEditable=false
       bIsStaticSwitch=false
       bIsVisible=false
       type {classStructOrEnum {refPath=&NiagaraFloat}}
      }
      Plane Size {
       bIsDynamic=false
       bIsEditable=false
       bIsStaticSwitch=false
       bIsVisible=false
       type {classStructOrEnum {refPath=&Vector2f}}
      }
      Wedge Horizontal Angle {
       bIsDynamic=false
       bIsEditable=false
       bIsStaticSwitch=false
       bIsVisible=false
       type {classStructOrEnum {refPath=&NiagaraFloat}}
      }
      Large Radius {
       bIsDynamic=false
       bIsEditable=false
       bIsStaticSwitch=false
       bIsVisible=false
       type {classStructOrEnum {refPath=&NiagaraFloat}}
      }
      Plane Midpoint {
       bIsDynamic=false
       bIsEditable=false
       bIsStaticSwitch=false
       bIsVisible=false
       type {classStructOrEnum {refPath=&Vector2f}}
      }
      Wedge Vertical Angle {
       bIsDynamic=false
       bIsEditable=false
       bIsStaticSwitch=false
       bIsVisible=false
       type {classStructOrEnum {refPath=&NiagaraFloat}}
      }
      Cone Angle {
       bIsDynamic=false
       bIsEditable=false
       bIsStaticSwitch=false
       bIsVisible=false
       type {classStructOrEnum {refPath=&NiagaraFloat}}
      }
      Handle Radius {
       bIsDynamic=false
       bIsEditable=false
       bIsStaticSwitch=false
       bIsVisible=false
       type {classStructOrEnum {refPath=&NiagaraFloat}}
      }
      Cone Inner Angle {
       bIsDynamic=false
       bIsEditable=false
       bIsStaticSwitch=false
       bIsVisible=false
       type {classStructOrEnum {refPath=&NiagaraFloat}}
      }
      Cone Radial Angle {
       bIsDynamic=false
       bIsEditable=false
       bIsStaticSwitch=false
       bIsVisible=false
       type {classStructOrEnum {refPath=&NiagaraFloat}}
      }
      Cone Angle Type {
       bIsDynamic=false
       bIsEditable=false
       bIsStaticSwitch=true
       bIsVisible=false
       type {classStructOrEnum {refPath=&ENiagara_AngleInput}}
      }
      Flatten Endcaps {bIsDynamic=false, bIsEditable=false, bIsStaticSwitch=true, bIsVisible=false, type {classStructOrEnum {refPath=&NiagaraBool}}}
      Apply To Particle Position {
       bIsDynamic=false
       bIsEditable=true
       bIsStaticSwitch=true
       bIsVisible=true
       type {classStructOrEnum {refPath=&ENiagara_AttributeSamplingApplyOutput}}
      }
      Cone Surface Distribution {
       bIsDynamic=false
       bIsEditable=false
       bIsStaticSwitch=false
       bIsVisible=false
       type {classStructOrEnum {refPath=&NiagaraFloat}}
      }
      Ring / Disc Distribution Mode {
       bIsDynamic=false
       bIsEditable=true
       bIsStaticSwitch=true
       bIsVisible=true
       type {classStructOrEnum {refPath=&ENiagaraSphereDistributionMode}}
      }
      Sphere Distribution {
       bIsDynamic=false
       bIsEditable=false
       bIsStaticSwitch=true
       bIsVisible=false
       type {classStructOrEnum {refPath=&ENiagaraSphereDistributionMode}}
      }
      Torus Distribution Mode {
       bIsDynamic=false
       bIsEditable=false
       bIsStaticSwitch=true
       bIsVisible=false
       type {classStructOrEnum {refPath=&ENiagaraTorusDistributionMode}}
      }
      Sphere Surface Distribution {
       bIsDynamic=false
       bIsEditable=false
       bIsStaticSwitch=false
       bIsVisible=false
       type {classStructOrEnum {refPath=&NiagaraFloat}}
      }
      Hemisphere Distribution {
       bIsDynamic=false
       bIsEditable=false
       bIsStaticSwitch=false
       bIsVisible=false
       type {classStructOrEnum {refPath=&Vector2f}}
      }
      Cylinder Distribution {
       bIsDynamic=false
       bIsEditable=false
       bIsStaticSwitch=true
       bIsVisible=false
       type {classStructOrEnum {refPath=&ENiagara_CylinderMode}}
      }
      Surface Only {bIsDynamic=false, bIsEditable=false, bIsStaticSwitch=false, bIsVisible=false, type {classStructOrEnum {refPath=&NiagaraBool}}}
      Surface Only Band Thickness {
       bIsDynamic=false
       bIsEditable=false
       bIsStaticSwitch=false
       bIsVisible=false
       type {classStructOrEnum {refPath=&NiagaraFloat}}
      }
      Use Endcaps In Surface Only Mode {bIsDynamic=false, bIsEditable=false, bIsStaticSwitch=true, bIsVisible=false, type {classStructOrEnum {refPath=&NiagaraBool}}}
      Hemicircle X {bIsDynamic=false, bIsEditable=false, bIsStaticSwitch=true, bIsVisible=false, type {classStructOrEnum {refPath=&NiagaraBool}}}
      Hemicircle Y {bIsDynamic=false, bIsEditable=false, bIsStaticSwitch=true, bIsVisible=false, type {classStructOrEnum {refPath=&NiagaraBool}}}
      Disc Coverage {bIsDynamic=false, bIsEditable=true, bIsStaticSwitch=false, bIsVisible=true, type {classStructOrEnum {refPath=&NiagaraFloat}}}
      Hemicircle Internal Cap {bIsDynamic=false, bIsEditable=false, bIsStaticSwitch=true, bIsVisible=false, type {classStructOrEnum {refPath=&NiagaraBool}}}
      U Position {bIsDynamic=true, bIsEditable=false, bIsStaticSwitch=false, bIsVisible=false, type {classStructOrEnum {refPath=&NiagaraFloat}}}
      Enable Lathe Profile {bIsDynamic=false, bIsEditable=false, bIsStaticSwitch=true, bIsVisible=false, type {classStructOrEnum {refPath=&NiagaraBool}}}
      Lathe Profile {
       bIsDynamic=false
       bIsEditable=false
       bIsStaticSwitch=false
       bIsVisible=false
       type {classStructOrEnum {refPath=&NiagaraDataInterfaceCurve}}
      }
      Surface Distribution {
       bIsDynamic=false
       bIsEditable=false
       bIsStaticSwitch=false
       bIsVisible=false
       type {classStructOrEnum {refPath=&NiagaraFloat}}
      }
      Surface Expansion Mode {
       bIsDynamic=false
       bIsEditable=false
       bIsStaticSwitch=false
       bIsVisible=false
       type {classStructOrEnum {refPath=&ENiagaraExpansionMode}}
      }
      U Distribution {bIsDynamic=false, bIsEditable=true, bIsStaticSwitch=false, bIsVisible=true, type {classStructOrEnum {refPath=&NiagaraFloat}}}
      V Position {
       bIsDynamic=false
       bIsEditable=false
       bIsStaticSwitch=false
       bIsVisible=false
       type {classStructOrEnum {refPath=&NiagaraFloat}}
      }
      Radius Position {
       bIsDynamic=false
       bIsEditable=false
       bIsStaticSwitch=false
       bIsVisible=false
       type {classStructOrEnum {refPath=&NiagaraFloat}}
      }
      Sphere Radius Position {
       bIsDynamic=false
       bIsEditable=false
       bIsStaticSwitch=false
       bIsVisible=false
       type {classStructOrEnum {refPath=&NiagaraFloat}}
      }
      V Distribution {
       bIsDynamic=false
       bIsEditable=false
       bIsStaticSwitch=false
       bIsVisible=false
       type {classStructOrEnum {refPath=&NiagaraFloat}}
      }
      Evaluation Type {
       bIsDynamic=false
       bIsEditable=false
       bIsStaticSwitch=true
       bIsVisible=false
       type {classStructOrEnum {refPath=&ENiagaraRandomnessEvaluation}}
      }
      Uniform Distribution {
       bIsDynamic=false
       bIsEditable=false
       bIsStaticSwitch=false
       bIsVisible=false
       type {classStructOrEnum {refPath=&NiagaraFloat}}
      }
      Uniform Spiral Amount {
       bIsDynamic=false
       bIsEditable=false
       bIsStaticSwitch=false
       bIsVisible=false
       type {classStructOrEnum {refPath=&NiagaraFloat}}
      }
      Uniform Spiral Falloff {
       bIsDynamic=false
       bIsEditable=false
       bIsStaticSwitch=false
       bIsVisible=false
       type {classStructOrEnum {refPath=&NiagaraFloat}}
      }
      Uniform Count {
       bIsDynamic=false
       bIsEditable=false
       bIsStaticSwitch=false
       bIsVisible=false
       type {classStructOrEnum {refPath=&NiagaraInt32}}
      }
      Uniform Particle Index {
       bIsDynamic=false
       bIsEditable=false
       bIsStaticSwitch=false
       bIsVisible=false
       type {classStructOrEnum {refPath=&NiagaraInt32}}
      }
      Hemisphere Angle Type {
       bIsDynamic=false
       bIsEditable=false
       bIsStaticSwitch=true
       bIsVisible=false
       type {classStructOrEnum {refPath=&ENiagara_AngleInput}}
      }
      Uniform Endpoint Offset {
       bIsDynamic=false
       bIsEditable=false
       bIsStaticSwitch=false
       bIsVisible=false
       type {classStructOrEnum {refPath=&NiagaraFloat}}
      }
      MaskSpawnGroup {bIsDynamic=false, bIsEditable=true, bIsStaticSwitch=false, bIsVisible=true, type {classStructOrEnum {refPath=&NiagaraBool}}}
      Spawn Group Mask {bIsDynamic=false, bIsEditable=false, bIsStaticSwitch=false, bIsVisible=true, type {classStructOrEnum {refPath=&NiagaraInt32}}}
      Spawn Group {
       bIsDynamic=false
       bIsEditable=false
       bIsStaticSwitch=false
       bIsVisible=false
       type {classStructOrEnum {refPath=&NiagaraInt32}}
      }
      Randomness Mode {
       bIsDynamic=false
       bIsEditable=true
       bIsStaticSwitch=true
       bIsVisible=true
       type {classStructOrEnum {refPath=&ENiagaraRandomnessMode}}
      }
      Override Seed {bIsDynamic=false, bIsEditable=true, bIsStaticSwitch=false, bIsVisible=true, type {classStructOrEnum {refPath=&NiagaraBool}}}
      Random Seed {bIsDynamic=false, bIsEditable=false, bIsStaticSwitch=false, bIsVisible=true, type {classStructOrEnum {refPath=&NiagaraInt32}}}
      Fixed Random Seed {bIsDynamic=false, bIsEditable=false, bIsStaticSwitch=true, bIsVisible=false, type {classStructOrEnum {refPath=&NiagaraBool}}}
      Transform Order {
       bIsDynamic=false
       bIsEditable=true
       bIsStaticSwitch=true
       bIsVisible=true
       type {classStructOrEnum {refPath=&ENiagara_TransformOrder}}
      }
      Scale Mode {
       bIsDynamic=false
       bIsEditable=true
       bIsStaticSwitch=true
       bIsVisible=true
       type {classStructOrEnum {refPath=&ENiagara_ScaleMode}}
      }
      Non Uniform Scale {bIsDynamic=false, bIsEditable=true, bIsStaticSwitch=false, bIsVisible=true, type {classStructOrEnum {refPath=&Vector3f}}}
      UseOwnerScale {bIsDynamic=false, bIsEditable=true, bIsStaticSwitch=false, bIsVisible=true, type {classStructOrEnum {refPath=&NiagaraBool}}}
      Apply Owner Scale {bIsDynamic=false, bIsEditable=true, bIsStaticSwitch=false, bIsVisible=true, type {classStructOrEnum {refPath=&Vector3f}}}
      Rotation Mode {
       bIsDynamic=false
       bIsEditable=true
       bIsStaticSwitch=true
       bIsVisible=true
       type {classStructOrEnum {refPath=&ENiagara_RotationMode}}
      }
      Rotation Axis {
       bIsDynamic=false
       bIsEditable=false
       bIsStaticSwitch=false
       bIsVisible=false
       type {classStructOrEnum {refPath=&Vector3f}}
      }
      Rotation Matrix {
       bIsDynamic=false
       bIsEditable=false
       bIsStaticSwitch=false
       bIsVisible=false
       type {classStructOrEnum {refPath=&NiagaraMatrix}}
      }
      Rotation Quaternion {bIsDynamic=false, bIsEditable=false, bIsStaticSwitch=false, bIsVisible=false, type {classStructOrEnum {refPath=&Quat4f}}}
      Yaw / Pitch / Roll {
       bIsDynamic=false
       bIsEditable=false
       bIsStaticSwitch=false
       bIsVisible=false
       type {classStructOrEnum {refPath=&Vector3f}}
      }
      Rotation Angle {
       bIsDynamic=false
       bIsEditable=false
       bIsStaticSwitch=false
       bIsVisible=false
       type {classStructOrEnum {refPath=&NiagaraFloat}}
      }
      Invert Rotation Quaternion {bIsDynamic=false, bIsEditable=false, bIsStaticSwitch=false, bIsVisible=false, type {classStructOrEnum {refPath=&NiagaraBool}}}
      Additional Quaternion Rotation {bIsDynamic=false, bIsEditable=false, bIsStaticSwitch=false, bIsVisible=false, type {classStructOrEnum {refPath=&NiagaraBool}}}
      Additional Yaw / Pitch / Roll {
       bIsDynamic=false
       bIsEditable=false
       bIsStaticSwitch=false
       bIsVisible=false
       type {classStructOrEnum {refPath=&Vector3f}}
      }
      Rotation Angle Type {
       bIsDynamic=false
       bIsEditable=false
       bIsStaticSwitch=true
       bIsVisible=false
       type {classStructOrEnum {refPath=&ENiagara_AngleInput}}
      }
      Rotation Coordinate Space {
       bIsDynamic=false
       bIsEditable=true
       bIsStaticSwitch=false
       bIsVisible=true
       type {classStructOrEnum {refPath=&ENiagaraCoordinateSpace}}
      }
      Offset Mode {
       bIsDynamic=false
       bIsEditable=true
       bIsStaticSwitch=true
       bIsVisible=true
       type {classStructOrEnum {refPath=&ENiagara_OffsetMode}}
      }
      Shape Origin {
       bIsDynamic=false
       bIsEditable=true
       bIsStaticSwitch=false
       bIsVisible=true
       type {classStructOrEnum {refPath=&NiagaraPosition}}
      }
      Offset {
       bIsDynamic=false
       bIsEditable=false
       bIsStaticSwitch=false
       bIsVisible=false
       type {classStructOrEnum {refPath=&Vector3f}}
      }
      Offset Coordinate Space {
       bIsDynamic=false
       bIsEditable=false
       bIsStaticSwitch=false
       bIsVisible=false
       type {classStructOrEnum {refPath=&ENiagaraCoordinateSpace}}
      }
      Custom Transform Matrix {
       bIsDynamic=false
       bIsEditable=false
       bIsStaticSwitch=false
       bIsVisible=false
       type {classStructOrEnum {refPath=&NiagaraMatrix}}
      }
      Transform Method {
       bIsDynamic=false
       bIsEditable=true
       bIsStaticSwitch=true
       bIsVisible=true
       type {classStructOrEnum {refPath=&ENiagara_TransformType}}
      }
     ]
     moduleScript {refPath=&ShapeLocation}
    }
    {
     bIsSetParametersModule=false
     enabled=true
     moduleName=AddVelocity
     inputs [
      Velocity Mode {
       bIsDynamic=false
       bIsEditable=true
       bIsStaticSwitch=true
       bIsVisible=true
       type {classStructOrEnum {refPath=&ENiagara_VelocityMode}}
      }
      Velocity {bIsDynamic=true, bIsEditable=false, bIsStaticSwitch=false, bIsVisible=false, type {classStructOrEnum {refPath=&Vector3f}}}
      Velocity Speed {bIsDynamic=true, bIsEditable=true, bIsStaticSwitch=false, bIsVisible=true, type {classStructOrEnum {refPath=&NiagaraFloat}}}
      Velocity Speed Scale {bIsDynamic=true, bIsEditable=false, bIsStaticSwitch=false, bIsVisible=false, type {classStructOrEnum {refPath=&NiagaraFloat}}}
      Module.Use Falloff Distance {bIsDynamic=false, bIsEditable=true, bIsStaticSwitch=false, bIsVisible=true, type {classStructOrEnum {refPath=&NiagaraBool}}}
      Constrain To Radius {bIsDynamic=false, bIsEditable=false, bIsStaticSwitch=false, bIsVisible=true, type {classStructOrEnum {refPath=&NiagaraFloat}}}
      Distribution Along Cone Axis {
       bIsDynamic=false
       bIsEditable=false
       bIsStaticSwitch=false
       bIsVisible=false
       type {classStructOrEnum {refPath=&NiagaraFloat}}
      }
      Radius Falloff Near / Far {
       bIsDynamic=false
       bIsEditable=false
       bIsStaticSwitch=false
       bIsVisible=false
       type {classStructOrEnum {refPath=&Vector2f}}
      }
      Radius Falloff Exponent {
       bIsDynamic=false
       bIsEditable=false
       bIsStaticSwitch=false
       bIsVisible=false
       type {classStructOrEnum {refPath=&NiagaraFloat}}
      }
      Speed Falloff From Cone Axis {
       bIsDynamic=false
       bIsEditable=false
       bIsStaticSwitch=false
       bIsVisible=false
       type {classStructOrEnum {refPath=&NiagaraFloat}}
      }
      Invert Velocity Falloff {bIsDynamic=false, bIsEditable=false, bIsStaticSwitch=false, bIsVisible=false, type {classStructOrEnum {refPath=&NiagaraBool}}}
      Override Seed {bIsDynamic=false, bIsEditable=true, bIsStaticSwitch=false, bIsVisible=true, type {classStructOrEnum {refPath=&NiagaraBool}}}
      Random Seed {bIsDynamic=false, bIsEditable=false, bIsStaticSwitch=false, bIsVisible=true, type {classStructOrEnum {refPath=&NiagaraInt32}}}
      Fixed Random Seed {bIsDynamic=false, bIsEditable=false, bIsStaticSwitch=true, bIsVisible=false, type {classStructOrEnum {refPath=&NiagaraBool}}}
      Randomness Mode {
       bIsDynamic=false
       bIsEditable=true
       bIsStaticSwitch=true
       bIsVisible=true
       type {classStructOrEnum {refPath=&ENiagaraRandomnessMode}}
      }
      Cone Axis {
       bIsDynamic=false
       bIsEditable=false
       bIsStaticSwitch=false
       bIsVisible=false
       type {classStructOrEnum {refPath=&Vector3f}}
      }
      Cone Angle {
       bIsDynamic=false
       bIsEditable=false
       bIsStaticSwitch=false
       bIsVisible=false
       type {classStructOrEnum {refPath=&NiagaraFloat}}
      }
      Inner Cone Angle {
       bIsDynamic=false
       bIsEditable=false
       bIsStaticSwitch=false
       bIsVisible=false
       type {classStructOrEnum {refPath=&NiagaraFloat}}
      }
      Cone Angle Mode {
       bIsDynamic=false
       bIsEditable=false
       bIsStaticSwitch=true
       bIsVisible=false
       type {classStructOrEnum {refPath=&ENiagara_AngleInput}}
      }
      Rotation Mode {
       bIsDynamic=false
       bIsEditable=false
       bIsStaticSwitch=true
       bIsVisible=false
       type {classStructOrEnum {refPath=&ENiagara_RotationMode}}
      }
      Rotation Axis {
       bIsDynamic=false
       bIsEditable=false
       bIsStaticSwitch=false
       bIsVisible=false
       type {classStructOrEnum {refPath=&Vector3f}}
      }
      Yaw / Pitch / Roll {
       bIsDynamic=false
       bIsEditable=false
       bIsStaticSwitch=false
       bIsVisible=false
       type {classStructOrEnum {refPath=&Vector3f}}
      }
      Rotation Quaternion {bIsDynamic=false, bIsEditable=false, bIsStaticSwitch=false, bIsVisible=false, type {classStructOrEnum {refPath=&Quat4f}}}
      Rotation Matrix {
       bIsDynamic=false
       bIsEditable=false
       bIsStaticSwitch=false
       bIsVisible=false
       type {classStructOrEnum {refPath=&NiagaraMatrix}}
      }
      Rotation Angle {
       bIsDynamic=false
       bIsEditable=false
       bIsStaticSwitch=false
       bIsVisible=false
       type {classStructOrEnum {refPath=&NiagaraFloat}}
      }
      Invert Rotation Quaternion {bIsDynamic=false, bIsEditable=false, bIsStaticSwitch=false, bIsVisible=false, type {classStructOrEnum {refPath=&NiagaraBool}}}
      Additional Quaternion Rotation {bIsDynamic=false, bIsEditable=false, bIsStaticSwitch=false, bIsVisible=false, type {classStructOrEnum {refPath=&NiagaraBool}}}
      Additional Yaw / Pitch / Roll {
       bIsDynamic=false
       bIsEditable=false
       bIsStaticSwitch=false
       bIsVisible=false
       type {classStructOrEnum {refPath=&Vector3f}}
      }
      Rotation Angle Type {
       bIsDynamic=false
       bIsEditable=false
       bIsStaticSwitch=true
       bIsVisible=false
       type {classStructOrEnum {refPath=&ENiagara_AngleInput}}
      }
      Rotation Coordinate Space {
       bIsDynamic=false
       bIsEditable=false
       bIsStaticSwitch=false
       bIsVisible=false
       type {classStructOrEnum {refPath=&ENiagaraCoordinateSpace}}
      }
      Velocity Origin {
       bIsDynamic=false
       bIsEditable=true
       bIsStaticSwitch=false
       bIsVisible=true
       type {classStructOrEnum {refPath=&NiagaraPosition}}
      }
      Origin Offset {bIsDynamic=false, bIsEditable=true, bIsStaticSwitch=false, bIsVisible=true, type {classStructOrEnum {refPath=&Vector3f}}}
      Origin Offset Coordinate Space {
       bIsDynamic=false
       bIsEditable=true
       bIsStaticSwitch=false
       bIsVisible=true
       type {classStructOrEnum {refPath=&ENiagaraCoordinateSpace}}
      }
      Default Position {
       bIsDynamic=false
       bIsEditable=true
       bIsStaticSwitch=false
       bIsVisible=true
       type {classStructOrEnum {refPath=&NiagaraPosition}}
      }
     ]
     moduleScript {refPath=&AddVelocity}
    }
   ]
  }
  particleUpdateScript {
   scriptName=ParticleUpdateScript
   modules [
    {
     bIsSetParametersModule=false
     enabled=true
     moduleName=ParticleState
     inputs [
      Kill Particles When Lifetime Has Elapsed {bIsDynamic=false, bIsEditable=true, bIsStaticSwitch=true, bIsVisible=true, type {classStructOrEnum {refPath=&NiagaraBool}}}
      Loop Particles Lifetime {bIsDynamic=false, bIsEditable=false, bIsStaticSwitch=false, bIsVisible=false, type {classStructOrEnum {refPath=&NiagaraBool}}}
      Let Infinitely Lived Particles Die When Emitter Deactivates {bIsDynamic=false, bIsEditable=false, bIsStaticSwitch=false, bIsVisible=false, type {classStructOrEnum {refPath=&NiagaraBool}}}
      Lifetime {bIsDynamic=false, bIsEditable=true, bIsStaticSwitch=false, bIsVisible=true, type {classStructOrEnum {refPath=&NiagaraFloat}}}
      DeltaTime {bIsDynamic=false, bIsEditable=true, bIsStaticSwitch=false, bIsVisible=true, type {classStructOrEnum {refPath=&NiagaraFloat}}}
     ]
     moduleScript {refPath=&ParticleState}
    }
    {
     bIsSetParametersModule=false
     enabled=true
     moduleName=Drag
     inputs [
      Use Linear Drag {bIsDynamic=false, bIsEditable=true, bIsStaticSwitch=false, bIsVisible=true, type {classStructOrEnum {refPath=&NiagaraBool}}}
      Drag {bIsDynamic=false, bIsEditable=true, bIsStaticSwitch=false, bIsVisible=true, type {classStructOrEnum {refPath=&NiagaraFloat}}}
      Use Rotational Drag {bIsDynamic=false, bIsEditable=true, bIsStaticSwitch=false, bIsVisible=true, type {classStructOrEnum {refPath=&NiagaraBool}}}
      Rotational Drag {bIsDynamic=false, bIsEditable=true, bIsStaticSwitch=false, bIsVisible=true, type {classStructOrEnum {refPath=&NiagaraFloat}}}
      Ignore Mass {bIsDynamic=false, bIsEditable=true, bIsStaticSwitch=false, bIsVisible=true, type {classStructOrEnum {refPath=&NiagaraBool}}}
     ]
     moduleScript {refPath=&Drag}
    }
    {
     bIsSetParametersModule=false
     enabled=true
     moduleName=SolveForcesAndVelocity
     inputs [
      Clamp Velocity {bIsDynamic=false, bIsEditable=true, bIsStaticSwitch=false, bIsVisible=true, type {classStructOrEnum {refPath=&NiagaraBool}}}
      Speed Limit {bIsDynamic=false, bIsEditable=false, bIsStaticSwitch=false, bIsVisible=true, type {classStructOrEnum {refPath=&NiagaraFloat}}}
      Limit Acceleration {bIsDynamic=false, bIsEditable=true, bIsStaticSwitch=false, bIsVisible=true, type {classStructOrEnum {refPath=&NiagaraBool}}}
      Acceleration Limit {bIsDynamic=false, bIsEditable=false, bIsStaticSwitch=false, bIsVisible=true, type {classStructOrEnum {refPath=&NiagaraFloat}}}
      Manually Enable Rotational Solver {bIsDynamic=false, bIsEditable=true, bIsStaticSwitch=false, bIsVisible=true, type {classStructOrEnum {refPath=&NiagaraBool}}}
      Write to Intrinsic Properties {bIsDynamic=false, bIsEditable=true, bIsStaticSwitch=false, bIsVisible=true, type {classStructOrEnum {refPath=&NiagaraBool}}}
      Write to Presolve Properties {bIsDynamic=false, bIsEditable=true, bIsStaticSwitch=true, bIsVisible=true, type {classStructOrEnum {refPath=&NiagaraBool}}}
      DeltaTime {bIsDynamic=false, bIsEditable=true, bIsStaticSwitch=false, bIsVisible=true, type {classStructOrEnum {refPath=&NiagaraFloat}}}
      Mass {bIsDynamic=false, bIsEditable=true, bIsStaticSwitch=false, bIsVisible=true, type {classStructOrEnum {refPath=&NiagaraFloat}}}
      Position {
       bIsDynamic=false
       bIsEditable=true
       bIsStaticSwitch=false
       bIsVisible=true
       type {classStructOrEnum {refPath=&NiagaraPosition}}
      }
      Velocity {bIsDynamic=false, bIsEditable=true, bIsStaticSwitch=false, bIsVisible=true, type {classStructOrEnum {refPath=&Vector3f}}}
      PreviousVelocity {bIsDynamic=false, bIsEditable=true, bIsStaticSwitch=false, bIsVisible=true, type {classStructOrEnum {refPath=&Vector3f}}}
      Force {bIsDynamic=false, bIsEditable=true, bIsStaticSwitch=false, bIsVisible=true, type {classStructOrEnum {refPath=&Vector3f}}}
      Mesh Orientation {bIsDynamic=false, bIsEditable=false, bIsStaticSwitch=false, bIsVisible=false, type {classStructOrEnum {refPath=&Quat4f}}}
      Rotational Velocity {
       bIsDynamic=false
       bIsEditable=false
       bIsStaticSwitch=false
       bIsVisible=false
       type {classStructOrEnum {refPath=&Vector3f}}
      }
      Rotational Force {
       bIsDynamic=false
       bIsEditable=false
       bIsStaticSwitch=false
       bIsVisible=false
       type {classStructOrEnum {refPath=&Vector3f}}
      }
      Rotational Inertia {
       bIsDynamic=false
       bIsEditable=false
       bIsStaticSwitch=false
       bIsVisible=false
       type {classStructOrEnum {refPath=&NiagaraFloat}}
      }
     ]
     moduleScript {refPath=&SolveForcesAndVelocity}
    }
    {
     bIsSetParametersModule=false
     enabled=true
     moduleName=DynamicMaterialParameters
     inputs [
      Write Parameter Index 0 {bIsDynamic=false, bIsEditable=true, bIsStaticSwitch=true, bIsVisible=true, type {classStructOrEnum {refPath=&NiagaraBool}}}
      Module.Index0Param0Write {bIsDynamic=false, bIsEditable=true, bIsStaticSwitch=true, bIsVisible=true, type {classStructOrEnum {refPath=&NiagaraBool}}}
      Index 0 Param 1 {bIsDynamic=true, bIsEditable=true, bIsStaticSwitch=false, bIsVisible=true, type {classStructOrEnum {refPath=&NiagaraFloat}}}
      Module.Index0Param1Write {bIsDynamic=false, bIsEditable=true, bIsStaticSwitch=true, bIsVisible=true, type {classStructOrEnum {refPath=&NiagaraBool}}}
      Index 0 Param 2 {bIsDynamic=false, bIsEditable=true, bIsStaticSwitch=false, bIsVisible=true, type {classStructOrEnum {refPath=&NiagaraFloat}}}
      Module.Index0Param2Write {bIsDynamic=false, bIsEditable=true, bIsStaticSwitch=true, bIsVisible=true, type {classStructOrEnum {refPath=&NiagaraBool}}}
      Index 0 Param 3 {bIsDynamic=false, bIsEditable=true, bIsStaticSwitch=false, bIsVisible=true, type {classStructOrEnum {refPath=&NiagaraFloat}}}
      Module.Index0Param3Write {bIsDynamic=false, bIsEditable=true, bIsStaticSwitch=true, bIsVisible=true, type {classStructOrEnum {refPath=&NiagaraBool}}}
      Index 0 Param 4 {bIsDynamic=false, bIsEditable=true, bIsStaticSwitch=false, bIsVisible=true, type {classStructOrEnum {refPath=&NiagaraFloat}}}
      Write Parameter Index 1 {bIsDynamic=false, bIsEditable=true, bIsStaticSwitch=true, bIsVisible=true, type {classStructOrEnum {refPath=&NiagaraBool}}}
      Module.Index1Param0Write {bIsDynamic=false, bIsEditable=true, bIsStaticSwitch=true, bIsVisible=true, type {classStructOrEnum {refPath=&NiagaraBool}}}
      Index 1 Param 1 {bIsDynamic=false, bIsEditable=true, bIsStaticSwitch=false, bIsVisible=true, type {classStructOrEnum {refPath=&NiagaraFloat}}}
      Module.Index1Param1Write {bIsDynamic=false, bIsEditable=true, bIsStaticSwitch=true, bIsVisible=true, type {classStructOrEnum {refPath=&NiagaraBool}}}
      Index 1 Param 2 {bIsDynamic=false, bIsEditable=true, bIsStaticSwitch=false, bIsVisible=true, type {classStructOrEnum {refPath=&NiagaraFloat}}}
      Module.Index1Param2Write {bIsDynamic=false, bIsEditable=true, bIsStaticSwitch=true, bIsVisible=true, type {classStructOrEnum {refPath=&NiagaraBool}}}
      Index 1 Param 3 {bIsDynamic=false, bIsEditable=true, bIsStaticSwitch=false, bIsVisible=true, type {classStructOrEnum {refPath=&NiagaraFloat}}}
      Module.Index1Param3Write {bIsDynamic=false, bIsEditable=true, bIsStaticSwitch=true, bIsVisible=true, type {classStructOrEnum {refPath=&NiagaraBool}}}
      Index 1 Param 4 {bIsDynamic=false, bIsEditable=true, bIsStaticSwitch=false, bIsVisible=true, type {classStructOrEnum {refPath=&NiagaraFloat}}}
      Write Parameter Index 2 {bIsDynamic=false, bIsEditable=true, bIsStaticSwitch=true, bIsVisible=true, type {classStructOrEnum {refPath=&NiagaraBool}}}
      Module.Index2Param0Write {bIsDynamic=false, bIsEditable=true, bIsStaticSwitch=true, bIsVisible=true, type {classStructOrEnum {refPath=&NiagaraBool}}}
      Index 2 Param 1 {bIsDynamic=true, bIsEditable=true, bIsStaticSwitch=false, bIsVisible=true, type {classStructOrEnum {refPath=&NiagaraFloat}}}
      Module.Index2Param1Write {bIsDynamic=false, bIsEditable=true, bIsStaticSwitch=true, bIsVisible=true, type {classStructOrEnum {refPath=&NiagaraBool}}}
      Index 2 Param 2 {bIsDynamic=true, bIsEditable=true, bIsStaticSwitch=false, bIsVisible=true, type {classStructOrEnum {refPath=&NiagaraFloat}}}
      Module.Index2Param2Write {bIsDynamic=false, bIsEditable=true, bIsStaticSwitch=true, bIsVisible=true, type {classStructOrEnum {refPath=&NiagaraBool}}}
      Index 2 Param 3 {bIsDynamic=false, bIsEditable=true, bIsStaticSwitch=false, bIsVisible=true, type {classStructOrEnum {refPath=&NiagaraFloat}}}
      Module.Index2Param3Write {bIsDynamic=false, bIsEditable=true, bIsStaticSwitch=true, bIsVisible=true, type {classStructOrEnum {refPath=&NiagaraBool}}}
      Index 2 Param 4 {bIsDynamic=false, bIsEditable=true, bIsStaticSwitch=false, bIsVisible=true, type {classStructOrEnum {refPath=&NiagaraFloat}}}
      Write Parameter Index 3 {bIsDynamic=false, bIsEditable=true, bIsStaticSwitch=true, bIsVisible=true, type {classStructOrEnum {refPath=&NiagaraBool}}}
      Index 3 Param 1 {
       bIsDynamic=false
       bIsEditable=false
       bIsStaticSwitch=false
       bIsVisible=false
       type {classStructOrEnum {refPath=&NiagaraFloat}}
      }
      Index 3 Param 2 {
       bIsDynamic=false
       bIsEditable=false
       bIsStaticSwitch=false
       bIsVisible=false
       type {classStructOrEnum {refPath=&NiagaraFloat}}
      }
      Index 3 Param 3 {
       bIsDynamic=false
       bIsEditable=false
       bIsStaticSwitch=false
       bIsVisible=false
       type {classStructOrEnum {refPath=&NiagaraFloat}}
      }
      Index 3 Param 4 {
       bIsDynamic=false
       bIsEditable=false
       bIsStaticSwitch=false
       bIsVisible=false
       type {classStructOrEnum {refPath=&NiagaraFloat}}
      }
     ]
     moduleScript {refPath=&DynamicMaterialParameters}
    }
    {
     bIsSetParametersModule=false
     enabled=true
     moduleName=ScaleColor
     inputs [
      Scale Mode {
       bIsDynamic=false
       bIsEditable=true
       bIsStaticSwitch=true
       bIsVisible=true
       type {classStructOrEnum {refPath=/Niagara/Enums/ENiagaraScaleColorMode}}
      }
      Linear Color Curve {
       bIsDynamic=false
       bIsEditable=false
       bIsStaticSwitch=false
       bIsVisible=false
       type {classStructOrEnum {refPath=&NiagaraDataInterfaceColorCurve}}
      }
      ScaleRGB {bIsDynamic=false, bIsEditable=true, bIsStaticSwitch=false, bIsVisible=true, type {classStructOrEnum {refPath=&NiagaraBool}}}
      Scale RGB {bIsDynamic=false, bIsEditable=true, bIsStaticSwitch=false, bIsVisible=true, type {classStructOrEnum {refPath=&Vector3f}}}
      ScaleRGBA {bIsDynamic=false, bIsEditable=false, bIsStaticSwitch=false, bIsVisible=false, type {classStructOrEnum {refPath=&NiagaraBool}}}
      Scale RGBA {
       bIsDynamic=false
       bIsEditable=false
       bIsStaticSwitch=false
       bIsVisible=false
       type {classStructOrEnum {refPath=/Script/CoreUObject.Vector4f}}
      }
      Curve Index {
       bIsDynamic=false
       bIsEditable=false
       bIsStaticSwitch=false
       bIsVisible=false
       type {classStructOrEnum {refPath=&NiagaraFloat}}
      }
      ScaleA {bIsDynamic=false, bIsEditable=true, bIsStaticSwitch=false, bIsVisible=true, type {classStructOrEnum {refPath=&NiagaraBool}}}
      Scale Alpha {bIsDynamic=true, bIsEditable=true, bIsStaticSwitch=false, bIsVisible=true, type {classStructOrEnum {refPath=&NiagaraFloat}}}
      Color Value To Scale {
       bIsDynamic=false
       bIsEditable=true
       bIsStaticSwitch=false
       bIsVisible=true
       type {classStructOrEnum {refPath=&LinearColor}}
      }
     ]
     moduleScript {refPath=&ScaleColor}
    }
    {
     bIsSetParametersModule=false
     enabled=true
     moduleName=ScaleSpriteSize
     inputs [
      Scale Sprite Size Mode {
       bIsDynamic=false
       bIsEditable=true
       bIsStaticSwitch=true
       bIsVisible=true
       type {classStructOrEnum {refPath=/Niagara/Enums/SpriteRenderer/ENiagara_ScaleSpriteSize}}
      }
      Uniform Curve Sprite Scale {
       bIsDynamic=false
       bIsEditable=true
       bIsStaticSwitch=false
       bIsVisible=true
       type {classStructOrEnum {refPath=&NiagaraDataInterfaceCurve}}
      }
      Non-Uniform Curve Sprite Scale {
       bIsDynamic=false
       bIsEditable=false
       bIsStaticSwitch=false
       bIsVisible=false
       type {classStructOrEnum {refPath=&NiagaraDataInterfaceVector2DCurve}}
      }
      Uniform Curve Index {bIsDynamic=false, bIsEditable=true, bIsStaticSwitch=false, bIsVisible=true, type {classStructOrEnum {refPath=&NiagaraFloat}}}
      Non-Uniform Curve Index {
       bIsDynamic=false
       bIsEditable=false
       bIsStaticSwitch=false
       bIsVisible=false
       type {classStructOrEnum {refPath=&NiagaraFloat}}
      }
      Non-Uniform Curve Scale {
       bIsDynamic=false
       bIsEditable=false
       bIsStaticSwitch=false
       bIsVisible=false
       type {classStructOrEnum {refPath=&Vector2f}}
      }
      Uniform Curve Scale {bIsDynamic=false, bIsEditable=true, bIsStaticSwitch=false, bIsVisible=true, type {classStructOrEnum {refPath=&NiagaraFloat}}}
      Scale Factor {
       bIsDynamic=false
       bIsEditable=false
       bIsStaticSwitch=false
       bIsVisible=false
       type {classStructOrEnum {refPath=&Vector2f}}
      }
      Uniform Scale Factor {
       bIsDynamic=false
       bIsEditable=false
       bIsStaticSwitch=false
       bIsVisible=false
       type {classStructOrEnum {refPath=&NiagaraFloat}}
      }
      Initial Sprite Size {bIsDynamic=false, bIsEditable=true, bIsStaticSwitch=false, bIsVisible=true, type {classStructOrEnum {refPath=&Vector2f}}}
     ]
     moduleScript {refPath=&ScaleSpriteSize}
    }
   ]
  }
  rendererClasses [{refPath=&NiagaraSpriteRendererProperties}]
  renderers [{rendererIndex=0, rendererClass {refPath=&NiagaraSpriteRendererProperties}}]
 }
 Ground_Cracks001 {
  emitterSpawnScript {scriptName=EmitterSpawnScript, modules []}
  emitterUpdateScript {
   scriptName=EmitterUpdateScript
   modules [
    {
     bIsSetParametersModule=false
     enabled=true
     moduleName=EmitterState
     inputs [
      Life Cycle Mode {
       bIsDynamic=false
       bIsEditable=true
       bIsStaticSwitch=true
       bIsVisible=true
       type {classStructOrEnum {refPath=&ENiagaraEmitterLifeCycleMode}}
      }
      Inactive Response {
       bIsDynamic=false
       bIsEditable=true
       bIsStaticSwitch=true
       bIsVisible=true
       type {classStructOrEnum {refPath=&ENiagaraInactiveMode}}
      }
      Loop Behavior {
       bIsDynamic=false
       bIsEditable=true
       bIsStaticSwitch=true
       bIsVisible=true
       type {classStructOrEnum {refPath=&ENiagara_EmitterStateOptions}}
      }
      Loop Duration Mode {
       bIsDynamic=false
       bIsEditable=true
       bIsStaticSwitch=true
       bIsVisible=true
       type {classStructOrEnum {refPath=&ENiagara_InfiniteLoopDuration}}
      }
      Loop Count {
       bIsDynamic=false
       bIsEditable=false
       bIsStaticSwitch=false
       bIsVisible=false
       type {classStructOrEnum {refPath=&NiagaraInt32}}
      }
      Loop Duration {bIsDynamic=false, bIsEditable=true, bIsStaticSwitch=false, bIsVisible=true, type {classStructOrEnum {refPath=&NiagaraFloat}}}
      Recalculate Duration Each Loop {bIsDynamic=false, bIsEditable=false, bIsStaticSwitch=false, bIsVisible=false, type {classStructOrEnum {refPath=&NiagaraBool}}}
      UseLoopDelay {bIsDynamic=false, bIsEditable=true, bIsStaticSwitch=false, bIsVisible=true, type {classStructOrEnum {refPath=&NiagaraBool}}}
      Loop Delay {bIsDynamic=false, bIsEditable=false, bIsStaticSwitch=false, bIsVisible=true, type {classStructOrEnum {refPath=&NiagaraFloat}}}
      Delay First Loop Only {bIsDynamic=false, bIsEditable=false, bIsStaticSwitch=false, bIsVisible=false, type {classStructOrEnum {refPath=&NiagaraBool}}}
      Scalability Mode {
       bIsDynamic=false
       bIsEditable=true
       bIsStaticSwitch=true
       bIsVisible=true
       type {classStructOrEnum {refPath=&ENiagaraEmitterScalabilityMode_Limited}}
      }
      Enable Distance Culling {bIsDynamic=false, bIsEditable=false, bIsStaticSwitch=true, bIsVisible=false, type {classStructOrEnum {refPath=&NiagaraBool}}}
      MinDistance {
       bIsDynamic=false
       bIsEditable=false
       bIsStaticSwitch=false
       bIsVisible=false
       type {classStructOrEnum {refPath=&NiagaraFloat}}
      }
      Min Distance Response {
       bIsDynamic=false
       bIsEditable=false
       bIsStaticSwitch=false
       bIsVisible=false
       type {classStructOrEnum {refPath=&ENiagaraExecutionStateManagement}}
      }
      MaxDistance {
       bIsDynamic=false
       bIsEditable=false
       bIsStaticSwitch=false
       bIsVisible=false
       type {classStructOrEnum {refPath=&NiagaraFloat}}
      }
      Max Distance Response {
       bIsDynamic=false
       bIsEditable=false
       bIsStaticSwitch=false
       bIsVisible=false
       type {classStructOrEnum {refPath=&ENiagaraExecutionStateManagement}}
      }
      Scale Spawn Count {bIsDynamic=false, bIsEditable=false, bIsStaticSwitch=true, bIsVisible=false, type {classStructOrEnum {refPath=&NiagaraBool}}}
      Spawn Count Scale {
       bIsDynamic=false
       bIsEditable=false
       bIsStaticSwitch=false
       bIsVisible=false
       type {classStructOrEnum {refPath=&NiagaraFloat}}
      }
      Spawn Count Scale By Distance {
       bIsDynamic=false
       bIsEditable=false
       bIsStaticSwitch=false
       bIsVisible=false
       type {classStructOrEnum {refPath=&NiagaraDataInterfaceCurve}}
      }
      Enable Visibility Culling {bIsDynamic=false, bIsEditable=false, bIsStaticSwitch=true, bIsVisible=false, type {classStructOrEnum {refPath=&NiagaraBool}}}
      Visibility Cull Response {
       bIsDynamic=false
       bIsEditable=false
       bIsStaticSwitch=false
       bIsVisible=false
       type {classStructOrEnum {refPath=&ENiagaraExecutionStateManagement}}
      }
      Visibility Cull Delay {
       bIsDynamic=false
       bIsEditable=false
       bIsStaticSwitch=false
       bIsVisible=false
       type {classStructOrEnum {refPath=&NiagaraFloat}}
      }
      Reset Age On Awaken {bIsDynamic=false, bIsEditable=false, bIsStaticSwitch=false, bIsVisible=false, type {classStructOrEnum {refPath=&NiagaraBool}}}
     ]
     moduleScript {refPath=&EmitterState}
    }
    {
     bIsSetParametersModule=false
     enabled=true
     moduleName=SpawnBurst_Instantaneous
     inputs [
      Spawn Count {bIsDynamic=false, bIsEditable=true, bIsStaticSwitch=false, bIsVisible=true, type {classStructOrEnum {refPath=&NiagaraInt32}}}
      Spawn Time {bIsDynamic=false, bIsEditable=true, bIsStaticSwitch=false, bIsVisible=true, type {classStructOrEnum {refPath=&NiagaraFloat}}}
      Use Spawn Probability {bIsDynamic=false, bIsEditable=true, bIsStaticSwitch=false, bIsVisible=true, type {classStructOrEnum {refPath=&NiagaraBool}}}
      Spawn Probability {bIsDynamic=false, bIsEditable=false, bIsStaticSwitch=false, bIsVisible=true, type {classStructOrEnum {refPath=&NiagaraFloat}}}
      UseLoopCountLimit {bIsDynamic=false, bIsEditable=true, bIsStaticSwitch=false, bIsVisible=true, type {classStructOrEnum {refPath=&NiagaraBool}}}
      Loop Count Limit {bIsDynamic=false, bIsEditable=false, bIsStaticSwitch=false, bIsVisible=true, type {classStructOrEnum {refPath=&NiagaraInt32}}}
      Spawn Group {bIsDynamic=false, bIsEditable=true, bIsStaticSwitch=false, bIsVisible=true, type {classStructOrEnum {refPath=&NiagaraInt32}}}
      Age {bIsDynamic=false, bIsEditable=true, bIsStaticSwitch=false, bIsVisible=true, type {classStructOrEnum {refPath=&NiagaraFloat}}}
     ]
     moduleScript {refPath=&SpawnBurst_Instantaneous}
    }
   ]
  }
  particleSpawnScript {
   scriptName=ParticleSpawnScript
   modules [
    {
     bIsSetParametersModule=false
     enabled=true
     moduleName=InitializeParticle001
     inputs [
      Lifetime Mode {
       bIsDynamic=false
       bIsEditable=true
       bIsStaticSwitch=true
       bIsVisible=true
       type {classStructOrEnum {refPath=&ENiagara_LifetimeMode}}
      }
      Lifetime Min {
       bIsDynamic=false
       bIsEditable=false
       bIsStaticSwitch=false
       bIsVisible=false
       type {classStructOrEnum {refPath=&NiagaraFloat}}
      }
      Lifetime {bIsDynamic=false, bIsEditable=true, bIsStaticSwitch=false, bIsVisible=true, type {classStructOrEnum {refPath=&NiagaraFloat}}}
      Lifetime Max {
       bIsDynamic=false
       bIsEditable=false
       bIsStaticSwitch=false
       bIsVisible=false
       type {classStructOrEnum {refPath=&NiagaraFloat}}
      }
      Lifetime Randomness Mode {
       bIsDynamic=false
       bIsEditable=false
       bIsStaticSwitch=true
       bIsVisible=false
       type {classStructOrEnum {refPath=&ENiagaraRandomnessMode}}
      }
      Color Mode {
       bIsDynamic=false
       bIsEditable=true
       bIsStaticSwitch=true
       bIsVisible=true
       type {classStructOrEnum {refPath=&ENiagara_ColorInitializationMode}}
      }
      Color Minimum {
       bIsDynamic=false
       bIsEditable=false
       bIsStaticSwitch=false
       bIsVisible=false
       type {classStructOrEnum {refPath=&LinearColor}}
      }
      Color {
       bIsDynamic=false
       bIsEditable=true
       bIsStaticSwitch=false
       bIsVisible=true
       type {classStructOrEnum {refPath=&LinearColor}}
      }
      AdjustHue {bIsDynamic=false, bIsEditable=false, bIsStaticSwitch=false, bIsVisible=false, type {classStructOrEnum {refPath=&NiagaraBool}}}
      Hue Shift Range {
       bIsDynamic=false
       bIsEditable=false
       bIsStaticSwitch=false
       bIsVisible=false
       type {classStructOrEnum {refPath=&Vector2f}}
      }
      Color Maximum {
       bIsDynamic=false
       bIsEditable=false
       bIsStaticSwitch=false
       bIsVisible=false
       type {classStructOrEnum {refPath=&LinearColor}}
      }
      AdjustSaturation {bIsDynamic=false, bIsEditable=false, bIsStaticSwitch=false, bIsVisible=false, type {classStructOrEnum {refPath=&NiagaraBool}}}
      Saturation Range {
       bIsDynamic=false
       bIsEditable=false
       bIsStaticSwitch=false
       bIsVisible=false
       type {classStructOrEnum {refPath=&Vector2f}}
      }
      AdjustValue {bIsDynamic=false, bIsEditable=false, bIsStaticSwitch=false, bIsVisible=false, type {classStructOrEnum {refPath=&NiagaraBool}}}
      Value Range {
       bIsDynamic=false
       bIsEditable=false
       bIsStaticSwitch=false
       bIsVisible=false
       type {classStructOrEnum {refPath=&Vector2f}}
      }
      AdjustAlpha {bIsDynamic=false, bIsEditable=false, bIsStaticSwitch=false, bIsVisible=false, type {classStructOrEnum {refPath=&NiagaraBool}}}
      Alpha Scale Range {
       bIsDynamic=false
       bIsEditable=false
       bIsStaticSwitch=false
       bIsVisible=false
       type {classStructOrEnum {refPath=&Vector2f}}
      }
      Color Channel Mode {
       bIsDynamic=false
       bIsEditable=false
       bIsStaticSwitch=true
       bIsVisible=false
       type {classStructOrEnum {refPath=&ENiagaraChannelCorrelation}}
      }
      Color Randomness Mode {
       bIsDynamic=false
       bIsEditable=false
       bIsStaticSwitch=true
       bIsVisible=false
       type {classStructOrEnum {refPath=&ENiagaraRandomnessMode}}
      }
      Position Mode {
       bIsDynamic=false
       bIsEditable=true
       bIsStaticSwitch=true
       bIsVisible=true
       type {classStructOrEnum {refPath=&ENiagara_PositionInitializationMode}}
      }
      Position {
       bIsDynamic=false
       bIsEditable=false
       bIsStaticSwitch=false
       bIsVisible=false
       type {classStructOrEnum {refPath=&NiagaraPosition}}
      }
      UsePositionOffset {bIsDynamic=false, bIsEditable=true, bIsStaticSwitch=false, bIsVisible=true, type {classStructOrEnum {refPath=&NiagaraBool}}}
      Position Offset {bIsDynamic=false, bIsEditable=true, bIsStaticSwitch=false, bIsVisible=true, type {classStructOrEnum {refPath=&Vector3f}}}
      Position Offset Coordinate Space {
       bIsDynamic=false
       bIsEditable=true
       bIsStaticSwitch=false
       bIsVisible=true
       type {classStructOrEnum {refPath=&ENiagaraCoordinateSpace}}
      }
      Mass Mode {
       bIsDynamic=false
       bIsEditable=true
       bIsStaticSwitch=true
       bIsVisible=true
       type {classStructOrEnum {refPath=&ENiagara_MassInitializationMode}}
      }
      Mass Min {
       bIsDynamic=false
       bIsEditable=false
       bIsStaticSwitch=false
       bIsVisible=false
       type {classStructOrEnum {refPath=&NiagaraFloat}}
      }
      Mass {
       bIsDynamic=false
       bIsEditable=false
       bIsStaticSwitch=false
       bIsVisible=false
       type {classStructOrEnum {refPath=&NiagaraFloat}}
      }
      Mass Max {
       bIsDynamic=false
       bIsEditable=false
       bIsStaticSwitch=false
       bIsVisible=false
       type {classStructOrEnum {refPath=&NiagaraFloat}}
      }
      Mass Randomness Mode {
       bIsDynamic=false
       bIsEditable=false
       bIsStaticSwitch=true
       bIsVisible=false
       type {classStructOrEnum {refPath=&ENiagaraRandomnessMode}}
      }
      Write Material Random {bIsDynamic=false, bIsEditable=true, bIsStaticSwitch=false, bIsVisible=true, type {classStructOrEnum {refPath=&NiagaraBool}}}
      Material Random {bIsDynamic=false, bIsEditable=true, bIsStaticSwitch=false, bIsVisible=true, type {classStructOrEnum {refPath=&NiagaraFloat}}}
      Sprite Size Mode {
       bIsDynamic=false
       bIsEditable=true
       bIsStaticSwitch=true
       bIsVisible=true
       type {classStructOrEnum {refPath=&ENiagara_SizeScaleMode}}
      }
      Sprite Size {
       bIsDynamic=false
       bIsEditable=false
       bIsStaticSwitch=false
       bIsVisible=false
       type {classStructOrEnum {refPath=&Vector2f}}
      }
      Uniform Sprite Size {
       bIsDynamic=false
       bIsEditable=false
       bIsStaticSwitch=false
       bIsVisible=false
       type {classStructOrEnum {refPath=&NiagaraFloat}}
      }
      Uniform Sprite Size Min {bIsDynamic=false, bIsEditable=true, bIsStaticSwitch=false, bIsVisible=true, type {classStructOrEnum {refPath=&NiagaraFloat}}}
      Sprite Size Min {
       bIsDynamic=false
       bIsEditable=false
       bIsStaticSwitch=false
       bIsVisible=false
       type {classStructOrEnum {refPath=&Vector2f}}
      }
      Sprite Size Max {
       bIsDynamic=false
       bIsEditable=false
       bIsStaticSwitch=false
       bIsVisible=false
       type {classStructOrEnum {refPath=&Vector2f}}
      }
      Uniform Sprite Size Max {bIsDynamic=false, bIsEditable=true, bIsStaticSwitch=false, bIsVisible=true, type {classStructOrEnum {refPath=&NiagaraFloat}}}
      Sprite Size Randomness Mode {
       bIsDynamic=false
       bIsEditable=true
       bIsStaticSwitch=true
       bIsVisible=true
       type {classStructOrEnum {refPath=&ENiagaraRandomnessMode}}
      }
      Sprite Rotation Mode {
       bIsDynamic=false
       bIsEditable=true
       bIsStaticSwitch=true
       bIsVisible=true
       type {classStructOrEnum {refPath=&ENiagara_SpriteRotationMode}}
      }
      Sprite Rotation Angle Min {bIsDynamic=false, bIsEditable=true, bIsStaticSwitch=false, bIsVisible=true, type {classStructOrEnum {refPath=&NiagaraFloat}}}
      Sprite Rotation Angle {
       bIsDynamic=false
       bIsEditable=false
       bIsStaticSwitch=false
       bIsVisible=false
       type {classStructOrEnum {refPath=&NiagaraFloat}}
      }
      Sprite Rotation Angle Max {bIsDynamic=false, bIsEditable=true, bIsStaticSwitch=false, bIsVisible=true, type {classStructOrEnum {refPath=&NiagaraFloat}}}
      Sprite Rotation Randomness Mode {
       bIsDynamic=false
       bIsEditable=true
       bIsStaticSwitch=true
       bIsVisible=true
       type {classStructOrEnum {refPath=&ENiagaraRandomnessMode}}
      }
      Sprite UV Mode {
       bIsDynamic=false
       bIsEditable=true
       bIsStaticSwitch=true
       bIsVisible=true
       type {classStructOrEnum {refPath=&ENiagara_UVFlippingMode}}
      }
      Sprite UV Randomness Mode {
       bIsDynamic=false
       bIsEditable=false
       bIsStaticSwitch=true
       bIsVisible=false
       type {classStructOrEnum {refPath=&ENiagaraRandomnessMode}}
      }
      Sprite UV Scale {
       bIsDynamic=false
       bIsEditable=false
       bIsStaticSwitch=false
       bIsVisible=false
       type {classStructOrEnum {refPath=&Vector2f}}
      }
      Mesh Scale Mode {
       bIsDynamic=false
       bIsEditable=true
       bIsStaticSwitch=true
       bIsVisible=true
       type {classStructOrEnum {refPath=&ENiagara_SizeScaleMode}}
      }
      Mesh Uniform Scale Min {
       bIsDynamic=false
       bIsEditable=false
       bIsStaticSwitch=false
       bIsVisible=false
       type {classStructOrEnum {refPath=&NiagaraFloat}}
      }
      Mesh Uniform Scale {
       bIsDynamic=false
       bIsEditable=false
       bIsStaticSwitch=false
       bIsVisible=false
       type {classStructOrEnum {refPath=&NiagaraFloat}}
      }
      Mesh Scale Min {
       bIsDynamic=false
       bIsEditable=false
       bIsStaticSwitch=false
       bIsVisible=false
       type {classStructOrEnum {refPath=&Vector3f}}
      }
      Mesh Scale {
       bIsDynamic=false
       bIsEditable=false
       bIsStaticSwitch=false
       bIsVisible=false
       type {classStructOrEnum {refPath=&Vector3f}}
      }
      Mesh Scale Max {
       bIsDynamic=false
       bIsEditable=false
       bIsStaticSwitch=false
       bIsVisible=false
       type {classStructOrEnum {refPath=&Vector3f}}
      }
      Mesh Uniform Scale Max {
       bIsDynamic=false
       bIsEditable=false
       bIsStaticSwitch=false
       bIsVisible=false
       type {classStructOrEnum {refPath=&NiagaraFloat}}
      }
      Mesh Scale Randomness Mode {
       bIsDynamic=false
       bIsEditable=false
       bIsStaticSwitch=true
       bIsVisible=false
       type {classStructOrEnum {refPath=&ENiagaraRandomnessMode}}
      }
      Mesh Renderer Array Visibility Mode {
       bIsDynamic=false
       bIsEditable=true
       bIsStaticSwitch=true
       bIsVisible=true
       type {classStructOrEnum {refPath=&ENiagara_UnsetDirectSetRandom}}
      }
      Mesh Renderer Mesh Index {
       bIsDynamic=false
       bIsEditable=false
       bIsStaticSwitch=false
       bIsVisible=false
       type {classStructOrEnum {refPath=&NiagaraInt32}}
      }
      Wrap Index By Mesh Count {bIsDynamic=false, bIsEditable=false, bIsStaticSwitch=true, bIsVisible=false, type {classStructOrEnum {refPath=&NiagaraBool}}}
      Mesh Renderer Info {
       bIsDynamic=false
       bIsEditable=false
       bIsStaticSwitch=false
       bIsVisible=false
       type {classStructOrEnum {refPath=&NiagaraDataInterfaceMeshRendererInfo}}
      }
      Mesh Renderer Random Seed {
       bIsDynamic=false
       bIsEditable=false
       bIsStaticSwitch=false
       bIsVisible=false
       type {classStructOrEnum {refPath=&NiagaraInt32}}
      }
      Ribbon Width Mode {
       bIsDynamic=false
       bIsEditable=true
       bIsStaticSwitch=true
       bIsVisible=true
       type {classStructOrEnum {refPath=&ENiagara_UnsetDirectSet}}
      }
      Ribbon Width {
       bIsDynamic=false
       bIsEditable=false
       bIsStaticSwitch=false
       bIsVisible=false
       type {classStructOrEnum {refPath=&NiagaraFloat}}
      }
      Ribbon Facing Mode {
       bIsDynamic=false
       bIsEditable=true
       bIsStaticSwitch=true
       bIsVisible=true
       type {classStructOrEnum {refPath=&ENiagara_UnsetDirectSet}}
      }
      Ribbon Facing Vector {
       bIsDynamic=false
       bIsEditable=false
       bIsStaticSwitch=false
       bIsVisible=false
       type {classStructOrEnum {refPath=&Vector3f}}
      }
      Ribbon Twist Mode {
       bIsDynamic=false
       bIsEditable=true
       bIsStaticSwitch=true
       bIsVisible=true
       type {classStructOrEnum {refPath=&ENiagara_UnsetDirectSet}}
      }
      "Ribbon Twist " {
       bIsDynamic=false
       bIsEditable=false
       bIsStaticSwitch=false
       bIsVisible=false
       type {classStructOrEnum {refPath=&NiagaraFloat}}
      }
     ]
     moduleScript {refPath=&InitializeParticle}
    }
    {
     bIsSetParametersModule=true
     enabled=true
     moduleName=SetVariables_#1
     inputs [Particles.RandomDecalYZ {bIsDynamic=true, bIsEditable=true, bIsStaticSwitch=false, bIsVisible=true, type {classStructOrEnum {refPath=&NiagaraFloat}}}]
     moduleScript {refPath=&self:Ground_Cracks001_0.NiagaraScriptSource_0.NiagaraGraph_0.NiagaraNodeAssignment_3.SetVariables_#1}
    }
    {
     bIsSetParametersModule=true
     enabled=true
     moduleName=SetVariables_#2
     inputs [Particles.DecalSize {bIsDynamic=true, bIsEditable=true, bIsStaticSwitch=false, bIsVisible=true, type {classStructOrEnum {refPath=&Vector3f}}}]
     moduleScript {refPath=&self:Ground_Cracks001_0.NiagaraScriptSource_0.NiagaraGraph_0.NiagaraNodeAssignment_1.SetVariables_#2}
    }
    {
     bIsSetParametersModule=false
     enabled=true
     moduleName=InitialDecalOrientation
     inputs [
      Decal Orientation Mode {
       bIsDynamic=false
       bIsEditable=true
       bIsStaticSwitch=true
       bIsVisible=true
       type {classStructOrEnum {refPath=&ENiagara_MeshOrientationOptions}}
      }
      Orientation Matrix {
       bIsDynamic=false
       bIsEditable=false
       bIsStaticSwitch=false
       bIsVisible=false
       type {classStructOrEnum {refPath=&NiagaraMatrix}}
      }
      Orientation Vector {bIsDynamic=false, bIsEditable=true, bIsStaticSwitch=false, bIsVisible=true, type {classStructOrEnum {refPath=&Vector3f}}}
      Orientation Coordinate Space {
       bIsDynamic=false
       bIsEditable=true
       bIsStaticSwitch=true
       bIsVisible=true
       type {classStructOrEnum {refPath=&ENiagaraCoordinateSpace}}
      }
      Decal Axis to Orient {bIsDynamic=false, bIsEditable=true, bIsStaticSwitch=false, bIsVisible=true, type {classStructOrEnum {refPath=&Vector3f}}}
      Random Seed {
       bIsDynamic=false
       bIsEditable=false
       bIsStaticSwitch=false
       bIsVisible=false
       type {classStructOrEnum {refPath=&NiagaraInt32}}
      }
      Use Rotation Vector {bIsDynamic=false, bIsEditable=true, bIsStaticSwitch=false, bIsVisible=true, type {classStructOrEnum {refPath=&NiagaraBool}}}
      Rotation {bIsDynamic=true, bIsEditable=true, bIsStaticSwitch=false, bIsVisible=true, type {classStructOrEnum {refPath=&Vector3f}}}
      Rotation Coordinate Space {
       bIsDynamic=false
       bIsEditable=true
       bIsStaticSwitch=true
       bIsVisible=true
       type {classStructOrEnum {refPath=/Niagara/Enums/ENiagaraDecalTransforms}}
      }
      Orientation Quaternion {bIsDynamic=false, bIsEditable=false, bIsStaticSwitch=false, bIsVisible=false, type {classStructOrEnum {refPath=&Quat4f}}}
     ]
     moduleScript {refPath=&InitialDecalOrientation}
    }
   ]
  }
  particleUpdateScript {
   scriptName=ParticleUpdateScript
   modules [
    {
     bIsSetParametersModule=false
     enabled=true
     moduleName=ParticleState
     inputs [
      Kill Particles When Lifetime Has Elapsed {bIsDynamic=false, bIsEditable=true, bIsStaticSwitch=true, bIsVisible=true, type {classStructOrEnum {refPath=&NiagaraBool}}}
      Loop Particles Lifetime {bIsDynamic=false, bIsEditable=false, bIsStaticSwitch=false, bIsVisible=false, type {classStructOrEnum {refPath=&NiagaraBool}}}
      Let Infinitely Lived Particles Die When Emitter Deactivates {bIsDynamic=false, bIsEditable=false, bIsStaticSwitch=false, bIsVisible=false, type {classStructOrEnum {refPath=&NiagaraBool}}}
      Lifetime {bIsDynamic=false, bIsEditable=true, bIsStaticSwitch=false, bIsVisible=true, type {classStructOrEnum {refPath=&NiagaraFloat}}}
      DeltaTime {bIsDynamic=false, bIsEditable=true, bIsStaticSwitch=false, bIsVisible=true, type {classStructOrEnum {refPath=&NiagaraFloat}}}
     ]
     moduleScript {refPath=&ParticleState}
    }
    {
     bIsSetParametersModule=true
     enabled=true
     moduleName=SetVariables_#3
     inputs [Particles.DecalFade {bIsDynamic=true, bIsEditable=true, bIsStaticSwitch=false, bIsVisible=true, type {classStructOrEnum {refPath=&NiagaraFloat}}}]
     moduleScript {refPath=&self:Ground_Cracks001_0.NiagaraScriptSource_0.NiagaraGraph_0.NiagaraNodeAssignment_2.SetVariables_#3}
    }
    {
     bIsSetParametersModule=true
     enabled=true
     moduleName=SetVariables_#4
     inputs [Particles.DecalSize {bIsDynamic=true, bIsEditable=true, bIsStaticSwitch=false, bIsVisible=true, type {classStructOrEnum {refPath=&Vector3f}}}]
     moduleScript {refPath=&self:Ground_Cracks001_0.NiagaraScriptSource_0.NiagaraGraph_0.NiagaraNodeAssignment_4.SetVariables_#4}
    }
    {
     bIsSetParametersModule=false
     enabled=true
     moduleName=SolveForcesAndVelocity
     inputs [
      Clamp Velocity {bIsDynamic=false, bIsEditable=true, bIsStaticSwitch=false, bIsVisible=true, type {classStructOrEnum {refPath=&NiagaraBool}}}
      Speed Limit {bIsDynamic=false, bIsEditable=false, bIsStaticSwitch=false, bIsVisible=true, type {classStructOrEnum {refPath=&NiagaraFloat}}}
      Limit Acceleration {bIsDynamic=false, bIsEditable=true, bIsStaticSwitch=false, bIsVisible=true, type {classStructOrEnum {refPath=&NiagaraBool}}}
      Acceleration Limit {bIsDynamic=false, bIsEditable=false, bIsStaticSwitch=false, bIsVisible=true, type {classStructOrEnum {refPath=&NiagaraFloat}}}
      Manually Enable Rotational Solver {bIsDynamic=false, bIsEditable=true, bIsStaticSwitch=false, bIsVisible=true, type {classStructOrEnum {refPath=&NiagaraBool}}}
      Write to Intrinsic Properties {bIsDynamic=false, bIsEditable=true, bIsStaticSwitch=false, bIsVisible=true, type {classStructOrEnum {refPath=&NiagaraBool}}}
      Write to Presolve Properties {bIsDynamic=false, bIsEditable=true, bIsStaticSwitch=true, bIsVisible=true, type {classStructOrEnum {refPath=&NiagaraBool}}}
      DeltaTime {bIsDynamic=false, bIsEditable=true, bIsStaticSwitch=false, bIsVisible=true, type {classStructOrEnum {refPath=&NiagaraFloat}}}
      Mass {bIsDynamic=false, bIsEditable=true, bIsStaticSwitch=false, bIsVisible=true, type {classStructOrEnum {refPath=&NiagaraFloat}}}
      Position {
       bIsDynamic=false
       bIsEditable=true
       bIsStaticSwitch=false
       bIsVisible=true
       type {classStructOrEnum {refPath=&NiagaraPosition}}
      }
      Velocity {bIsDynamic=false, bIsEditable=true, bIsStaticSwitch=false, bIsVisible=true, type {classStructOrEnum {refPath=&Vector3f}}}
      PreviousVelocity {bIsDynamic=false, bIsEditable=true, bIsStaticSwitch=false, bIsVisible=true, type {classStructOrEnum {refPath=&Vector3f}}}
      Force {bIsDynamic=false, bIsEditable=true, bIsStaticSwitch=false, bIsVisible=true, type {classStructOrEnum {refPath=&Vector3f}}}
      Mesh Orientation {bIsDynamic=false, bIsEditable=false, bIsStaticSwitch=false, bIsVisible=false, type {classStructOrEnum {refPath=&Quat4f}}}
      Rotational Velocity {
       bIsDynamic=false
       bIsEditable=false
       bIsStaticSwitch=false
       bIsVisible=false
       type {classStructOrEnum {refPath=&Vector3f}}
      }
      Rotational Force {
       bIsDynamic=false
       bIsEditable=false
       bIsStaticSwitch=false
       bIsVisible=false
       type {classStructOrEnum {refPath=&Vector3f}}
      }
      Rotational Inertia {
       bIsDynamic=false
       bIsEditable=false
       bIsStaticSwitch=false
       bIsVisible=false
       type {classStructOrEnum {refPath=&NiagaraFloat}}
      }
     ]
     moduleScript {refPath=&SolveForcesAndVelocity}
    }
   ]
  }
  rendererClasses [{refPath=&NiagaraDecalRendererProperties}]
  renderers [{rendererIndex=0, rendererClass {refPath=&NiagaraDecalRendererProperties}}]
 }
 Ground_Crater {
  emitterSpawnScript {scriptName=EmitterSpawnScript, modules []}
  emitterUpdateScript {
   scriptName=EmitterUpdateScript
   modules [
    {
     bIsSetParametersModule=false
     enabled=true
     moduleName=EmitterState
     inputs [
      Life Cycle Mode {
       bIsDynamic=false
       bIsEditable=true
       bIsStaticSwitch=true
       bIsVisible=true
       type {classStructOrEnum {refPath=&ENiagaraEmitterLifeCycleMode}}
      }
      Inactive Response {
       bIsDynamic=false
       bIsEditable=true
       bIsStaticSwitch=true
       bIsVisible=true
       type {classStructOrEnum {refPath=&ENiagaraInactiveMode}}
      }
      Loop Behavior {
       bIsDynamic=false
       bIsEditable=true
       bIsStaticSwitch=true
       bIsVisible=true
       type {classStructOrEnum {refPath=&ENiagara_EmitterStateOptions}}
      }
      Loop Duration Mode {
       bIsDynamic=false
       bIsEditable=true
       bIsStaticSwitch=true
       bIsVisible=true
       type {classStructOrEnum {refPath=&ENiagara_InfiniteLoopDuration}}
      }
      Loop Count {
       bIsDynamic=false
       bIsEditable=false
       bIsStaticSwitch=false
       bIsVisible=false
       type {classStructOrEnum {refPath=&NiagaraInt32}}
      }
      Loop Duration {bIsDynamic=false, bIsEditable=true, bIsStaticSwitch=false, bIsVisible=true, type {classStructOrEnum {refPath=&NiagaraFloat}}}
      Recalculate Duration Each Loop {bIsDynamic=false, bIsEditable=false, bIsStaticSwitch=false, bIsVisible=false, type {classStructOrEnum {refPath=&NiagaraBool}}}
      UseLoopDelay {bIsDynamic=false, bIsEditable=true, bIsStaticSwitch=false, bIsVisible=true, type {classStructOrEnum {refPath=&NiagaraBool}}}
      Loop Delay {bIsDynamic=false, bIsEditable=false, bIsStaticSwitch=false, bIsVisible=true, type {classStructOrEnum {refPath=&NiagaraFloat}}}
      Delay First Loop Only {bIsDynamic=false, bIsEditable=false, bIsStaticSwitch=false, bIsVisible=false, type {classStructOrEnum {refPath=&NiagaraBool}}}
      Scalability Mode {
       bIsDynamic=false
       bIsEditable=true
       bIsStaticSwitch=true
       bIsVisible=true
       type {classStructOrEnum {refPath=&ENiagaraEmitterScalabilityMode_Limited}}
      }
      Enable Distance Culling {bIsDynamic=false, bIsEditable=false, bIsStaticSwitch=true, bIsVisible=false, type {classStructOrEnum {refPath=&NiagaraBool}}}
      MinDistance {
       bIsDynamic=false
       bIsEditable=false
       bIsStaticSwitch=false
       bIsVisible=false
       type {classStructOrEnum {refPath=&NiagaraFloat}}
      }
      Min Distance Response {
       bIsDynamic=false
       bIsEditable=false
       bIsStaticSwitch=false
       bIsVisible=false
       type {classStructOrEnum {refPath=&ENiagaraExecutionStateManagement}}
      }
      MaxDistance {
       bIsDynamic=false
       bIsEditable=false
       bIsStaticSwitch=false
       bIsVisible=false
       type {classStructOrEnum {refPath=&NiagaraFloat}}
      }
      Max Distance Response {
       bIsDynamic=false
       bIsEditable=false
       bIsStaticSwitch=false
       bIsVisible=false
       type {classStructOrEnum {refPath=&ENiagaraExecutionStateManagement}}
      }
      Scale Spawn Count {bIsDynamic=false, bIsEditable=false, bIsStaticSwitch=true, bIsVisible=false, type {classStructOrEnum {refPath=&NiagaraBool}}}
      Spawn Count Scale {
       bIsDynamic=false
       bIsEditable=false
       bIsStaticSwitch=false
       bIsVisible=false
       type {classStructOrEnum {refPath=&NiagaraFloat}}
      }
      Spawn Count Scale By Distance {
       bIsDynamic=false
       bIsEditable=false
       bIsStaticSwitch=false
       bIsVisible=false
       type {classStructOrEnum {refPath=&NiagaraDataInterfaceCurve}}
      }
      Enable Visibility Culling {bIsDynamic=false, bIsEditable=false, bIsStaticSwitch=true, bIsVisible=false, type {classStructOrEnum {refPath=&NiagaraBool}}}
      Visibility Cull Response {
       bIsDynamic=false
       bIsEditable=false
       bIsStaticSwitch=false
       bIsVisible=false
       type {classStructOrEnum {refPath=&ENiagaraExecutionStateManagement}}
      }
      Visibility Cull Delay {
       bIsDynamic=false
       bIsEditable=false
       bIsStaticSwitch=false
       bIsVisible=false
       type {classStructOrEnum {refPath=&NiagaraFloat}}
      }
      Reset Age On Awaken {bIsDynamic=false, bIsEditable=false, bIsStaticSwitch=false, bIsVisible=false, type {classStructOrEnum {refPath=&NiagaraBool}}}
     ]
     moduleScript {refPath=&EmitterState}
    }
    {
     bIsSetParametersModule=false
     enabled=true
     moduleName=SpawnBurst_Instantaneous
     inputs [
      Spawn Count {bIsDynamic=false, bIsEditable=true, bIsStaticSwitch=false, bIsVisible=true, type {classStructOrEnum {refPath=&NiagaraInt32}}}
      Spawn Time {bIsDynamic=false, bIsEditable=true, bIsStaticSwitch=false, bIsVisible=true, type {classStructOrEnum {refPath=&NiagaraFloat}}}
      Use Spawn Probability {bIsDynamic=false, bIsEditable=true, bIsStaticSwitch=false, bIsVisible=true, type {classStructOrEnum {refPath=&NiagaraBool}}}
      Spawn Probability {bIsDynamic=false, bIsEditable=false, bIsStaticSwitch=false, bIsVisible=true, type {classStructOrEnum {refPath=&NiagaraFloat}}}
      UseLoopCountLimit {bIsDynamic=false, bIsEditable=true, bIsStaticSwitch=false, bIsVisible=true, type {classStructOrEnum {refPath=&NiagaraBool}}}
      Loop Count Limit {bIsDynamic=false, bIsEditable=false, bIsStaticSwitch=false, bIsVisible=true, type {classStructOrEnum {refPath=&NiagaraInt32}}}
      Spawn Group {bIsDynamic=false, bIsEditable=true, bIsStaticSwitch=false, bIsVisible=true, type {classStructOrEnum {refPath=&NiagaraInt32}}}
      Age {bIsDynamic=false, bIsEditable=true, bIsStaticSwitch=false, bIsVisible=true, type {classStructOrEnum {refPath=&NiagaraFloat}}}
     ]
     moduleScript {refPath=&SpawnBurst_Instantaneous}
    }
   ]
  }
  particleSpawnScript {
   scriptName=ParticleSpawnScript
   modules [
    {
     bIsSetParametersModule=false
     enabled=false
     moduleName=DynamicMaterialParameters
     inputs [
      Write Parameter Index 0 {
       bIsDynamic=false
       bIsEditable=true
       bIsStaticSwitch=true
       bIsVisible=true
       type {classStructOrEnum {refPath=/Niagara/Enums/Utility/ENiagara_DynamicMaterialParametersInputMode}}
      }
      Index 0 Vector 1 {
       bIsDynamic=false
       bIsEditable=false
       bIsStaticSwitch=false
       bIsVisible=false
       type {classStructOrEnum {refPath=&Vector3f}}
      }
      Module.Index0Float0Write {bIsDynamic=false, bIsEditable=true, bIsStaticSwitch=true, bIsVisible=true, type {classStructOrEnum {refPath=&NiagaraBool}}}
      Index 0 Param 1 {bIsDynamic=true, bIsEditable=true, bIsStaticSwitch=false, bIsVisible=true, type {classStructOrEnum {refPath=&NiagaraFloat}}}
      Module.Index0Float1Write {bIsDynamic=false, bIsEditable=true, bIsStaticSwitch=true, bIsVisible=true, type {classStructOrEnum {refPath=&NiagaraBool}}}
      Index 0 Param 2 {bIsDynamic=false, bIsEditable=true, bIsStaticSwitch=false, bIsVisible=true, type {classStructOrEnum {refPath=&NiagaraFloat}}}
      Module.Index0Float2Write {bIsDynamic=false, bIsEditable=true, bIsStaticSwitch=true, bIsVisible=true, type {classStructOrEnum {refPath=&NiagaraBool}}}
      Index 0 Param 3 {bIsDynamic=false, bIsEditable=true, bIsStaticSwitch=false, bIsVisible=true, type {classStructOrEnum {refPath=&NiagaraFloat}}}
      Module.Index0Float3Write {bIsDynamic=false, bIsEditable=true, bIsStaticSwitch=true, bIsVisible=true, type {classStructOrEnum {refPath=&NiagaraBool}}}
      Index 0 Param 4 {bIsDynamic=false, bIsEditable=true, bIsStaticSwitch=false, bIsVisible=true, type {classStructOrEnum {refPath=&NiagaraFloat}}}
      Index 0 Vector2D 1 {
       bIsDynamic=false
       bIsEditable=false
       bIsStaticSwitch=false
       bIsVisible=false
       type {classStructOrEnum {refPath=&Vector2f}}
      }
      Index 0 Vector2D 2 {
       bIsDynamic=false
       bIsEditable=false
       bIsStaticSwitch=false
       bIsVisible=false
       type {classStructOrEnum {refPath=&Vector2f}}
      }
      Index 0 Vector 4 {
       bIsDynamic=false
       bIsEditable=false
       bIsStaticSwitch=false
       bIsVisible=false
       type {classStructOrEnum {refPath=&Vector4f}}
      }
      Index 0 Linear Color {
       bIsDynamic=false
       bIsEditable=false
       bIsStaticSwitch=false
       bIsVisible=false
       type {classStructOrEnum {refPath=&LinearColor}}
      }
      Write Parameter Index 1 {
       bIsDynamic=false
       bIsEditable=true
       bIsStaticSwitch=true
       bIsVisible=true
       type {classStructOrEnum {refPath=&ENiagara_DynamicMaterialParametersInputMode}}
      }
      Index 1 Vector 1 {
       bIsDynamic=false
       bIsEditable=false
       bIsStaticSwitch=false
       bIsVisible=false
       type {classStructOrEnum {refPath=&Vector3f}}
      }
      Index 1 Param 1 {
       bIsDynamic=false
       bIsEditable=false
       bIsStaticSwitch=false
       bIsVisible=false
       type {classStructOrEnum {refPath=&NiagaraFloat}}
      }
      Index 1 Param 2 {
       bIsDynamic=false
       bIsEditable=false
       bIsStaticSwitch=false
       bIsVisible=false
       type {classStructOrEnum {refPath=&NiagaraFloat}}
      }
      Index 1 Param 3 {
       bIsDynamic=false
       bIsEditable=false
       bIsStaticSwitch=false
       bIsVisible=false
       type {classStructOrEnum {refPath=&NiagaraFloat}}
      }
      Index 1 Param 4 {
       bIsDynamic=false
       bIsEditable=false
       bIsStaticSwitch=false
       bIsVisible=false
       type {classStructOrEnum {refPath=&NiagaraFloat}}
      }
      Index 1 Vector2D 1 {
       bIsDynamic=false
       bIsEditable=false
       bIsStaticSwitch=false
       bIsVisible=false
       type {classStructOrEnum {refPath=&Vector2f}}
      }
      Index 1 Vector2D 2 {
       bIsDynamic=false
       bIsEditable=false
       bIsStaticSwitch=false
       bIsVisible=false
       type {classStructOrEnum {refPath=&Vector2f}}
      }
      Index 1 Vector 4 {
       bIsDynamic=false
       bIsEditable=false
       bIsStaticSwitch=false
       bIsVisible=false
       type {classStructOrEnum {refPath=&Vector4f}}
      }
      Index 1 Linear Color {
       bIsDynamic=false
       bIsEditable=false
       bIsStaticSwitch=false
       bIsVisible=false
       type {classStructOrEnum {refPath=&LinearColor}}
      }
      Write Parameter Index 2 {
       bIsDynamic=false
       bIsEditable=true
       bIsStaticSwitch=true
       bIsVisible=true
       type {classStructOrEnum {refPath=&ENiagara_DynamicMaterialParametersInputMode}}
      }
      Index 2 Vector 1 {
       bIsDynamic=false
       bIsEditable=false
       bIsStaticSwitch=false
       bIsVisible=false
       type {classStructOrEnum {refPath=&Vector3f}}
      }
      Index 2 Param 1 {
       bIsDynamic=false
       bIsEditable=false
       bIsStaticSwitch=false
       bIsVisible=false
       type {classStructOrEnum {refPath=&NiagaraFloat}}
      }
      Index 2 Param 2 {
       bIsDynamic=false
       bIsEditable=false
       bIsStaticSwitch=false
       bIsVisible=false
       type {classStructOrEnum {refPath=&NiagaraFloat}}
      }
      Index 2 Param 3 {
       bIsDynamic=false
       bIsEditable=false
       bIsStaticSwitch=false
       bIsVisible=false
       type {classStructOrEnum {refPath=&NiagaraFloat}}
      }
      Index 2 Param 4 {
       bIsDynamic=false
       bIsEditable=false
       bIsStaticSwitch=false
       bIsVisible=false
       type {classStructOrEnum {refPath=&NiagaraFloat}}
      }
      Index 2 Vector2D 1 {
       bIsDynamic=false
       bIsEditable=false
       bIsStaticSwitch=false
       bIsVisible=false
       type {classStructOrEnum {refPath=&Vector2f}}
      }
      Index 2 Vector2D 2 {
       bIsDynamic=false
       bIsEditable=false
       bIsStaticSwitch=false
       bIsVisible=false
       type {classStructOrEnum {refPath=&Vector2f}}
      }
      Index 2 Vector 4 {
       bIsDynamic=false
       bIsEditable=false
       bIsStaticSwitch=false
       bIsVisible=false
       type {classStructOrEnum {refPath=&Vector4f}}
      }
      Index 2 Linear Color {
       bIsDynamic=false
       bIsEditable=false
       bIsStaticSwitch=false
       bIsVisible=false
       type {classStructOrEnum {refPath=&LinearColor}}
      }
      Write Parameter Index 3 {
       bIsDynamic=false
       bIsEditable=true
       bIsStaticSwitch=true
       bIsVisible=true
       type {classStructOrEnum {refPath=&ENiagara_DynamicMaterialParametersInputMode}}
      }
      Index 3 Vector 1 {
       bIsDynamic=false
       bIsEditable=false
       bIsStaticSwitch=false
       bIsVisible=false
       type {classStructOrEnum {refPath=&Vector3f}}
      }
      Index 3 Param 1 {
       bIsDynamic=false
       bIsEditable=false
       bIsStaticSwitch=false
       bIsVisible=false
       type {classStructOrEnum {refPath=&NiagaraFloat}}
      }
      Index 3 Param 2 {
       bIsDynamic=false
       bIsEditable=false
       bIsStaticSwitch=false
       bIsVisible=false
       type {classStructOrEnum {refPath=&NiagaraFloat}}
      }
      Index 3 Param 3 {
       bIsDynamic=false
       bIsEditable=false
       bIsStaticSwitch=false
       bIsVisible=false
       type {classStructOrEnum {refPath=&NiagaraFloat}}
      }
      Index 3 Param 4 {
       bIsDynamic=false
       bIsEditable=false
       bIsStaticSwitch=false
       bIsVisible=false
       type {classStructOrEnum {refPath=&NiagaraFloat}}
      }
      Index 3 Vector2D 1 {
       bIsDynamic=false
       bIsEditable=false
       bIsStaticSwitch=false
       bIsVisible=false
       type {classStructOrEnum {refPath=&Vector2f}}
      }
      Index 3 Vector2D 2 {
       bIsDynamic=false
       bIsEditable=false
       bIsStaticSwitch=false
       bIsVisible=false
       type {classStructOrEnum {refPath=&Vector2f}}
      }
      Index 3 Vector 4 {
       bIsDynamic=false
       bIsEditable=false
       bIsStaticSwitch=false
       bIsVisible=false
       type {classStructOrEnum {refPath=&Vector4f}}
      }
      Index 3 Linear Color {
       bIsDynamic=false
       bIsEditable=false
       bIsStaticSwitch=false
       bIsVisible=false
       type {classStructOrEnum {refPath=&LinearColor}}
      }
     ]
     moduleScript {refPath=&DynamicMaterialParameters}
    }
    {
     bIsSetParametersModule=false
     enabled=true
     moduleName=InitializeParticle001
     inputs [
      Lifetime Mode {
       bIsDynamic=false
       bIsEditable=true
       bIsStaticSwitch=true
       bIsVisible=true
       type {classStructOrEnum {refPath=&ENiagara_LifetimeMode}}
      }
      Lifetime Min {
       bIsDynamic=false
       bIsEditable=false
       bIsStaticSwitch=false
       bIsVisible=false
       type {classStructOrEnum {refPath=&NiagaraFloat}}
      }
      Lifetime {bIsDynamic=true, bIsEditable=true, bIsStaticSwitch=false, bIsVisible=true, type {classStructOrEnum {refPath=&NiagaraFloat}}}
      Lifetime Max {
       bIsDynamic=false
       bIsEditable=false
       bIsStaticSwitch=false
       bIsVisible=false
       type {classStructOrEnum {refPath=&NiagaraFloat}}
      }
      Lifetime Randomness Mode {
       bIsDynamic=false
       bIsEditable=false
       bIsStaticSwitch=true
       bIsVisible=false
       type {classStructOrEnum {refPath=&ENiagaraRandomnessMode}}
      }
      Color Mode {
       bIsDynamic=false
       bIsEditable=true
       bIsStaticSwitch=true
       bIsVisible=true
       type {classStructOrEnum {refPath=&ENiagara_ColorInitializationMode}}
      }
      Color Minimum {
       bIsDynamic=false
       bIsEditable=false
       bIsStaticSwitch=false
       bIsVisible=false
       type {classStructOrEnum {refPath=&LinearColor}}
      }
      Color {
       bIsDynamic=false
       bIsEditable=true
       bIsStaticSwitch=false
       bIsVisible=true
       type {classStructOrEnum {refPath=&LinearColor}}
      }
      AdjustHue {bIsDynamic=false, bIsEditable=false, bIsStaticSwitch=false, bIsVisible=false, type {classStructOrEnum {refPath=&NiagaraBool}}}
      Hue Shift Range {
       bIsDynamic=false
       bIsEditable=false
       bIsStaticSwitch=false
       bIsVisible=false
       type {classStructOrEnum {refPath=&Vector2f}}
      }
      Color Maximum {
       bIsDynamic=false
       bIsEditable=false
       bIsStaticSwitch=false
       bIsVisible=false
       type {classStructOrEnum {refPath=&LinearColor}}
      }
      AdjustSaturation {bIsDynamic=false, bIsEditable=false, bIsStaticSwitch=false, bIsVisible=false, type {classStructOrEnum {refPath=&NiagaraBool}}}
      Saturation Range {
       bIsDynamic=false
       bIsEditable=false
       bIsStaticSwitch=false
       bIsVisible=false
       type {classStructOrEnum {refPath=&Vector2f}}
      }
      AdjustValue {bIsDynamic=false, bIsEditable=false, bIsStaticSwitch=false, bIsVisible=false, type {classStructOrEnum {refPath=&NiagaraBool}}}
      Value Range {
       bIsDynamic=false
       bIsEditable=false
       bIsStaticSwitch=false
       bIsVisible=false
       type {classStructOrEnum {refPath=&Vector2f}}
      }
      AdjustAlpha {bIsDynamic=false, bIsEditable=false, bIsStaticSwitch=false, bIsVisible=false, type {classStructOrEnum {refPath=&NiagaraBool}}}
      Alpha Scale Range {
       bIsDynamic=false
       bIsEditable=false
       bIsStaticSwitch=false
       bIsVisible=false
       type {classStructOrEnum {refPath=&Vector2f}}
      }
      Color Channel Mode {
       bIsDynamic=false
       bIsEditable=false
       bIsStaticSwitch=true
       bIsVisible=false
       type {classStructOrEnum {refPath=&ENiagaraChannelCorrelation}}
      }
      Color Randomness Mode {
       bIsDynamic=false
       bIsEditable=false
       bIsStaticSwitch=true
       bIsVisible=false
       type {classStructOrEnum {refPath=&ENiagaraRandomnessMode}}
      }
      Position Mode {
       bIsDynamic=false
       bIsEditable=true
       bIsStaticSwitch=true
       bIsVisible=true
       type {classStructOrEnum {refPath=&ENiagara_PositionInitializationMode}}
      }
      Position {
       bIsDynamic=false
       bIsEditable=false
       bIsStaticSwitch=false
       bIsVisible=false
       type {classStructOrEnum {refPath=&NiagaraPosition}}
      }
      UsePositionOffset {bIsDynamic=false, bIsEditable=true, bIsStaticSwitch=false, bIsVisible=true, type {classStructOrEnum {refPath=&NiagaraBool}}}
      Position Offset {bIsDynamic=false, bIsEditable=false, bIsStaticSwitch=false, bIsVisible=true, type {classStructOrEnum {refPath=&Vector3f}}}
      Position Offset Coordinate Space {
       bIsDynamic=false
       bIsEditable=false
       bIsStaticSwitch=false
       bIsVisible=false
       type {classStructOrEnum {refPath=&ENiagaraCoordinateSpace}}
      }
      Mass Mode {
       bIsDynamic=false
       bIsEditable=true
       bIsStaticSwitch=true
       bIsVisible=true
       type {classStructOrEnum {refPath=&ENiagara_MassInitializationMode}}
      }
      Mass Min {
       bIsDynamic=false
       bIsEditable=false
       bIsStaticSwitch=false
       bIsVisible=false
       type {classStructOrEnum {refPath=&NiagaraFloat}}
      }
      Mass {
       bIsDynamic=false
       bIsEditable=false
       bIsStaticSwitch=false
       bIsVisible=false
       type {classStructOrEnum {refPath=&NiagaraFloat}}
      }
      Mass Max {
       bIsDynamic=false
       bIsEditable=false
       bIsStaticSwitch=false
       bIsVisible=false
       type {classStructOrEnum {refPath=&NiagaraFloat}}
      }
      Mass Randomness Mode {
       bIsDynamic=false
       bIsEditable=false
       bIsStaticSwitch=true
       bIsVisible=false
       type {classStructOrEnum {refPath=&ENiagaraRandomnessMode}}
      }
      Write Material Random {bIsDynamic=false, bIsEditable=true, bIsStaticSwitch=false, bIsVisible=true, type {classStructOrEnum {refPath=&NiagaraBool}}}
      Material Random {bIsDynamic=false, bIsEditable=true, bIsStaticSwitch=false, bIsVisible=true, type {classStructOrEnum {refPath=&NiagaraFloat}}}
      Sprite Size Mode {
       bIsDynamic=false
       bIsEditable=true
       bIsStaticSwitch=true
       bIsVisible=true
       type {classStructOrEnum {refPath=&ENiagara_SizeScaleMode}}
      }
      Sprite Size {
       bIsDynamic=false
       bIsEditable=false
       bIsStaticSwitch=false
       bIsVisible=false
       type {classStructOrEnum {refPath=&Vector2f}}
      }
      Uniform Sprite Size {
       bIsDynamic=false
       bIsEditable=false
       bIsStaticSwitch=false
       bIsVisible=false
       type {classStructOrEnum {refPath=&NiagaraFloat}}
      }
      Uniform Sprite Size Min {bIsDynamic=false, bIsEditable=true, bIsStaticSwitch=false, bIsVisible=true, type {classStructOrEnum {refPath=&NiagaraFloat}}}
      Sprite Size Min {
       bIsDynamic=false
       bIsEditable=false
       bIsStaticSwitch=false
       bIsVisible=false
       type {classStructOrEnum {refPath=&Vector2f}}
      }
      Sprite Size Max {
       bIsDynamic=false
       bIsEditable=false
       bIsStaticSwitch=false
       bIsVisible=false
       type {classStructOrEnum {refPath=&Vector2f}}
      }
      Uniform Sprite Size Max {bIsDynamic=false, bIsEditable=true, bIsStaticSwitch=false, bIsVisible=true, type {classStructOrEnum {refPath=&NiagaraFloat}}}
      Sprite Size Randomness Mode {
       bIsDynamic=false
       bIsEditable=true
       bIsStaticSwitch=true
       bIsVisible=true
       type {classStructOrEnum {refPath=&ENiagaraRandomnessMode}}
      }
      Sprite Rotation Mode {
       bIsDynamic=false
       bIsEditable=true
       bIsStaticSwitch=true
       bIsVisible=true
       type {classStructOrEnum {refPath=&ENiagara_SpriteRotationMode}}
      }
      Sprite Rotation Angle Min {bIsDynamic=false, bIsEditable=true, bIsStaticSwitch=false, bIsVisible=true, type {classStructOrEnum {refPath=&NiagaraFloat}}}
      Sprite Rotation Angle {
       bIsDynamic=false
       bIsEditable=false
       bIsStaticSwitch=false
       bIsVisible=false
       type {classStructOrEnum {refPath=&NiagaraFloat}}
      }
      Sprite Rotation Angle Max {bIsDynamic=false, bIsEditable=true, bIsStaticSwitch=false, bIsVisible=true, type {classStructOrEnum {refPath=&NiagaraFloat}}}
      Sprite Rotation Randomness Mode {
       bIsDynamic=false
       bIsEditable=true
       bIsStaticSwitch=true
       bIsVisible=true
       type {classStructOrEnum {refPath=&ENiagaraRandomnessMode}}
      }
      Sprite UV Mode {
       bIsDynamic=false
       bIsEditable=true
       bIsStaticSwitch=true
       bIsVisible=true
       type {classStructOrEnum {refPath=&ENiagara_UVFlippingMode}}
      }
      Sprite UV Randomness Mode {
       bIsDynamic=false
       bIsEditable=false
       bIsStaticSwitch=true
       bIsVisible=false
       type {classStructOrEnum {refPath=&ENiagaraRandomnessMode}}
      }
      Sprite UV Scale {
       bIsDynamic=false
       bIsEditable=false
       bIsStaticSwitch=false
       bIsVisible=false
       type {classStructOrEnum {refPath=&Vector2f}}
      }
      Mesh Scale Mode {
       bIsDynamic=false
       bIsEditable=true
       bIsStaticSwitch=true
       bIsVisible=true
       type {classStructOrEnum {refPath=&ENiagara_SizeScaleMode}}
      }
      Mesh Uniform Scale Min {
       bIsDynamic=false
       bIsEditable=false
       bIsStaticSwitch=false
       bIsVisible=false
       type {classStructOrEnum {refPath=&NiagaraFloat}}
      }
      Mesh Uniform Scale {
       bIsDynamic=false
       bIsEditable=false
       bIsStaticSwitch=false
       bIsVisible=false
       type {classStructOrEnum {refPath=&NiagaraFloat}}
      }
      Mesh Scale Min {
       bIsDynamic=false
       bIsEditable=false
       bIsStaticSwitch=false
       bIsVisible=false
       type {classStructOrEnum {refPath=&Vector3f}}
      }
      Mesh Scale {
       bIsDynamic=false
       bIsEditable=false
       bIsStaticSwitch=false
       bIsVisible=false
       type {classStructOrEnum {refPath=&Vector3f}}
      }
      Mesh Scale Max {
       bIsDynamic=false
       bIsEditable=false
       bIsStaticSwitch=false
       bIsVisible=false
       type {classStructOrEnum {refPath=&Vector3f}}
      }
      Mesh Uniform Scale Max {
       bIsDynamic=false
       bIsEditable=false
       bIsStaticSwitch=false
       bIsVisible=false
       type {classStructOrEnum {refPath=&NiagaraFloat}}
      }
      Mesh Scale Randomness Mode {
       bIsDynamic=false
       bIsEditable=false
       bIsStaticSwitch=true
       bIsVisible=false
       type {classStructOrEnum {refPath=&ENiagaraRandomnessMode}}
      }
      Mesh Renderer Array Visibility Mode {
       bIsDynamic=false
       bIsEditable=true
       bIsStaticSwitch=true
       bIsVisible=true
       type {classStructOrEnum {refPath=&ENiagara_UnsetDirectSetRandom}}
      }
      Mesh Renderer Mesh Index {
       bIsDynamic=false
       bIsEditable=false
       bIsStaticSwitch=false
       bIsVisible=false
       type {classStructOrEnum {refPath=&NiagaraInt32}}
      }
      Wrap Index By Mesh Count {bIsDynamic=false, bIsEditable=false, bIsStaticSwitch=true, bIsVisible=false, type {classStructOrEnum {refPath=&NiagaraBool}}}
      Mesh Renderer Info {
       bIsDynamic=false
       bIsEditable=false
       bIsStaticSwitch=false
       bIsVisible=false
       type {classStructOrEnum {refPath=&NiagaraDataInterfaceMeshRendererInfo}}
      }
      Mesh Renderer Random Seed {
       bIsDynamic=false
       bIsEditable=false
       bIsStaticSwitch=false
       bIsVisible=false
       type {classStructOrEnum {refPath=&NiagaraInt32}}
      }
      Ribbon Width Mode {
       bIsDynamic=false
       bIsEditable=true
       bIsStaticSwitch=true
       bIsVisible=true
       type {classStructOrEnum {refPath=&ENiagara_UnsetDirectSet}}
      }
      Ribbon Width {
       bIsDynamic=false
       bIsEditable=false
       bIsStaticSwitch=false
       bIsVisible=false
       type {classStructOrEnum {refPath=&NiagaraFloat}}
      }
      Ribbon Facing Mode {
       bIsDynamic=false
       bIsEditable=true
       bIsStaticSwitch=true
       bIsVisible=true
       type {classStructOrEnum {refPath=&ENiagara_UnsetDirectSet}}
      }
      Ribbon Facing Vector {
       bIsDynamic=false
       bIsEditable=false
       bIsStaticSwitch=false
       bIsVisible=false
       type {classStructOrEnum {refPath=&Vector3f}}
      }
      Ribbon Twist Mode {
       bIsDynamic=false
       bIsEditable=true
       bIsStaticSwitch=true
       bIsVisible=true
       type {classStructOrEnum {refPath=&ENiagara_UnsetDirectSet}}
      }
      "Ribbon Twist " {
       bIsDynamic=false
       bIsEditable=false
       bIsStaticSwitch=false
       bIsVisible=false
       type {classStructOrEnum {refPath=&NiagaraFloat}}
      }
     ]
     moduleScript {refPath=&InitializeParticle}
    }
    {
     bIsSetParametersModule=true
     enabled=true
     moduleName=SetVariables_#1
     inputs [Particles.RandomDecalYZ {bIsDynamic=true, bIsEditable=true, bIsStaticSwitch=false, bIsVisible=true, type {classStructOrEnum {refPath=&NiagaraFloat}}}]
     moduleScript {refPath=&self:Ground_Crater_0.NiagaraScriptSource_0.NiagaraGraph_0.NiagaraNodeAssignment_3.SetVariables_#1}
    }
    {
     bIsSetParametersModule=true
     enabled=true
     moduleName=SetVariables_#2
     inputs [Particles.DecalSize {bIsDynamic=true, bIsEditable=true, bIsStaticSwitch=false, bIsVisible=true, type {classStructOrEnum {refPath=&Vector3f}}}]
     moduleScript {refPath=&self:Ground_Crater_0.NiagaraScriptSource_0.NiagaraGraph_0.NiagaraNodeAssignment_1.SetVariables_#2}
    }
    {
     bIsSetParametersModule=false
     enabled=true
     moduleName=InitialMeshOrientation
     inputs [
      Mesh Orientation Mode {
       bIsDynamic=false
       bIsEditable=true
       bIsStaticSwitch=true
       bIsVisible=true
       type {classStructOrEnum {refPath=&ENiagara_MeshOrientationOptions}}
      }
      Orientation Quaternion {bIsDynamic=false, bIsEditable=false, bIsStaticSwitch=false, bIsVisible=false, type {classStructOrEnum {refPath=&Quat4f}}}
      Orientation Matrix {
       bIsDynamic=false
       bIsEditable=false
       bIsStaticSwitch=false
       bIsVisible=false
       type {classStructOrEnum {refPath=&NiagaraMatrix}}
      }
      Orientation Vector {bIsDynamic=false, bIsEditable=true, bIsStaticSwitch=false, bIsVisible=true, type {classStructOrEnum {refPath=&Vector3f}}}
      Orientation Coordinate Space {
       bIsDynamic=false
       bIsEditable=true
       bIsStaticSwitch=true
       bIsVisible=true
       type {classStructOrEnum {refPath=&ENiagaraCoordinateSpace}}
      }
      Mesh Axis To Orient {bIsDynamic=false, bIsEditable=true, bIsStaticSwitch=false, bIsVisible=true, type {classStructOrEnum {refPath=&Vector3f}}}
      Random Seed {
       bIsDynamic=false
       bIsEditable=false
       bIsStaticSwitch=false
       bIsVisible=false
       type {classStructOrEnum {refPath=&NiagaraInt32}}
      }
      Use Rotation Vector {bIsDynamic=false, bIsEditable=true, bIsStaticSwitch=false, bIsVisible=true, type {classStructOrEnum {refPath=&NiagaraBool}}}
      Rotation {bIsDynamic=true, bIsEditable=true, bIsStaticSwitch=false, bIsVisible=true, type {classStructOrEnum {refPath=&Vector3f}}}
      Rotation Coordinate Space {
       bIsDynamic=false
       bIsEditable=true
       bIsStaticSwitch=true
       bIsVisible=true
       type {classStructOrEnum {refPath=&ENiagaraMeshTransforms}}
      }
     ]
     moduleScript {refPath=&InitialMeshOrientation}
    }
    {
     bIsSetParametersModule=true
     enabled=true
     moduleName=SetVariables_#5
     inputs [Particles.VisibilityTag {bIsDynamic=true, bIsEditable=true, bIsStaticSwitch=false, bIsVisible=true, type {classStructOrEnum {refPath=&NiagaraInt32}}}]
     moduleScript {refPath=&self:Ground_Crater_0.NiagaraScriptSource_0.NiagaraGraph_0.NiagaraNodeAssignment_0.SetVariables_#5}
    }
    {
     bIsSetParametersModule=false
     enabled=true
     moduleName=InitialDecalOrientation
     inputs [
      Decal Orientation Mode {
       bIsDynamic=false
       bIsEditable=true
       bIsStaticSwitch=true
       bIsVisible=true
       type {classStructOrEnum {refPath=&ENiagara_MeshOrientationOptions}}
      }
      Orientation Matrix {
       bIsDynamic=false
       bIsEditable=false
       bIsStaticSwitch=false
       bIsVisible=false
       type {classStructOrEnum {refPath=&NiagaraMatrix}}
      }
      Orientation Vector {bIsDynamic=false, bIsEditable=true, bIsStaticSwitch=false, bIsVisible=true, type {classStructOrEnum {refPath=&Vector3f}}}
      Orientation Coordinate Space {
       bIsDynamic=false
       bIsEditable=true
       bIsStaticSwitch=true
       bIsVisible=true
       type {classStructOrEnum {refPath=&ENiagaraCoordinateSpace}}
      }
      Decal Axis to Orient {bIsDynamic=false, bIsEditable=true, bIsStaticSwitch=false, bIsVisible=true, type {classStructOrEnum {refPath=&Vector3f}}}
      Random Seed {
       bIsDynamic=false
       bIsEditable=false
       bIsStaticSwitch=false
       bIsVisible=false
       type {classStructOrEnum {refPath=&NiagaraInt32}}
      }
      Use Rotation Vector {bIsDynamic=false, bIsEditable=true, bIsStaticSwitch=false, bIsVisible=true, type {classStructOrEnum {refPath=&NiagaraBool}}}
      Rotation {bIsDynamic=true, bIsEditable=true, bIsStaticSwitch=false, bIsVisible=true, type {classStructOrEnum {refPath=&Vector3f}}}
      Rotation Coordinate Space {
       bIsDynamic=false
       bIsEditable=true
       bIsStaticSwitch=true
       bIsVisible=true
       type {classStructOrEnum {refPath=&ENiagaraDecalTransforms}}
      }
      Orientation Quaternion {bIsDynamic=false, bIsEditable=false, bIsStaticSwitch=false, bIsVisible=false, type {classStructOrEnum {refPath=&Quat4f}}}
     ]
     moduleScript {refPath=&InitialDecalOrientation}
    }
   ]
  }
  particleUpdateScript {
   scriptName=ParticleUpdateScript
   modules [
    {
     bIsSetParametersModule=false
     enabled=true
     moduleName=ParticleState
     inputs [
      Kill Particles When Lifetime Has Elapsed {bIsDynamic=false, bIsEditable=true, bIsStaticSwitch=true, bIsVisible=true, type {classStructOrEnum {refPath=&NiagaraBool}}}
      Loop Particles Lifetime {bIsDynamic=false, bIsEditable=false, bIsStaticSwitch=false, bIsVisible=false, type {classStructOrEnum {refPath=&NiagaraBool}}}
      Let Infinitely Lived Particles Die When Emitter Deactivates {bIsDynamic=false, bIsEditable=false, bIsStaticSwitch=false, bIsVisible=false, type {classStructOrEnum {refPath=&NiagaraBool}}}
      Lifetime {bIsDynamic=false, bIsEditable=true, bIsStaticSwitch=false, bIsVisible=true, type {classStructOrEnum {refPath=&NiagaraFloat}}}
      DeltaTime {bIsDynamic=false, bIsEditable=true, bIsStaticSwitch=false, bIsVisible=true, type {classStructOrEnum {refPath=&NiagaraFloat}}}
     ]
     moduleScript {refPath=&ParticleState}
    }
    {
     bIsSetParametersModule=true
     enabled=true
     moduleName=SetVariables_#3
     inputs [Particles.DecalFade {bIsDynamic=true, bIsEditable=true, bIsStaticSwitch=false, bIsVisible=true, type {classStructOrEnum {refPath=&NiagaraFloat}}}]
     moduleScript {refPath=&self:Ground_Crater_0.NiagaraScriptSource_0.NiagaraGraph_0.NiagaraNodeAssignment_2.SetVariables_#3}
    }
    {
     bIsSetParametersModule=true
     enabled=true
     moduleName=SetVariables_#4
     inputs [Particles.DecalSize {bIsDynamic=true, bIsEditable=true, bIsStaticSwitch=false, bIsVisible=true, type {classStructOrEnum {refPath=&Vector3f}}}]
     moduleScript {refPath=&self:Ground_Crater_0.NiagaraScriptSource_0.NiagaraGraph_0.NiagaraNodeAssignment_4.SetVariables_#4}
    }
    {
     bIsSetParametersModule=false
     enabled=true
     moduleName=SolveForcesAndVelocity
     inputs [
      Clamp Velocity {bIsDynamic=false, bIsEditable=true, bIsStaticSwitch=false, bIsVisible=true, type {classStructOrEnum {refPath=&NiagaraBool}}}
      Speed Limit {bIsDynamic=false, bIsEditable=false, bIsStaticSwitch=false, bIsVisible=true, type {classStructOrEnum {refPath=&NiagaraFloat}}}
      Limit Acceleration {bIsDynamic=false, bIsEditable=true, bIsStaticSwitch=false, bIsVisible=true, type {classStructOrEnum {refPath=&NiagaraBool}}}
      Acceleration Limit {bIsDynamic=false, bIsEditable=false, bIsStaticSwitch=false, bIsVisible=true, type {classStructOrEnum {refPath=&NiagaraFloat}}}
      Manually Enable Rotational Solver {bIsDynamic=false, bIsEditable=true, bIsStaticSwitch=false, bIsVisible=true, type {classStructOrEnum {refPath=&NiagaraBool}}}
      Write to Intrinsic Properties {bIsDynamic=false, bIsEditable=true, bIsStaticSwitch=false, bIsVisible=true, type {classStructOrEnum {refPath=&NiagaraBool}}}
      Write to Presolve Properties {bIsDynamic=false, bIsEditable=true, bIsStaticSwitch=true, bIsVisible=true, type {classStructOrEnum {refPath=&NiagaraBool}}}
      DeltaTime {bIsDynamic=false, bIsEditable=true, bIsStaticSwitch=false, bIsVisible=true, type {classStructOrEnum {refPath=&NiagaraFloat}}}
      Mass {bIsDynamic=false, bIsEditable=true, bIsStaticSwitch=false, bIsVisible=true, type {classStructOrEnum {refPath=&NiagaraFloat}}}
      Position {
       bIsDynamic=false
       bIsEditable=true
       bIsStaticSwitch=false
       bIsVisible=true
       type {classStructOrEnum {refPath=&NiagaraPosition}}
      }
      Velocity {bIsDynamic=false, bIsEditable=true, bIsStaticSwitch=false, bIsVisible=true, type {classStructOrEnum {refPath=&Vector3f}}}
      PreviousVelocity {bIsDynamic=false, bIsEditable=true, bIsStaticSwitch=false, bIsVisible=true, type {classStructOrEnum {refPath=&Vector3f}}}
      Force {bIsDynamic=false, bIsEditable=true, bIsStaticSwitch=false, bIsVisible=true, type {classStructOrEnum {refPath=&Vector3f}}}
      Mesh Orientation {bIsDynamic=false, bIsEditable=false, bIsStaticSwitch=false, bIsVisible=false, type {classStructOrEnum {refPath=&Quat4f}}}
      Rotational Velocity {
       bIsDynamic=false
       bIsEditable=false
       bIsStaticSwitch=false
       bIsVisible=false
       type {classStructOrEnum {refPath=&Vector3f}}
      }
      Rotational Force {
       bIsDynamic=false
       bIsEditable=false
       bIsStaticSwitch=false
       bIsVisible=false
       type {classStructOrEnum {refPath=&Vector3f}}
      }
      Rotational Inertia {
       bIsDynamic=false
       bIsEditable=false
       bIsStaticSwitch=false
       bIsVisible=false
       type {classStructOrEnum {refPath=&NiagaraFloat}}
      }
     ]
     moduleScript {refPath=&SolveForcesAndVelocity}
    }
    {
     bIsSetParametersModule=false
     enabled=true
     moduleName=ScaleColor
     inputs [
      Scale Mode {
       bIsDynamic=false
       bIsEditable=true
       bIsStaticSwitch=true
       bIsVisible=true
       type {classStructOrEnum {refPath=&ENiagaraScaleColorMode}}
      }
      Linear Color Curve {
       bIsDynamic=false
       bIsEditable=false
       bIsStaticSwitch=false
       bIsVisible=false
       type {classStructOrEnum {refPath=&NiagaraDataInterfaceColorCurve}}
      }
      ScaleRGB {bIsDynamic=false, bIsEditable=true, bIsStaticSwitch=false, bIsVisible=true, type {classStructOrEnum {refPath=&NiagaraBool}}}
      Scale RGB {bIsDynamic=false, bIsEditable=true, bIsStaticSwitch=false, bIsVisible=true, type {classStructOrEnum {refPath=&Vector3f}}}
      ScaleRGBA {bIsDynamic=false, bIsEditable=false, bIsStaticSwitch=false, bIsVisible=false, type {classStructOrEnum {refPath=&NiagaraBool}}}
      Scale RGBA {
       bIsDynamic=false
       bIsEditable=false
       bIsStaticSwitch=false
       bIsVisible=false
       type {classStructOrEnum {refPath=&Vector4f}}
      }
      Curve Index {
       bIsDynamic=false
       bIsEditable=false
       bIsStaticSwitch=false
       bIsVisible=false
       type {classStructOrEnum {refPath=&NiagaraFloat}}
      }
      ScaleA {bIsDynamic=false, bIsEditable=true, bIsStaticSwitch=false, bIsVisible=true, type {classStructOrEnum {refPath=&NiagaraBool}}}
      Scale Alpha {bIsDynamic=true, bIsEditable=true, bIsStaticSwitch=false, bIsVisible=true, type {classStructOrEnum {refPath=&NiagaraFloat}}}
      Color Value To Scale {
       bIsDynamic=false
       bIsEditable=true
       bIsStaticSwitch=false
       bIsVisible=true
       type {classStructOrEnum {refPath=&LinearColor}}
      }
     ]
     moduleScript {refPath=&ScaleColor}
    }
   ]
  }
  rendererClasses [{refPath=&NiagaraDecalRendererProperties}]
  renderers [{rendererIndex=0, rendererClass {refPath=&NiagaraDecalRendererProperties}}]
 }
 Puffs {
  emitterSpawnScript {scriptName=EmitterSpawnScript, modules []}
  emitterUpdateScript {
   scriptName=EmitterUpdateScript
   modules [
    {
     bIsSetParametersModule=false
     enabled=true
     moduleName=EmitterState
     inputs [
      Life Cycle Mode {
       bIsDynamic=false
       bIsEditable=true
       bIsStaticSwitch=true
       bIsVisible=true
       type {classStructOrEnum {refPath=&ENiagaraEmitterLifeCycleMode}}
      }
      Inactive Response {
       bIsDynamic=false
       bIsEditable=true
       bIsStaticSwitch=true
       bIsVisible=true
       type {classStructOrEnum {refPath=&ENiagaraInactiveMode}}
      }
      Loop Behavior {
       bIsDynamic=false
       bIsEditable=true
       bIsStaticSwitch=true
       bIsVisible=true
       type {classStructOrEnum {refPath=&ENiagara_EmitterStateOptions}}
      }
      Loop Duration Mode {
       bIsDynamic=false
       bIsEditable=true
       bIsStaticSwitch=true
       bIsVisible=true
       type {classStructOrEnum {refPath=&ENiagara_InfiniteLoopDuration}}
      }
      Loop Count {
       bIsDynamic=false
       bIsEditable=false
       bIsStaticSwitch=false
       bIsVisible=false
       type {classStructOrEnum {refPath=&NiagaraInt32}}
      }
      Loop Duration {bIsDynamic=false, bIsEditable=true, bIsStaticSwitch=false, bIsVisible=true, type {classStructOrEnum {refPath=&NiagaraFloat}}}
      Recalculate Duration Each Loop {bIsDynamic=false, bIsEditable=false, bIsStaticSwitch=false, bIsVisible=false, type {classStructOrEnum {refPath=&NiagaraBool}}}
      UseLoopDelay {bIsDynamic=false, bIsEditable=true, bIsStaticSwitch=false, bIsVisible=true, type {classStructOrEnum {refPath=&NiagaraBool}}}
      Loop Delay {bIsDynamic=false, bIsEditable=false, bIsStaticSwitch=false, bIsVisible=true, type {classStructOrEnum {refPath=&NiagaraFloat}}}
      Delay First Loop Only {bIsDynamic=false, bIsEditable=false, bIsStaticSwitch=false, bIsVisible=false, type {classStructOrEnum {refPath=&NiagaraBool}}}
      Scalability Mode {
       bIsDynamic=false
       bIsEditable=true
       bIsStaticSwitch=true
       bIsVisible=true
       type {classStructOrEnum {refPath=&ENiagaraEmitterScalabilityMode_Limited}}
      }
      Enable Distance Culling {bIsDynamic=false, bIsEditable=false, bIsStaticSwitch=true, bIsVisible=false, type {classStructOrEnum {refPath=&NiagaraBool}}}
      MinDistance {
       bIsDynamic=false
       bIsEditable=false
       bIsStaticSwitch=false
       bIsVisible=false
       type {classStructOrEnum {refPath=&NiagaraFloat}}
      }
      Min Distance Response {
       bIsDynamic=false
       bIsEditable=false
       bIsStaticSwitch=false
       bIsVisible=false
       type {classStructOrEnum {refPath=&ENiagaraExecutionStateManagement}}
      }
      MaxDistance {
       bIsDynamic=false
       bIsEditable=false
       bIsStaticSwitch=false
       bIsVisible=false
       type {classStructOrEnum {refPath=&NiagaraFloat}}
      }
      Max Distance Response {
       bIsDynamic=false
       bIsEditable=false
       bIsStaticSwitch=false
       bIsVisible=false
       type {classStructOrEnum {refPath=&ENiagaraExecutionStateManagement}}
      }
      Scale Spawn Count {bIsDynamic=false, bIsEditable=false, bIsStaticSwitch=true, bIsVisible=false, type {classStructOrEnum {refPath=&NiagaraBool}}}
      Spawn Count Scale {
       bIsDynamic=false
       bIsEditable=false
       bIsStaticSwitch=false
       bIsVisible=false
       type {classStructOrEnum {refPath=&NiagaraFloat}}
      }
      Spawn Count Scale By Distance {
       bIsDynamic=false
       bIsEditable=false
       bIsStaticSwitch=false
       bIsVisible=false
       type {classStructOrEnum {refPath=&NiagaraDataInterfaceCurve}}
      }
      Enable Visibility Culling {bIsDynamic=false, bIsEditable=false, bIsStaticSwitch=true, bIsVisible=false, type {classStructOrEnum {refPath=&NiagaraBool}}}
      Visibility Cull Response {
       bIsDynamic=false
       bIsEditable=false
       bIsStaticSwitch=false
       bIsVisible=false
       type {classStructOrEnum {refPath=&ENiagaraExecutionStateManagement}}
      }
      Visibility Cull Delay {
       bIsDynamic=false
       bIsEditable=false
       bIsStaticSwitch=false
       bIsVisible=false
       type {classStructOrEnum {refPath=&NiagaraFloat}}
      }
      Reset Age On Awaken {bIsDynamic=false, bIsEditable=false, bIsStaticSwitch=false, bIsVisible=false, type {classStructOrEnum {refPath=&NiagaraBool}}}
     ]
     moduleScript {refPath=&EmitterState}
    }
    {
     bIsSetParametersModule=false
     enabled=true
     moduleName=SpawnBurst_Instantaneous
     inputs [
      Spawn Count {bIsDynamic=false, bIsEditable=true, bIsStaticSwitch=false, bIsVisible=true, type {classStructOrEnum {refPath=&NiagaraInt32}}}
      Spawn Time {bIsDynamic=false, bIsEditable=true, bIsStaticSwitch=false, bIsVisible=true, type {classStructOrEnum {refPath=&NiagaraFloat}}}
      Use Spawn Probability {bIsDynamic=false, bIsEditable=true, bIsStaticSwitch=false, bIsVisible=true, type {classStructOrEnum {refPath=&NiagaraBool}}}
      Spawn Probability {bIsDynamic=false, bIsEditable=false, bIsStaticSwitch=false, bIsVisible=true, type {classStructOrEnum {refPath=&NiagaraFloat}}}
      UseLoopCountLimit {bIsDynamic=false, bIsEditable=true, bIsStaticSwitch=false, bIsVisible=true, type {classStructOrEnum {refPath=&NiagaraBool}}}
      Loop Count Limit {bIsDynamic=false, bIsEditable=false, bIsStaticSwitch=false, bIsVisible=true, type {classStructOrEnum {refPath=&NiagaraInt32}}}
      Spawn Group {bIsDynamic=false, bIsEditable=true, bIsStaticSwitch=false, bIsVisible=true, type {classStructOrEnum {refPath=&NiagaraInt32}}}
      Age {bIsDynamic=false, bIsEditable=true, bIsStaticSwitch=false, bIsVisible=true, type {classStructOrEnum {refPath=&NiagaraFloat}}}
     ]
     moduleScript {refPath=&SpawnBurst_Instantaneous}
    }
   ]
  }
  particleSpawnScript {
   scriptName=ParticleSpawnScript
   modules [
    {
     bIsSetParametersModule=false
     enabled=true
     moduleName=InitializeParticle
     inputs [
      Lifetime Mode {
       bIsDynamic=false
       bIsEditable=true
       bIsStaticSwitch=true
       bIsVisible=true
       type {classStructOrEnum {refPath=&ENiagara_LifetimeMode}}
      }
      Lifetime Min {bIsDynamic=false, bIsEditable=true, bIsStaticSwitch=false, bIsVisible=true, type {classStructOrEnum {refPath=&NiagaraFloat}}}
      Lifetime {
       bIsDynamic=false
       bIsEditable=false
       bIsStaticSwitch=false
       bIsVisible=false
       type {classStructOrEnum {refPath=&NiagaraFloat}}
      }
      Lifetime Max {bIsDynamic=false, bIsEditable=true, bIsStaticSwitch=false, bIsVisible=true, type {classStructOrEnum {refPath=&NiagaraFloat}}}
      Lifetime Randomness Mode {
       bIsDynamic=false
       bIsEditable=true
       bIsStaticSwitch=true
       bIsVisible=true
       type {classStructOrEnum {refPath=&ENiagaraRandomnessMode}}
      }
      Color Mode {
       bIsDynamic=false
       bIsEditable=true
       bIsStaticSwitch=true
       bIsVisible=true
       type {classStructOrEnum {refPath=&ENiagara_ColorInitializationMode}}
      }
      Color Minimum {
       bIsDynamic=false
       bIsEditable=false
       bIsStaticSwitch=false
       bIsVisible=false
       type {classStructOrEnum {refPath=&LinearColor}}
      }
      Color {
       bIsDynamic=true
       bIsEditable=true
       bIsStaticSwitch=false
       bIsVisible=true
       type {classStructOrEnum {refPath=&LinearColor}}
      }
      AdjustHue {bIsDynamic=false, bIsEditable=false, bIsStaticSwitch=false, bIsVisible=false, type {classStructOrEnum {refPath=&NiagaraBool}}}
      Hue Shift Range {
       bIsDynamic=false
       bIsEditable=false
       bIsStaticSwitch=false
       bIsVisible=false
       type {classStructOrEnum {refPath=&Vector2f}}
      }
      Color Maximum {
       bIsDynamic=false
       bIsEditable=false
       bIsStaticSwitch=false
       bIsVisible=false
       type {classStructOrEnum {refPath=&LinearColor}}
      }
      AdjustSaturation {bIsDynamic=false, bIsEditable=false, bIsStaticSwitch=false, bIsVisible=false, type {classStructOrEnum {refPath=&NiagaraBool}}}
      Saturation Range {
       bIsDynamic=false
       bIsEditable=false
       bIsStaticSwitch=false
       bIsVisible=false
       type {classStructOrEnum {refPath=&Vector2f}}
      }
      AdjustValue {bIsDynamic=false, bIsEditable=false, bIsStaticSwitch=false, bIsVisible=false, type {classStructOrEnum {refPath=&NiagaraBool}}}
      Value Range {
       bIsDynamic=false
       bIsEditable=false
       bIsStaticSwitch=false
       bIsVisible=false
       type {classStructOrEnum {refPath=&Vector2f}}
      }
      AdjustAlpha {bIsDynamic=false, bIsEditable=false, bIsStaticSwitch=false, bIsVisible=false, type {classStructOrEnum {refPath=&NiagaraBool}}}
      Alpha Scale Range {
       bIsDynamic=false
       bIsEditable=false
       bIsStaticSwitch=false
       bIsVisible=false
       type {classStructOrEnum {refPath=&Vector2f}}
      }
      Color Channel Mode {
       bIsDynamic=false
       bIsEditable=false
       bIsStaticSwitch=true
       bIsVisible=false
       type {classStructOrEnum {refPath=&ENiagaraChannelCorrelation}}
      }
      Color Randomness Mode {
       bIsDynamic=false
       bIsEditable=false
       bIsStaticSwitch=true
       bIsVisible=false
       type {classStructOrEnum {refPath=&ENiagaraRandomnessMode}}
      }
      Position Mode {
       bIsDynamic=false
       bIsEditable=true
       bIsStaticSwitch=true
       bIsVisible=true
       type {classStructOrEnum {refPath=&ENiagara_PositionInitializationMode}}
      }
      Position {
       bIsDynamic=false
       bIsEditable=false
       bIsStaticSwitch=false
       bIsVisible=false
       type {classStructOrEnum {refPath=&NiagaraPosition}}
      }
      UsePositionOffset {bIsDynamic=false, bIsEditable=true, bIsStaticSwitch=false, bIsVisible=true, type {classStructOrEnum {refPath=&NiagaraBool}}}
      Position Offset {bIsDynamic=false, bIsEditable=false, bIsStaticSwitch=false, bIsVisible=true, type {classStructOrEnum {refPath=&Vector3f}}}
      Position Offset Coordinate Space {
       bIsDynamic=false
       bIsEditable=false
       bIsStaticSwitch=false
       bIsVisible=false
       type {classStructOrEnum {refPath=&ENiagaraCoordinateSpace}}
      }
      Mass Mode {
       bIsDynamic=false
       bIsEditable=true
       bIsStaticSwitch=true
       bIsVisible=true
       type {classStructOrEnum {refPath=&ENiagara_MassInitializationMode}}
      }
      Mass Min {
       bIsDynamic=false
       bIsEditable=false
       bIsStaticSwitch=false
       bIsVisible=false
       type {classStructOrEnum {refPath=&NiagaraFloat}}
      }
      Mass {
       bIsDynamic=false
       bIsEditable=false
       bIsStaticSwitch=false
       bIsVisible=false
       type {classStructOrEnum {refPath=&NiagaraFloat}}
      }
      Mass Max {
       bIsDynamic=false
       bIsEditable=false
       bIsStaticSwitch=false
       bIsVisible=false
       type {classStructOrEnum {refPath=&NiagaraFloat}}
      }
      Mass Randomness Mode {
       bIsDynamic=false
       bIsEditable=false
       bIsStaticSwitch=true
       bIsVisible=false
       type {classStructOrEnum {refPath=&ENiagaraRandomnessMode}}
      }
      Write Material Random {bIsDynamic=false, bIsEditable=true, bIsStaticSwitch=false, bIsVisible=true, type {classStructOrEnum {refPath=&NiagaraBool}}}
      Material Random {bIsDynamic=false, bIsEditable=true, bIsStaticSwitch=false, bIsVisible=true, type {classStructOrEnum {refPath=&NiagaraFloat}}}
      Sprite Size Mode {
       bIsDynamic=false
       bIsEditable=true
       bIsStaticSwitch=true
       bIsVisible=true
       type {classStructOrEnum {refPath=&ENiagara_SizeScaleMode}}
      }
      Sprite Size {
       bIsDynamic=false
       bIsEditable=false
       bIsStaticSwitch=false
       bIsVisible=false
       type {classStructOrEnum {refPath=&Vector2f}}
      }
      Uniform Sprite Size {
       bIsDynamic=false
       bIsEditable=false
       bIsStaticSwitch=false
       bIsVisible=false
       type {classStructOrEnum {refPath=&NiagaraFloat}}
      }
      Uniform Sprite Size Min {
       bIsDynamic=false
       bIsEditable=false
       bIsStaticSwitch=false
       bIsVisible=false
       type {classStructOrEnum {refPath=&NiagaraFloat}}
      }
      Sprite Size Min {
       bIsDynamic=false
       bIsEditable=false
       bIsStaticSwitch=false
       bIsVisible=false
       type {classStructOrEnum {refPath=&Vector2f}}
      }
      Sprite Size Max {
       bIsDynamic=false
       bIsEditable=false
       bIsStaticSwitch=false
       bIsVisible=false
       type {classStructOrEnum {refPath=&Vector2f}}
      }
      Uniform Sprite Size Max {
       bIsDynamic=false
       bIsEditable=false
       bIsStaticSwitch=false
       bIsVisible=false
       type {classStructOrEnum {refPath=&NiagaraFloat}}
      }
      Sprite Size Randomness Mode {
       bIsDynamic=false
       bIsEditable=false
       bIsStaticSwitch=true
       bIsVisible=false
       type {classStructOrEnum {refPath=&ENiagaraRandomnessMode}}
      }
      Sprite Rotation Mode {
       bIsDynamic=false
       bIsEditable=true
       bIsStaticSwitch=true
       bIsVisible=true
       type {classStructOrEnum {refPath=&ENiagara_SpriteRotationMode}}
      }
      Sprite Rotation Angle Min {
       bIsDynamic=false
       bIsEditable=false
       bIsStaticSwitch=false
       bIsVisible=false
       type {classStructOrEnum {refPath=&NiagaraFloat}}
      }
      Sprite Rotation Angle {
       bIsDynamic=false
       bIsEditable=false
       bIsStaticSwitch=false
       bIsVisible=false
       type {classStructOrEnum {refPath=&NiagaraFloat}}
      }
      Sprite Rotation Angle Max {
       bIsDynamic=false
       bIsEditable=false
       bIsStaticSwitch=false
       bIsVisible=false
       type {classStructOrEnum {refPath=&NiagaraFloat}}
      }
      Sprite Rotation Randomness Mode {
       bIsDynamic=false
       bIsEditable=false
       bIsStaticSwitch=true
       bIsVisible=false
       type {classStructOrEnum {refPath=&ENiagaraRandomnessMode}}
      }
      Sprite UV Mode {
       bIsDynamic=false
       bIsEditable=true
       bIsStaticSwitch=true
       bIsVisible=true
       type {classStructOrEnum {refPath=&ENiagara_UVFlippingMode}}
      }
      Sprite UV Randomness Mode {
       bIsDynamic=false
       bIsEditable=false
       bIsStaticSwitch=true
       bIsVisible=false
       type {classStructOrEnum {refPath=&ENiagaraRandomnessMode}}
      }
      Sprite UV Scale {
       bIsDynamic=false
       bIsEditable=false
       bIsStaticSwitch=false
       bIsVisible=false
       type {classStructOrEnum {refPath=&Vector2f}}
      }
      Mesh Scale Mode {
       bIsDynamic=false
       bIsEditable=true
       bIsStaticSwitch=true
       bIsVisible=true
       type {classStructOrEnum {refPath=&ENiagara_SizeScaleMode}}
      }
      Mesh Uniform Scale Min {bIsDynamic=false, bIsEditable=true, bIsStaticSwitch=false, bIsVisible=true, type {classStructOrEnum {refPath=&NiagaraFloat}}}
      Mesh Uniform Scale {bIsDynamic=true, bIsEditable=false, bIsStaticSwitch=false, bIsVisible=false, type {classStructOrEnum {refPath=&NiagaraFloat}}}
      Mesh Scale Min {
       bIsDynamic=false
       bIsEditable=false
       bIsStaticSwitch=false
       bIsVisible=false
       type {classStructOrEnum {refPath=&Vector3f}}
      }
      Mesh Scale {
       bIsDynamic=false
       bIsEditable=false
       bIsStaticSwitch=false
       bIsVisible=false
       type {classStructOrEnum {refPath=&Vector3f}}
      }
      Mesh Scale Max {
       bIsDynamic=false
       bIsEditable=false
       bIsStaticSwitch=false
       bIsVisible=false
       type {classStructOrEnum {refPath=&Vector3f}}
      }
      Mesh Uniform Scale Max {bIsDynamic=false, bIsEditable=true, bIsStaticSwitch=false, bIsVisible=true, type {classStructOrEnum {refPath=&NiagaraFloat}}}
      Mesh Scale Randomness Mode {
       bIsDynamic=false
       bIsEditable=true
       bIsStaticSwitch=true
       bIsVisible=true
       type {classStructOrEnum {refPath=&ENiagaraRandomnessMode}}
      }
      Mesh Renderer Array Visibility Mode {
       bIsDynamic=false
       bIsEditable=true
       bIsStaticSwitch=true
       bIsVisible=true
       type {classStructOrEnum {refPath=&ENiagara_UnsetDirectSetRandom}}
      }
      Mesh Renderer Mesh Index {
       bIsDynamic=false
       bIsEditable=false
       bIsStaticSwitch=false
       bIsVisible=false
       type {classStructOrEnum {refPath=&NiagaraInt32}}
      }
      Wrap Index By Mesh Count {bIsDynamic=false, bIsEditable=false, bIsStaticSwitch=true, bIsVisible=false, type {classStructOrEnum {refPath=&NiagaraBool}}}
      Mesh Renderer Info {
       bIsDynamic=false
       bIsEditable=false
       bIsStaticSwitch=false
       bIsVisible=false
       type {classStructOrEnum {refPath=&NiagaraDataInterfaceMeshRendererInfo}}
      }
      Mesh Renderer Random Seed {
       bIsDynamic=false
       bIsEditable=false
       bIsStaticSwitch=false
       bIsVisible=false
       type {classStructOrEnum {refPath=&NiagaraInt32}}
      }
      Ribbon Width Mode {
       bIsDynamic=false
       bIsEditable=true
       bIsStaticSwitch=true
       bIsVisible=true
       type {classStructOrEnum {refPath=&ENiagara_UnsetDirectSet}}
      }
      Ribbon Width {
       bIsDynamic=false
       bIsEditable=false
       bIsStaticSwitch=false
       bIsVisible=false
       type {classStructOrEnum {refPath=&NiagaraFloat}}
      }
      Ribbon Facing Mode {
       bIsDynamic=false
       bIsEditable=true
       bIsStaticSwitch=true
       bIsVisible=true
       type {classStructOrEnum {refPath=&ENiagara_UnsetDirectSet}}
      }
      Ribbon Facing Vector {
       bIsDynamic=false
       bIsEditable=false
       bIsStaticSwitch=false
       bIsVisible=false
       type {classStructOrEnum {refPath=&Vector3f}}
      }
      Ribbon Twist Mode {
       bIsDynamic=false
       bIsEditable=true
       bIsStaticSwitch=true
       bIsVisible=true
       type {classStructOrEnum {refPath=&ENiagara_UnsetDirectSet}}
      }
      "Ribbon Twist " {
       bIsDynamic=false
       bIsEditable=false
       bIsStaticSwitch=false
       bIsVisible=false
       type {classStructOrEnum {refPath=&NiagaraFloat}}
      }
     ]
     moduleScript {refPath=&InitializeParticle}
    }
    {
     bIsSetParametersModule=false
     enabled=true
     moduleName=ShapeLocation
     inputs [
      Shape Primitive {
       bIsDynamic=false
       bIsEditable=true
       bIsStaticSwitch=true
       bIsVisible=true
       type {classStructOrEnum {refPath=&ENiagara_LocationShapes}}
      }
      Box / Plane Mode {
       bIsDynamic=false
       bIsEditable=false
       bIsStaticSwitch=true
       bIsVisible=false
       type {classStructOrEnum {refPath=&ENiagara_BoxPlaneMode}}
      }
      Cone Mode {
       bIsDynamic=false
       bIsEditable=false
       bIsStaticSwitch=true
       bIsVisible=false
       type {classStructOrEnum {refPath=&ENiagara_ConeMode}}
      }
      Cylinder Height {
       bIsDynamic=false
       bIsEditable=false
       bIsStaticSwitch=false
       bIsVisible=false
       type {classStructOrEnum {refPath=&NiagaraFloat}}
      }
      Sphere Radius {
       bIsDynamic=false
       bIsEditable=false
       bIsStaticSwitch=false
       bIsVisible=false
       type {classStructOrEnum {refPath=&NiagaraFloat}}
      }
      Ring / Disc Mode {
       bIsDynamic=false
       bIsEditable=true
       bIsStaticSwitch=true
       bIsVisible=true
       type {classStructOrEnum {refPath=&ENiagaraRingDiscMode}}
      }
      Torus Mode {
       bIsDynamic=false
       bIsEditable=false
       bIsStaticSwitch=true
       bIsVisible=false
       type {classStructOrEnum {refPath=&ENiagaraShapeTorusMode}}
      }
      Knot Ring Loops {
       bIsDynamic=false
       bIsEditable=false
       bIsStaticSwitch=false
       bIsVisible=false
       type {classStructOrEnum {refPath=&NiagaraInt32}}
      }
      Knot Helix Loops {
       bIsDynamic=false
       bIsEditable=false
       bIsStaticSwitch=false
       bIsVisible=false
       type {classStructOrEnum {refPath=&NiagaraInt32}}
      }
      Enable Torus Knot Radius {bIsDynamic=false, bIsEditable=false, bIsStaticSwitch=false, bIsVisible=false, type {classStructOrEnum {refPath=&NiagaraBool}}}
      Torus Knot Tube Radius {
       bIsDynamic=false
       bIsEditable=false
       bIsStaticSwitch=false
       bIsVisible=false
       type {classStructOrEnum {refPath=&NiagaraFloat}}
      }
      Cylinder Radius {
       bIsDynamic=false
       bIsEditable=false
       bIsStaticSwitch=false
       bIsVisible=false
       type {classStructOrEnum {refPath=&NiagaraFloat}}
      }
      Cylinder Height Midpoint {
       bIsDynamic=false
       bIsEditable=false
       bIsStaticSwitch=false
       bIsVisible=false
       type {classStructOrEnum {refPath=&NiagaraFloat}}
      }
      Ring Radius {bIsDynamic=false, bIsEditable=true, bIsStaticSwitch=false, bIsVisible=true, type {classStructOrEnum {refPath=&NiagaraFloat}}}
      Box Size {
       bIsDynamic=false
       bIsEditable=false
       bIsStaticSwitch=false
       bIsVisible=false
       type {classStructOrEnum {refPath=&Vector3f}}
      }
      Box Midpoint {
       bIsDynamic=false
       bIsEditable=false
       bIsStaticSwitch=false
       bIsVisible=false
       type {classStructOrEnum {refPath=&Vector3f}}
      }
      Cone Length {
       bIsDynamic=false
       bIsEditable=false
       bIsStaticSwitch=false
       bIsVisible=false
       type {classStructOrEnum {refPath=&NiagaraFloat}}
      }
      Plane Size {
       bIsDynamic=false
       bIsEditable=false
       bIsStaticSwitch=false
       bIsVisible=false
       type {classStructOrEnum {refPath=&Vector2f}}
      }
      Wedge Horizontal Angle {
       bIsDynamic=false
       bIsEditable=false
       bIsStaticSwitch=false
       bIsVisible=false
       type {classStructOrEnum {refPath=&NiagaraFloat}}
      }
      Large Radius {
       bIsDynamic=false
       bIsEditable=false
       bIsStaticSwitch=false
       bIsVisible=false
       type {classStructOrEnum {refPath=&NiagaraFloat}}
      }
      Plane Midpoint {
       bIsDynamic=false
       bIsEditable=false
       bIsStaticSwitch=false
       bIsVisible=false
       type {classStructOrEnum {refPath=&Vector2f}}
      }
      Wedge Vertical Angle {
       bIsDynamic=false
       bIsEditable=false
       bIsStaticSwitch=false
       bIsVisible=false
       type {classStructOrEnum {refPath=&NiagaraFloat}}
      }
      Cone Angle {
       bIsDynamic=false
       bIsEditable=false
       bIsStaticSwitch=false
       bIsVisible=false
       type {classStructOrEnum {refPath=&NiagaraFloat}}
      }
      Handle Radius {
       bIsDynamic=false
       bIsEditable=false
       bIsStaticSwitch=false
       bIsVisible=false
       type {classStructOrEnum {refPath=&NiagaraFloat}}
      }
      Cone Inner Angle {
       bIsDynamic=false
       bIsEditable=false
       bIsStaticSwitch=false
       bIsVisible=false
       type {classStructOrEnum {refPath=&NiagaraFloat}}
      }
      Cone Radial Angle {
       bIsDynamic=false
       bIsEditable=false
       bIsStaticSwitch=false
       bIsVisible=false
       type {classStructOrEnum {refPath=&NiagaraFloat}}
      }
      Cone Angle Type {
       bIsDynamic=false
       bIsEditable=false
       bIsStaticSwitch=true
       bIsVisible=false
       type {classStructOrEnum {refPath=&ENiagara_AngleInput}}
      }
      Flatten Endcaps {bIsDynamic=false, bIsEditable=false, bIsStaticSwitch=true, bIsVisible=false, type {classStructOrEnum {refPath=&NiagaraBool}}}
      Apply To Particle Position {
       bIsDynamic=false
       bIsEditable=true
       bIsStaticSwitch=true
       bIsVisible=true
       type {classStructOrEnum {refPath=&ENiagara_AttributeSamplingApplyOutput}}
      }
      Cone Surface Distribution {
       bIsDynamic=false
       bIsEditable=false
       bIsStaticSwitch=false
       bIsVisible=false
       type {classStructOrEnum {refPath=&NiagaraFloat}}
      }
      Ring / Disc Distribution Mode {
       bIsDynamic=false
       bIsEditable=true
       bIsStaticSwitch=true
       bIsVisible=true
       type {classStructOrEnum {refPath=&ENiagaraSphereDistributionMode}}
      }
      Sphere Distribution {
       bIsDynamic=false
       bIsEditable=false
       bIsStaticSwitch=true
       bIsVisible=false
       type {classStructOrEnum {refPath=&ENiagaraSphereDistributionMode}}
      }
      Torus Distribution Mode {
       bIsDynamic=false
       bIsEditable=false
       bIsStaticSwitch=true
       bIsVisible=false
       type {classStructOrEnum {refPath=&ENiagaraTorusDistributionMode}}
      }
      Sphere Surface Distribution {
       bIsDynamic=false
       bIsEditable=false
       bIsStaticSwitch=false
       bIsVisible=false
       type {classStructOrEnum {refPath=&NiagaraFloat}}
      }
      Hemisphere Distribution {
       bIsDynamic=false
       bIsEditable=false
       bIsStaticSwitch=false
       bIsVisible=false
       type {classStructOrEnum {refPath=&Vector2f}}
      }
      Cylinder Distribution {
       bIsDynamic=false
       bIsEditable=false
       bIsStaticSwitch=true
       bIsVisible=false
       type {classStructOrEnum {refPath=&ENiagara_CylinderMode}}
      }
      Surface Only {bIsDynamic=false, bIsEditable=false, bIsStaticSwitch=false, bIsVisible=false, type {classStructOrEnum {refPath=&NiagaraBool}}}
      Surface Only Band Thickness {
       bIsDynamic=false
       bIsEditable=false
       bIsStaticSwitch=false
       bIsVisible=false
       type {classStructOrEnum {refPath=&NiagaraFloat}}
      }
      Use Endcaps In Surface Only Mode {bIsDynamic=false, bIsEditable=false, bIsStaticSwitch=true, bIsVisible=false, type {classStructOrEnum {refPath=&NiagaraBool}}}
      Hemicircle X {bIsDynamic=false, bIsEditable=false, bIsStaticSwitch=true, bIsVisible=false, type {classStructOrEnum {refPath=&NiagaraBool}}}
      Hemicircle Y {bIsDynamic=false, bIsEditable=false, bIsStaticSwitch=true, bIsVisible=false, type {classStructOrEnum {refPath=&NiagaraBool}}}
      Disc Coverage {bIsDynamic=false, bIsEditable=true, bIsStaticSwitch=false, bIsVisible=true, type {classStructOrEnum {refPath=&NiagaraFloat}}}
      Hemicircle Internal Cap {bIsDynamic=false, bIsEditable=false, bIsStaticSwitch=true, bIsVisible=false, type {classStructOrEnum {refPath=&NiagaraBool}}}
      U Position {bIsDynamic=true, bIsEditable=false, bIsStaticSwitch=false, bIsVisible=false, type {classStructOrEnum {refPath=&NiagaraFloat}}}
      Enable Lathe Profile {bIsDynamic=false, bIsEditable=false, bIsStaticSwitch=true, bIsVisible=false, type {classStructOrEnum {refPath=&NiagaraBool}}}
      Lathe Profile {
       bIsDynamic=false
       bIsEditable=false
       bIsStaticSwitch=false
       bIsVisible=false
       type {classStructOrEnum {refPath=&NiagaraDataInterfaceCurve}}
      }
      Surface Distribution {
       bIsDynamic=false
       bIsEditable=false
       bIsStaticSwitch=false
       bIsVisible=false
       type {classStructOrEnum {refPath=&NiagaraFloat}}
      }
      Surface Expansion Mode {
       bIsDynamic=false
       bIsEditable=false
       bIsStaticSwitch=false
       bIsVisible=false
       type {classStructOrEnum {refPath=&ENiagaraExpansionMode}}
      }
      U Distribution {bIsDynamic=false, bIsEditable=true, bIsStaticSwitch=false, bIsVisible=true, type {classStructOrEnum {refPath=&NiagaraFloat}}}
      V Position {
       bIsDynamic=false
       bIsEditable=false
       bIsStaticSwitch=false
       bIsVisible=false
       type {classStructOrEnum {refPath=&NiagaraFloat}}
      }
      Radius Position {
       bIsDynamic=false
       bIsEditable=false
       bIsStaticSwitch=false
       bIsVisible=false
       type {classStructOrEnum {refPath=&NiagaraFloat}}
      }
      Sphere Radius Position {
       bIsDynamic=false
       bIsEditable=false
       bIsStaticSwitch=false
       bIsVisible=false
       type {classStructOrEnum {refPath=&NiagaraFloat}}
      }
      V Distribution {
       bIsDynamic=false
       bIsEditable=false
       bIsStaticSwitch=false
       bIsVisible=false
       type {classStructOrEnum {refPath=&NiagaraFloat}}
      }
      Evaluation Type {
       bIsDynamic=false
       bIsEditable=false
       bIsStaticSwitch=true
       bIsVisible=false
       type {classStructOrEnum {refPath=&ENiagaraRandomnessEvaluation}}
      }
      Uniform Distribution {
       bIsDynamic=false
       bIsEditable=false
       bIsStaticSwitch=false
       bIsVisible=false
       type {classStructOrEnum {refPath=&NiagaraFloat}}
      }
      Uniform Spiral Amount {
       bIsDynamic=false
       bIsEditable=false
       bIsStaticSwitch=false
       bIsVisible=false
       type {classStructOrEnum {refPath=&NiagaraFloat}}
      }
      Uniform Spiral Falloff {
       bIsDynamic=false
       bIsEditable=false
       bIsStaticSwitch=false
       bIsVisible=false
       type {classStructOrEnum {refPath=&NiagaraFloat}}
      }
      Uniform Count {
       bIsDynamic=false
       bIsEditable=false
       bIsStaticSwitch=false
       bIsVisible=false
       type {classStructOrEnum {refPath=&NiagaraInt32}}
      }
      Uniform Particle Index {
       bIsDynamic=false
       bIsEditable=false
       bIsStaticSwitch=false
       bIsVisible=false
       type {classStructOrEnum {refPath=&NiagaraInt32}}
      }
      Hemisphere Angle Type {
       bIsDynamic=false
       bIsEditable=false
       bIsStaticSwitch=true
       bIsVisible=false
       type {classStructOrEnum {refPath=&ENiagara_AngleInput}}
      }
      Uniform Endpoint Offset {
       bIsDynamic=false
       bIsEditable=false
       bIsStaticSwitch=false
       bIsVisible=false
       type {classStructOrEnum {refPath=&NiagaraFloat}}
      }
      MaskSpawnGroup {bIsDynamic=false, bIsEditable=true, bIsStaticSwitch=false, bIsVisible=true, type {classStructOrEnum {refPath=&NiagaraBool}}}
      Spawn Group Mask {bIsDynamic=false, bIsEditable=false, bIsStaticSwitch=false, bIsVisible=true, type {classStructOrEnum {refPath=&NiagaraInt32}}}
      Spawn Group {
       bIsDynamic=false
       bIsEditable=false
       bIsStaticSwitch=false
       bIsVisible=false
       type {classStructOrEnum {refPath=&NiagaraInt32}}
      }
      Randomness Mode {
       bIsDynamic=false
       bIsEditable=true
       bIsStaticSwitch=true
       bIsVisible=true
       type {classStructOrEnum {refPath=&ENiagaraRandomnessMode}}
      }
      Override Seed {bIsDynamic=false, bIsEditable=true, bIsStaticSwitch=false, bIsVisible=true, type {classStructOrEnum {refPath=&NiagaraBool}}}
      Random Seed {bIsDynamic=false, bIsEditable=false, bIsStaticSwitch=false, bIsVisible=true, type {classStructOrEnum {refPath=&NiagaraInt32}}}
      Fixed Random Seed {bIsDynamic=false, bIsEditable=false, bIsStaticSwitch=true, bIsVisible=false, type {classStructOrEnum {refPath=&NiagaraBool}}}
      Transform Order {
       bIsDynamic=false
       bIsEditable=true
       bIsStaticSwitch=true
       bIsVisible=true
       type {classStructOrEnum {refPath=&ENiagara_TransformOrder}}
      }
      Scale Mode {
       bIsDynamic=false
       bIsEditable=true
       bIsStaticSwitch=true
       bIsVisible=true
       type {classStructOrEnum {refPath=&ENiagara_ScaleMode}}
      }
      Non Uniform Scale {bIsDynamic=false, bIsEditable=true, bIsStaticSwitch=false, bIsVisible=true, type {classStructOrEnum {refPath=&Vector3f}}}
      UseOwnerScale {bIsDynamic=false, bIsEditable=true, bIsStaticSwitch=false, bIsVisible=true, type {classStructOrEnum {refPath=&NiagaraBool}}}
      Apply Owner Scale {bIsDynamic=false, bIsEditable=true, bIsStaticSwitch=false, bIsVisible=true, type {classStructOrEnum {refPath=&Vector3f}}}
      Rotation Mode {
       bIsDynamic=false
       bIsEditable=true
       bIsStaticSwitch=true
       bIsVisible=true
       type {classStructOrEnum {refPath=&ENiagara_RotationMode}}
      }
      Rotation Axis {
       bIsDynamic=false
       bIsEditable=false
       bIsStaticSwitch=false
       bIsVisible=false
       type {classStructOrEnum {refPath=&Vector3f}}
      }
      Rotation Matrix {
       bIsDynamic=false
       bIsEditable=false
       bIsStaticSwitch=false
       bIsVisible=false
       type {classStructOrEnum {refPath=&NiagaraMatrix}}
      }
      Rotation Quaternion {bIsDynamic=false, bIsEditable=false, bIsStaticSwitch=false, bIsVisible=false, type {classStructOrEnum {refPath=&Quat4f}}}
      Yaw / Pitch / Roll {
       bIsDynamic=false
       bIsEditable=false
       bIsStaticSwitch=false
       bIsVisible=false
       type {classStructOrEnum {refPath=&Vector3f}}
      }
      Rotation Angle {
       bIsDynamic=false
       bIsEditable=false
       bIsStaticSwitch=false
       bIsVisible=false
       type {classStructOrEnum {refPath=&NiagaraFloat}}
      }
      Invert Rotation Quaternion {bIsDynamic=false, bIsEditable=false, bIsStaticSwitch=false, bIsVisible=false, type {classStructOrEnum {refPath=&NiagaraBool}}}
      Additional Quaternion Rotation {bIsDynamic=false, bIsEditable=false, bIsStaticSwitch=false, bIsVisible=false, type {classStructOrEnum {refPath=&NiagaraBool}}}
      Additional Yaw / Pitch / Roll {
       bIsDynamic=false
       bIsEditable=false
       bIsStaticSwitch=false
       bIsVisible=false
       type {classStructOrEnum {refPath=&Vector3f}}
      }
      Rotation Angle Type {
       bIsDynamic=false
       bIsEditable=false
       bIsStaticSwitch=true
       bIsVisible=false
       type {classStructOrEnum {refPath=&ENiagara_AngleInput}}
      }
      Rotation Coordinate Space {
       bIsDynamic=false
       bIsEditable=true
       bIsStaticSwitch=false
       bIsVisible=true
       type {classStructOrEnum {refPath=&ENiagaraCoordinateSpace}}
      }
      Offset Mode {
       bIsDynamic=false
       bIsEditable=true
       bIsStaticSwitch=true
       bIsVisible=true
       type {classStructOrEnum {refPath=&ENiagara_OffsetMode}}
      }
      Shape Origin {
       bIsDynamic=false
       bIsEditable=true
       bIsStaticSwitch=false
       bIsVisible=true
       type {classStructOrEnum {refPath=&NiagaraPosition}}
      }
      Offset {
       bIsDynamic=false
       bIsEditable=false
       bIsStaticSwitch=false
       bIsVisible=false
       type {classStructOrEnum {refPath=&Vector3f}}
      }
      Offset Coordinate Space {
       bIsDynamic=false
       bIsEditable=false
       bIsStaticSwitch=false
       bIsVisible=false
       type {classStructOrEnum {refPath=&ENiagaraCoordinateSpace}}
      }
      Custom Transform Matrix {
       bIsDynamic=false
       bIsEditable=false
       bIsStaticSwitch=false
       bIsVisible=false
       type {classStructOrEnum {refPath=&NiagaraMatrix}}
      }
      Transform Method {
       bIsDynamic=false
       bIsEditable=true
       bIsStaticSwitch=true
       bIsVisible=true
       type {classStructOrEnum {refPath=&ENiagara_TransformType}}
      }
     ]
     moduleScript {refPath=&ShapeLocation}
    }
    {
     bIsSetParametersModule=false
     enabled=true
     moduleName=DynamicMaterialParameters
     inputs [
      Write Parameter Index 0 {bIsDynamic=false, bIsEditable=true, bIsStaticSwitch=true, bIsVisible=true, type {classStructOrEnum {refPath=&NiagaraBool}}}
      Module.Index0Param0Write {bIsDynamic=false, bIsEditable=true, bIsStaticSwitch=true, bIsVisible=true, type {classStructOrEnum {refPath=&NiagaraBool}}}
      Index 0 Param 1 {bIsDynamic=false, bIsEditable=true, bIsStaticSwitch=false, bIsVisible=true, type {classStructOrEnum {refPath=&NiagaraFloat}}}
      Module.Index0Param1Write {bIsDynamic=false, bIsEditable=true, bIsStaticSwitch=true, bIsVisible=true, type {classStructOrEnum {refPath=&NiagaraBool}}}
      Index 0 Param 2 {bIsDynamic=false, bIsEditable=true, bIsStaticSwitch=false, bIsVisible=true, type {classStructOrEnum {refPath=&NiagaraFloat}}}
      Module.Index0Param2Write {bIsDynamic=false, bIsEditable=true, bIsStaticSwitch=true, bIsVisible=true, type {classStructOrEnum {refPath=&NiagaraBool}}}
      Index 0 Param 3 {bIsDynamic=true, bIsEditable=true, bIsStaticSwitch=false, bIsVisible=true, type {classStructOrEnum {refPath=&NiagaraFloat}}}
      Module.Index0Param3Write {bIsDynamic=false, bIsEditable=true, bIsStaticSwitch=true, bIsVisible=true, type {classStructOrEnum {refPath=&NiagaraBool}}}
      Index 0 Param 4 {bIsDynamic=false, bIsEditable=true, bIsStaticSwitch=false, bIsVisible=true, type {classStructOrEnum {refPath=&NiagaraFloat}}}
      Write Parameter Index 1 {bIsDynamic=false, bIsEditable=true, bIsStaticSwitch=true, bIsVisible=true, type {classStructOrEnum {refPath=&NiagaraBool}}}
      Index 1 Param 1 {
       bIsDynamic=false
       bIsEditable=false
       bIsStaticSwitch=false
       bIsVisible=false
       type {classStructOrEnum {refPath=&NiagaraFloat}}
      }
      Index 1 Param 2 {
       bIsDynamic=false
       bIsEditable=false
       bIsStaticSwitch=false
       bIsVisible=false
       type {classStructOrEnum {refPath=&NiagaraFloat}}
      }
      Index 1 Param 3 {
       bIsDynamic=false
       bIsEditable=false
       bIsStaticSwitch=false
       bIsVisible=false
       type {classStructOrEnum {refPath=&NiagaraFloat}}
      }
      Index 1 Param 4 {
       bIsDynamic=false
       bIsEditable=false
       bIsStaticSwitch=false
       bIsVisible=false
       type {classStructOrEnum {refPath=&NiagaraFloat}}
      }
      Write Parameter Index 2 {bIsDynamic=false, bIsEditable=true, bIsStaticSwitch=true, bIsVisible=true, type {classStructOrEnum {refPath=&NiagaraBool}}}
      Index 2 Param 1 {
       bIsDynamic=false
       bIsEditable=false
       bIsStaticSwitch=false
       bIsVisible=false
       type {classStructOrEnum {refPath=&NiagaraFloat}}
      }
      Index 2 Param 2 {
       bIsDynamic=false
       bIsEditable=false
       bIsStaticSwitch=false
       bIsVisible=false
       type {classStructOrEnum {refPath=&NiagaraFloat}}
      }
      Index 2 Param 3 {
       bIsDynamic=false
       bIsEditable=false
       bIsStaticSwitch=false
       bIsVisible=false
       type {classStructOrEnum {refPath=&NiagaraFloat}}
      }
      Index 2 Param 4 {
       bIsDynamic=false
       bIsEditable=false
       bIsStaticSwitch=false
       bIsVisible=false
       type {classStructOrEnum {refPath=&NiagaraFloat}}
      }
      Write Parameter Index 3 {bIsDynamic=false, bIsEditable=true, bIsStaticSwitch=true, bIsVisible=true, type {classStructOrEnum {refPath=&NiagaraBool}}}
      Index 3 Param 1 {
       bIsDynamic=false
       bIsEditable=false
       bIsStaticSwitch=false
       bIsVisible=false
       type {classStructOrEnum {refPath=&NiagaraFloat}}
      }
      Index 3 Param 2 {
       bIsDynamic=false
       bIsEditable=false
       bIsStaticSwitch=false
       bIsVisible=false
       type {classStructOrEnum {refPath=&NiagaraFloat}}
      }
      Index 3 Param 3 {
       bIsDynamic=false
       bIsEditable=false
       bIsStaticSwitch=false
       bIsVisible=false
       type {classStructOrEnum {refPath=&NiagaraFloat}}
      }
      Index 3 Param 4 {
       bIsDynamic=false
       bIsEditable=false
       bIsStaticSwitch=false
       bIsVisible=false
       type {classStructOrEnum {refPath=&NiagaraFloat}}
      }
     ]
     moduleScript {refPath=&DynamicMaterialParameters}
    }
    {
     bIsSetParametersModule=false
     enabled=true
     moduleName=AddVelocity
     inputs [
      Velocity Mode {
       bIsDynamic=false
       bIsEditable=true
       bIsStaticSwitch=true
       bIsVisible=true
       type {classStructOrEnum {refPath=&ENiagara_VelocityMode}}
      }
      Velocity {bIsDynamic=true, bIsEditable=false, bIsStaticSwitch=false, bIsVisible=false, type {classStructOrEnum {refPath=&Vector3f}}}
      Velocity Speed {bIsDynamic=true, bIsEditable=true, bIsStaticSwitch=false, bIsVisible=true, type {classStructOrEnum {refPath=&NiagaraFloat}}}
      Velocity Speed Scale {bIsDynamic=true, bIsEditable=false, bIsStaticSwitch=false, bIsVisible=false, type {classStructOrEnum {refPath=&NiagaraFloat}}}
      Module.Use Falloff Distance {bIsDynamic=false, bIsEditable=true, bIsStaticSwitch=false, bIsVisible=true, type {classStructOrEnum {refPath=&NiagaraBool}}}
      Constrain To Radius {bIsDynamic=false, bIsEditable=false, bIsStaticSwitch=false, bIsVisible=true, type {classStructOrEnum {refPath=&NiagaraFloat}}}
      Distribution Along Cone Axis {
       bIsDynamic=false
       bIsEditable=false
       bIsStaticSwitch=false
       bIsVisible=false
       type {classStructOrEnum {refPath=&NiagaraFloat}}
      }
      Radius Falloff Near / Far {
       bIsDynamic=false
       bIsEditable=false
       bIsStaticSwitch=false
       bIsVisible=false
       type {classStructOrEnum {refPath=&Vector2f}}
      }
      Radius Falloff Exponent {
       bIsDynamic=false
       bIsEditable=false
       bIsStaticSwitch=false
       bIsVisible=false
       type {classStructOrEnum {refPath=&NiagaraFloat}}
      }
      Speed Falloff From Cone Axis {
       bIsDynamic=false
       bIsEditable=false
       bIsStaticSwitch=false
       bIsVisible=false
       type {classStructOrEnum {refPath=&NiagaraFloat}}
      }
      Invert Velocity Falloff {bIsDynamic=false, bIsEditable=false, bIsStaticSwitch=false, bIsVisible=false, type {classStructOrEnum {refPath=&NiagaraBool}}}
      Override Seed {bIsDynamic=false, bIsEditable=true, bIsStaticSwitch=false, bIsVisible=true, type {classStructOrEnum {refPath=&NiagaraBool}}}
      Random Seed {bIsDynamic=false, bIsEditable=false, bIsStaticSwitch=false, bIsVisible=true, type {classStructOrEnum {refPath=&NiagaraInt32}}}
      Fixed Random Seed {bIsDynamic=false, bIsEditable=false, bIsStaticSwitch=true, bIsVisible=false, type {classStructOrEnum {refPath=&NiagaraBool}}}
      Randomness Mode {
       bIsDynamic=false
       bIsEditable=true
       bIsStaticSwitch=true
       bIsVisible=true
       type {classStructOrEnum {refPath=&ENiagaraRandomnessMode}}
      }
      Cone Axis {
       bIsDynamic=false
       bIsEditable=false
       bIsStaticSwitch=false
       bIsVisible=false
       type {classStructOrEnum {refPath=&Vector3f}}
      }
      Cone Angle {
       bIsDynamic=false
       bIsEditable=false
       bIsStaticSwitch=false
       bIsVisible=false
       type {classStructOrEnum {refPath=&NiagaraFloat}}
      }
      Inner Cone Angle {
       bIsDynamic=false
       bIsEditable=false
       bIsStaticSwitch=false
       bIsVisible=false
       type {classStructOrEnum {refPath=&NiagaraFloat}}
      }
      Cone Angle Mode {
       bIsDynamic=false
       bIsEditable=false
       bIsStaticSwitch=true
       bIsVisible=false
       type {classStructOrEnum {refPath=&ENiagara_AngleInput}}
      }
      Rotation Mode {
       bIsDynamic=false
       bIsEditable=false
       bIsStaticSwitch=true
       bIsVisible=false
       type {classStructOrEnum {refPath=&ENiagara_RotationMode}}
      }
      Rotation Axis {
       bIsDynamic=false
       bIsEditable=false
       bIsStaticSwitch=false
       bIsVisible=false
       type {classStructOrEnum {refPath=&Vector3f}}
      }
      Yaw / Pitch / Roll {
       bIsDynamic=false
       bIsEditable=false
       bIsStaticSwitch=false
       bIsVisible=false
       type {classStructOrEnum {refPath=&Vector3f}}
      }
      Rotation Quaternion {bIsDynamic=false, bIsEditable=false, bIsStaticSwitch=false, bIsVisible=false, type {classStructOrEnum {refPath=&Quat4f}}}
      Rotation Matrix {
       bIsDynamic=false
       bIsEditable=false
       bIsStaticSwitch=false
       bIsVisible=false
       type {classStructOrEnum {refPath=&NiagaraMatrix}}
      }
      Rotation Angle {
       bIsDynamic=false
       bIsEditable=false
       bIsStaticSwitch=false
       bIsVisible=false
       type {classStructOrEnum {refPath=&NiagaraFloat}}
      }
      Invert Rotation Quaternion {bIsDynamic=false, bIsEditable=false, bIsStaticSwitch=false, bIsVisible=false, type {classStructOrEnum {refPath=&NiagaraBool}}}
      Additional Quaternion Rotation {bIsDynamic=false, bIsEditable=false, bIsStaticSwitch=false, bIsVisible=false, type {classStructOrEnum {refPath=&NiagaraBool}}}
      Additional Yaw / Pitch / Roll {
       bIsDynamic=false
       bIsEditable=false
       bIsStaticSwitch=false
       bIsVisible=false
       type {classStructOrEnum {refPath=&Vector3f}}
      }
      Rotation Angle Type {
       bIsDynamic=false
       bIsEditable=false
       bIsStaticSwitch=true
       bIsVisible=false
       type {classStructOrEnum {refPath=&ENiagara_AngleInput}}
      }
      Rotation Coordinate Space {
       bIsDynamic=false
       bIsEditable=false
       bIsStaticSwitch=false
       bIsVisible=false
       type {classStructOrEnum {refPath=&ENiagaraCoordinateSpace}}
      }
      Velocity Origin {
       bIsDynamic=false
       bIsEditable=true
       bIsStaticSwitch=false
       bIsVisible=true
       type {classStructOrEnum {refPath=&NiagaraPosition}}
      }
      Origin Offset {bIsDynamic=false, bIsEditable=true, bIsStaticSwitch=false, bIsVisible=true, type {classStructOrEnum {refPath=&Vector3f}}}
      Origin Offset Coordinate Space {
       bIsDynamic=false
       bIsEditable=true
       bIsStaticSwitch=false
       bIsVisible=true
       type {classStructOrEnum {refPath=&ENiagaraCoordinateSpace}}
      }
      Default Position {
       bIsDynamic=false
       bIsEditable=true
       bIsStaticSwitch=false
       bIsVisible=true
       type {classStructOrEnum {refPath=&NiagaraPosition}}
      }
     ]
     moduleScript {refPath=&AddVelocity}
    }
    {
     bIsSetParametersModule=false
     enabled=true
     moduleName=InitialMeshOrientation
     inputs [
      Mesh Orientation Mode {
       bIsDynamic=false
       bIsEditable=true
       bIsStaticSwitch=true
       bIsVisible=true
       type {classStructOrEnum {refPath=&ENiagara_MeshOrientationOptions}}
      }
      Orientation Quaternion {bIsDynamic=false, bIsEditable=false, bIsStaticSwitch=false, bIsVisible=false, type {classStructOrEnum {refPath=&Quat4f}}}
      Orientation Matrix {
       bIsDynamic=false
       bIsEditable=false
       bIsStaticSwitch=false
       bIsVisible=false
       type {classStructOrEnum {refPath=&NiagaraMatrix}}
      }
      Orientation Vector {
       bIsDynamic=false
       bIsEditable=false
       bIsStaticSwitch=false
       bIsVisible=false
       type {classStructOrEnum {refPath=&Vector3f}}
      }
      Orientation Coordinate Space {
       bIsDynamic=false
       bIsEditable=false
       bIsStaticSwitch=true
       bIsVisible=false
       type {classStructOrEnum {refPath=&ENiagaraCoordinateSpace}}
      }
      Mesh Axis To Orient {
       bIsDynamic=false
       bIsEditable=false
       bIsStaticSwitch=false
       bIsVisible=false
       type {classStructOrEnum {refPath=&Vector3f}}
      }
      Random Seed {bIsDynamic=false, bIsEditable=true, bIsStaticSwitch=false, bIsVisible=true, type {classStructOrEnum {refPath=&NiagaraInt32}}}
      Use Rotation Vector {bIsDynamic=false, bIsEditable=true, bIsStaticSwitch=false, bIsVisible=true, type {classStructOrEnum {refPath=&NiagaraBool}}}
      Rotation {bIsDynamic=false, bIsEditable=false, bIsStaticSwitch=false, bIsVisible=true, type {classStructOrEnum {refPath=&Vector3f}}}
      Rotation Coordinate Space {
       bIsDynamic=false
       bIsEditable=false
       bIsStaticSwitch=true
       bIsVisible=true
       type {classStructOrEnum {refPath=&ENiagaraMeshTransforms}}
      }
     ]
     moduleScript {refPath=&InitialMeshOrientation}
    }
   ]
  }
  particleUpdateScript {
   scriptName=ParticleUpdateScript
   modules [
    {
     bIsSetParametersModule=false
     enabled=true
     moduleName=ParticleState
     inputs [
      Kill Particles When Lifetime Has Elapsed {bIsDynamic=false, bIsEditable=true, bIsStaticSwitch=true, bIsVisible=true, type {classStructOrEnum {refPath=&NiagaraBool}}}
      Loop Particles Lifetime {bIsDynamic=false, bIsEditable=false, bIsStaticSwitch=false, bIsVisible=false, type {classStructOrEnum {refPath=&NiagaraBool}}}
      Let Infinitely Lived Particles Die When Emitter Deactivates {bIsDynamic=false, bIsEditable=false, bIsStaticSwitch=false, bIsVisible=false, type {classStructOrEnum {refPath=&NiagaraBool}}}
      Lifetime {bIsDynamic=false, bIsEditable=true, bIsStaticSwitch=false, bIsVisible=true, type {classStructOrEnum {refPath=&NiagaraFloat}}}
      DeltaTime {bIsDynamic=false, bIsEditable=true, bIsStaticSwitch=false, bIsVisible=true, type {classStructOrEnum {refPath=&NiagaraFloat}}}
     ]
     moduleScript {refPath=&ParticleState}
    }
    {
     bIsSetParametersModule=false
     enabled=true
     moduleName=ScaleMeshSize
     inputs [
      Scale Factor {bIsDynamic=true, bIsEditable=true, bIsStaticSwitch=false, bIsVisible=true, type {classStructOrEnum {refPath=&Vector3f}}}
      Initial Mesh Scale {bIsDynamic=false, bIsEditable=true, bIsStaticSwitch=false, bIsVisible=true, type {classStructOrEnum {refPath=&Vector3f}}}
     ]
     moduleScript {refPath=&ScaleMeshSize}
    }
    {
     bIsSetParametersModule=false
     enabled=true
     moduleName=UpdateMeshOrientation
     inputs [
      Orientation Method {
       bIsDynamic=false
       bIsEditable=true
       bIsStaticSwitch=true
       bIsVisible=true
       type {classStructOrEnum {refPath=/Niagara/Enums/Utility/ENiagara_UpdateMeshOrientationMode}}
      }
      Facing Position {
       bIsDynamic=false
       bIsEditable=false
       bIsStaticSwitch=false
       bIsVisible=false
       type {classStructOrEnum {refPath=&NiagaraPosition}}
      }
      Facing Direction {
       bIsDynamic=false
       bIsEditable=false
       bIsStaticSwitch=false
       bIsVisible=false
       type {classStructOrEnum {refPath=&Vector3f}}
      }
      TransformFacingPosition {bIsDynamic=false, bIsEditable=false, bIsStaticSwitch=false, bIsVisible=false, type {classStructOrEnum {refPath=&NiagaraBool}}}
      Facing Position Coordinate Space {
       bIsDynamic=false
       bIsEditable=false
       bIsStaticSwitch=false
       bIsVisible=false
       type {classStructOrEnum {refPath=&ENiagaraCoordinateSpace}}
      }
      "Facing Origin " {
       bIsDynamic=false
       bIsEditable=false
       bIsStaticSwitch=false
       bIsVisible=false
       type {classStructOrEnum {refPath=&NiagaraPosition}}
      }
      TransformFacingOrigin {bIsDynamic=false, bIsEditable=false, bIsStaticSwitch=false, bIsVisible=false, type {classStructOrEnum {refPath=&NiagaraBool}}}
      Facing Origin Coordinate Space {
       bIsDynamic=false
       bIsEditable=false
       bIsStaticSwitch=false
       bIsVisible=false
       type {classStructOrEnum {refPath=&ENiagaraCoordinateSpace}}
      }
      TransformFacing {bIsDynamic=false, bIsEditable=false, bIsStaticSwitch=false, bIsVisible=false, type {classStructOrEnum {refPath=&NiagaraBool}}}
      Facing Coordinate Space {
       bIsDynamic=false
       bIsEditable=false
       bIsStaticSwitch=false
       bIsVisible=false
       type {classStructOrEnum {refPath=&ENiagaraCoordinateSpace}}
      }
      Facing Interpolation {bIsDynamic=false, bIsEditable=false, bIsStaticSwitch=false, bIsVisible=false, type {classStructOrEnum {refPath=&NiagaraBool}}}
      Facing Interpolation Rate {
       bIsDynamic=false
       bIsEditable=false
       bIsStaticSwitch=false
       bIsVisible=false
       type {classStructOrEnum {refPath=&NiagaraFloat}}
      }
      Stabilization Mode {
       bIsDynamic=false
       bIsEditable=false
       bIsStaticSwitch=true
       bIsVisible=false
       type {classStructOrEnum {refPath=/Niagara/Enums/Utility/ENiagara_OrientVectorCount}}
      }
      Side Direction {
       bIsDynamic=false
       bIsEditable=false
       bIsStaticSwitch=false
       bIsVisible=false
       type {classStructOrEnum {refPath=&Vector3f}}
      }
      TransformSide {bIsDynamic=false, bIsEditable=false, bIsStaticSwitch=false, bIsVisible=false, type {classStructOrEnum {refPath=&NiagaraBool}}}
      Side Coordinate Space {
       bIsDynamic=false
       bIsEditable=false
       bIsStaticSwitch=false
       bIsVisible=false
       type {classStructOrEnum {refPath=&ENiagaraCoordinateSpace}}
      }
      Up Direction {
       bIsDynamic=false
       bIsEditable=false
       bIsStaticSwitch=false
       bIsVisible=false
       type {classStructOrEnum {refPath=&Vector3f}}
      }
      TransformUp {bIsDynamic=false, bIsEditable=false, bIsStaticSwitch=false, bIsVisible=false, type {classStructOrEnum {refPath=&NiagaraBool}}}
      Up Coordinate Space {
       bIsDynamic=false
       bIsEditable=false
       bIsStaticSwitch=false
       bIsVisible=false
       type {classStructOrEnum {refPath=&ENiagaraCoordinateSpace}}
      }
      Rotation Vector {bIsDynamic=true, bIsEditable=true, bIsStaticSwitch=false, bIsVisible=true, type {classStructOrEnum {refPath=&Vector3f}}}
      Rotation Rate {bIsDynamic=false, bIsEditable=true, bIsStaticSwitch=false, bIsVisible=true, type {classStructOrEnum {refPath=&NiagaraFloat}}}
      Surface Normal {
       bIsDynamic=false
       bIsEditable=false
       bIsStaticSwitch=false
       bIsVisible=false
       type {classStructOrEnum {refPath=&Vector3f}}
      }
      Enabled {bIsDynamic=false, bIsEditable=false, bIsStaticSwitch=false, bIsVisible=false, type {classStructOrEnum {refPath=&NiagaraBool}}}
      Particle Velocity {
       bIsDynamic=false
       bIsEditable=false
       bIsStaticSwitch=false
       bIsVisible=false
       type {classStructOrEnum {refPath=&Vector3f}}
      }
      Rotation Coordinate Space {
       bIsDynamic=false
       bIsEditable=true
       bIsStaticSwitch=true
       bIsVisible=true
       type {classStructOrEnum {refPath=&ENiagaraMeshTransforms}}
      }
      Initial Forward Vector {
       bIsDynamic=false
       bIsEditable=false
       bIsStaticSwitch=false
       bIsVisible=false
       type {classStructOrEnum {refPath=&Vector3f}}
      }
      Reference Axis Vector {
       bIsDynamic=false
       bIsEditable=false
       bIsStaticSwitch=false
       bIsVisible=false
       type {classStructOrEnum {refPath=&Vector3f}}
      }
      Update Rotation Rate {bIsDynamic=false, bIsEditable=false, bIsStaticSwitch=false, bIsVisible=false, type {classStructOrEnum {refPath=&NiagaraBool}}}
      Velocity {
       bIsDynamic=false
       bIsEditable=false
       bIsStaticSwitch=false
       bIsVisible=false
       type {classStructOrEnum {refPath=&Vector3f}}
      }
      Override Quat {bIsDynamic=false, bIsEditable=false, bIsStaticSwitch=false, bIsVisible=false, type {classStructOrEnum {refPath=&NiagaraBool}}}
      Override Incoming Orientation {
       bIsDynamic=false
       bIsEditable=false
       bIsStaticSwitch=false
       bIsVisible=false
       type {classStructOrEnum {refPath=&NiagaraFloat}}
      }
      Override Quaternion {bIsDynamic=false, bIsEditable=false, bIsStaticSwitch=false, bIsVisible=false, type {classStructOrEnum {refPath=&Quat4f}}}
      First Frame {bIsDynamic=false, bIsEditable=true, bIsStaticSwitch=false, bIsVisible=true, type {classStructOrEnum {refPath=&NiagaraBool}}}
      Write to Orientation {
       bIsDynamic=false
       bIsEditable=true
       bIsStaticSwitch=true
       bIsVisible=true
       type {classStructOrEnum {refPath=&ENiagara_AttributeSamplingApplyOutput}}
      }
      Automatically Calculate Particle Radius {bIsDynamic=false, bIsEditable=false, bIsStaticSwitch=true, bIsVisible=false, type {classStructOrEnum {refPath=&NiagaraBool}}}
      Particle Radius {
       bIsDynamic=false
       bIsEditable=false
       bIsStaticSwitch=false
       bIsVisible=false
       type {classStructOrEnum {refPath=&NiagaraFloat}}
      }
      Particle Radius Scale {
       bIsDynamic=false
       bIsEditable=false
       bIsStaticSwitch=false
       bIsVisible=false
       type {classStructOrEnum {refPath=&NiagaraFloat}}
      }
      Mesh Radius Calculation Method {
       bIsDynamic=false
       bIsEditable=false
       bIsStaticSwitch=true
       bIsVisible=false
       type {classStructOrEnum {refPath=&ENiagaraCalculateRadiusOptions}}
      }
      Mesh Renderer {
       bIsDynamic=false
       bIsEditable=false
       bIsStaticSwitch=false
       bIsVisible=false
       type {classStructOrEnum {refPath=&NiagaraDataInterfaceMeshRendererInfo}}
      }
      Mesh Renderer Index {
       bIsDynamic=false
       bIsEditable=false
       bIsStaticSwitch=false
       bIsVisible=false
       type {classStructOrEnum {refPath=&NiagaraInt32}}
      }
      Scale {
       bIsDynamic=false
       bIsEditable=false
       bIsStaticSwitch=false
       bIsVisible=false
       type {classStructOrEnum {refPath=&Vector3f}}
      }
      "Foward Vector (Mesh Space)" {
       bIsDynamic=false
       bIsEditable=false
       bIsStaticSwitch=false
       bIsVisible=false
       type {classStructOrEnum {refPath=&Vector3f}}
      }
      "Right Vector (Mesh Space)" {
       bIsDynamic=false
       bIsEditable=false
       bIsStaticSwitch=false
       bIsVisible=false
       type {classStructOrEnum {refPath=&Vector3f}}
      }
      Up Vector {
       bIsDynamic=false
       bIsEditable=false
       bIsStaticSwitch=false
       bIsVisible=false
       type {classStructOrEnum {refPath=&Vector3f}}
      }
      Up Vector Coordinate Space {
       bIsDynamic=false
       bIsEditable=false
       bIsStaticSwitch=false
       bIsVisible=false
       type {classStructOrEnum {refPath=&ENiagaraCoordinateSpace}}
      }
      Bank Rate {
       bIsDynamic=false
       bIsEditable=false
       bIsStaticSwitch=false
       bIsVisible=false
       type {classStructOrEnum {refPath=&NiagaraFloat}}
      }
      Banking Angle Min / Max {
       bIsDynamic=false
       bIsEditable=false
       bIsStaticSwitch=false
       bIsVisible=false
       type {classStructOrEnum {refPath=&Vector2f}}
      }
      Rotation Decay Rate In Degrees {
       bIsDynamic=false
       bIsEditable=false
       bIsStaticSwitch=false
       bIsVisible=false
       type {classStructOrEnum {refPath=&NiagaraFloat}}
      }
      Upright Roll Correction Speed {
       bIsDynamic=false
       bIsEditable=false
       bIsStaticSwitch=false
       bIsVisible=false
       type {classStructOrEnum {refPath=&NiagaraFloat}}
      }
      Velocity Orientation Convergence Rate {
       bIsDynamic=false
       bIsEditable=false
       bIsStaticSwitch=false
       bIsVisible=false
       type {classStructOrEnum {refPath=&NiagaraFloat}}
      }
      Delta Time {bIsDynamic=false, bIsEditable=true, bIsStaticSwitch=false, bIsVisible=true, type {classStructOrEnum {refPath=&NiagaraFloat}}}
      Debug Coordinate System Scale {
       bIsDynamic=false
       bIsEditable=false
       bIsStaticSwitch=false
       bIsVisible=false
       type {classStructOrEnum {refPath=&NiagaraFloat}}
      }
     ]
     moduleScript {refPath=&UpdateMeshOrientation}
    }
    {
     bIsSetParametersModule=false
     enabled=true
     moduleName=AccelerationForce
     inputs [
      Acceleration {bIsDynamic=true, bIsEditable=true, bIsStaticSwitch=false, bIsVisible=true, type {classStructOrEnum {refPath=&Vector3f}}}
      Coordinate Space {
       bIsDynamic=false
       bIsEditable=true
       bIsStaticSwitch=false
       bIsVisible=true
       type {classStructOrEnum {refPath=&ENiagaraCoordinateSpace}}
      }
     ]
     moduleScript {refPath=&AccelerationForce}
    }
    {
     bIsSetParametersModule=false
     enabled=true
     moduleName=Drag
     inputs [
      Use Linear Drag {bIsDynamic=false, bIsEditable=true, bIsStaticSwitch=false, bIsVisible=true, type {classStructOrEnum {refPath=&NiagaraBool}}}
      Drag {bIsDynamic=true, bIsEditable=true, bIsStaticSwitch=false, bIsVisible=true, type {classStructOrEnum {refPath=&NiagaraFloat}}}
      Use Rotational Drag {bIsDynamic=false, bIsEditable=true, bIsStaticSwitch=false, bIsVisible=true, type {classStructOrEnum {refPath=&NiagaraBool}}}
      Rotational Drag {bIsDynamic=false, bIsEditable=true, bIsStaticSwitch=false, bIsVisible=true, type {classStructOrEnum {refPath=&NiagaraFloat}}}
      Ignore Mass {bIsDynamic=false, bIsEditable=true, bIsStaticSwitch=false, bIsVisible=true, type {classStructOrEnum {refPath=&NiagaraBool}}}
     ]
     moduleScript {refPath=&Drag}
    }
    {
     bIsSetParametersModule=false
     enabled=true
     moduleName=SolveForcesAndVelocity
     inputs [
      Clamp Velocity {bIsDynamic=false, bIsEditable=true, bIsStaticSwitch=false, bIsVisible=true, type {classStructOrEnum {refPath=&NiagaraBool}}}
      Speed Limit {bIsDynamic=false, bIsEditable=false, bIsStaticSwitch=false, bIsVisible=true, type {classStructOrEnum {refPath=&NiagaraFloat}}}
      Limit Acceleration {bIsDynamic=false, bIsEditable=true, bIsStaticSwitch=false, bIsVisible=true, type {classStructOrEnum {refPath=&NiagaraBool}}}
      Acceleration Limit {bIsDynamic=false, bIsEditable=false, bIsStaticSwitch=false, bIsVisible=true, type {classStructOrEnum {refPath=&NiagaraFloat}}}
      Manually Enable Rotational Solver {bIsDynamic=false, bIsEditable=true, bIsStaticSwitch=false, bIsVisible=true, type {classStructOrEnum {refPath=&NiagaraBool}}}
      Write to Intrinsic Properties {bIsDynamic=false, bIsEditable=true, bIsStaticSwitch=false, bIsVisible=true, type {classStructOrEnum {refPath=&NiagaraBool}}}
      Write to Presolve Properties {bIsDynamic=false, bIsEditable=true, bIsStaticSwitch=true, bIsVisible=true, type {classStructOrEnum {refPath=&NiagaraBool}}}
      DeltaTime {bIsDynamic=false, bIsEditable=true, bIsStaticSwitch=false, bIsVisible=true, type {classStructOrEnum {refPath=&NiagaraFloat}}}
      Mass {bIsDynamic=false, bIsEditable=true, bIsStaticSwitch=false, bIsVisible=true, type {classStructOrEnum {refPath=&NiagaraFloat}}}
      Position {
       bIsDynamic=false
       bIsEditable=true
       bIsStaticSwitch=false
       bIsVisible=true
       type {classStructOrEnum {refPath=&NiagaraPosition}}
      }
      Velocity {bIsDynamic=false, bIsEditable=true, bIsStaticSwitch=false, bIsVisible=true, type {classStructOrEnum {refPath=&Vector3f}}}
      PreviousVelocity {bIsDynamic=false, bIsEditable=true, bIsStaticSwitch=false, bIsVisible=true, type {classStructOrEnum {refPath=&Vector3f}}}
      Force {bIsDynamic=false, bIsEditable=true, bIsStaticSwitch=false, bIsVisible=true, type {classStructOrEnum {refPath=&Vector3f}}}
      Mesh Orientation {bIsDynamic=false, bIsEditable=false, bIsStaticSwitch=false, bIsVisible=false, type {classStructOrEnum {refPath=&Quat4f}}}
      Rotational Velocity {
       bIsDynamic=false
       bIsEditable=false
       bIsStaticSwitch=false
       bIsVisible=false
       type {classStructOrEnum {refPath=&Vector3f}}
      }
      Rotational Force {
       bIsDynamic=false
       bIsEditable=false
       bIsStaticSwitch=false
       bIsVisible=false
       type {classStructOrEnum {refPath=&Vector3f}}
      }
      Rotational Inertia {
       bIsDynamic=false
       bIsEditable=false
       bIsStaticSwitch=false
       bIsVisible=false
       type {classStructOrEnum {refPath=&NiagaraFloat}}
      }
     ]
     moduleScript {refPath=&SolveForcesAndVelocity}
    }
    {
     bIsSetParametersModule=false
     enabled=true
     moduleName=ScaleColor
     inputs [
      Scale Mode {
       bIsDynamic=false
       bIsEditable=true
       bIsStaticSwitch=true
       bIsVisible=true
       type {classStructOrEnum {refPath=&ENiagaraScaleColorMode}}
      }
      Linear Color Curve {
       bIsDynamic=false
       bIsEditable=false
       bIsStaticSwitch=false
       bIsVisible=false
       type {classStructOrEnum {refPath=&NiagaraDataInterfaceColorCurve}}
      }
      ScaleRGB {bIsDynamic=false, bIsEditable=true, bIsStaticSwitch=false, bIsVisible=true, type {classStructOrEnum {refPath=&NiagaraBool}}}
      Scale RGB {bIsDynamic=false, bIsEditable=true, bIsStaticSwitch=false, bIsVisible=true, type {classStructOrEnum {refPath=&Vector3f}}}
      ScaleRGBA {bIsDynamic=false, bIsEditable=false, bIsStaticSwitch=false, bIsVisible=false, type {classStructOrEnum {refPath=&NiagaraBool}}}
      Scale RGBA {
       bIsDynamic=false
       bIsEditable=false
       bIsStaticSwitch=false
       bIsVisible=false
       type {classStructOrEnum {refPath=&Vector4f}}
      }
      Curve Index {
       bIsDynamic=false
       bIsEditable=false
       bIsStaticSwitch=false
       bIsVisible=false
       type {classStructOrEnum {refPath=&NiagaraFloat}}
      }
      ScaleA {bIsDynamic=false, bIsEditable=true, bIsStaticSwitch=false, bIsVisible=true, type {classStructOrEnum {refPath=&NiagaraBool}}}
      Scale Alpha {bIsDynamic=true, bIsEditable=true, bIsStaticSwitch=false, bIsVisible=true, type {classStructOrEnum {refPath=&NiagaraFloat}}}
      Color Value To Scale {
       bIsDynamic=false
       bIsEditable=true
       bIsStaticSwitch=false
       bIsVisible=true
       type {classStructOrEnum {refPath=&LinearColor}}
      }
     ]
     moduleScript {refPath=&ScaleColor}
    }
   ]
  }
  rendererClasses [{refPath=&NiagaraMeshRendererProperties}]
  renderers [
   {rendererIndex=0, rendererClass {refPath=&NiagaraMeshRendererProperties}}
   {rendererIndex=1, rendererClass {refPath=&NiagaraMeshRendererProperties}}
  ]
 }
 Swoosh {
  emitterSpawnScript {scriptName=EmitterSpawnScript, modules []}
  emitterUpdateScript {
   scriptName=EmitterUpdateScript
   modules [
    {
     bIsSetParametersModule=false
     enabled=true
     moduleName=EmitterState
     inputs [
      Life Cycle Mode {
       bIsDynamic=false
       bIsEditable=true
       bIsStaticSwitch=true
       bIsVisible=true
       type {classStructOrEnum {refPath=&ENiagaraEmitterLifeCycleMode}}
      }
      Inactive Response {
       bIsDynamic=false
       bIsEditable=true
       bIsStaticSwitch=true
       bIsVisible=true
       type {classStructOrEnum {refPath=&ENiagaraInactiveMode}}
      }
      Loop Behavior {
       bIsDynamic=false
       bIsEditable=true
       bIsStaticSwitch=true
       bIsVisible=true
       type {classStructOrEnum {refPath=&ENiagara_EmitterStateOptions}}
      }
      Loop Duration Mode {
       bIsDynamic=false
       bIsEditable=true
       bIsStaticSwitch=true
       bIsVisible=true
       type {classStructOrEnum {refPath=&ENiagara_InfiniteLoopDuration}}
      }
      Loop Count {
       bIsDynamic=false
       bIsEditable=false
       bIsStaticSwitch=false
       bIsVisible=false
       type {classStructOrEnum {refPath=&NiagaraInt32}}
      }
      Loop Duration {bIsDynamic=false, bIsEditable=true, bIsStaticSwitch=false, bIsVisible=true, type {classStructOrEnum {refPath=&NiagaraFloat}}}
      Recalculate Duration Each Loop {bIsDynamic=false, bIsEditable=false, bIsStaticSwitch=false, bIsVisible=false, type {classStructOrEnum {refPath=&NiagaraBool}}}
      UseLoopDelay {bIsDynamic=false, bIsEditable=true, bIsStaticSwitch=false, bIsVisible=true, type {classStructOrEnum {refPath=&NiagaraBool}}}
      Loop Delay {bIsDynamic=false, bIsEditable=false, bIsStaticSwitch=false, bIsVisible=true, type {classStructOrEnum {refPath=&NiagaraFloat}}}
      Delay First Loop Only {bIsDynamic=false, bIsEditable=false, bIsStaticSwitch=false, bIsVisible=false, type {classStructOrEnum {refPath=&NiagaraBool}}}
      Scalability Mode {
       bIsDynamic=false
       bIsEditable=true
       bIsStaticSwitch=true
       bIsVisible=true
       type {classStructOrEnum {refPath=&ENiagaraEmitterScalabilityMode_Limited}}
      }
      Enable Distance Culling {bIsDynamic=false, bIsEditable=false, bIsStaticSwitch=true, bIsVisible=false, type {classStructOrEnum {refPath=&NiagaraBool}}}
      MinDistance {
       bIsDynamic=false
       bIsEditable=false
       bIsStaticSwitch=false
       bIsVisible=false
       type {classStructOrEnum {refPath=&NiagaraFloat}}
      }
      Min Distance Response {
       bIsDynamic=false
       bIsEditable=false
       bIsStaticSwitch=false
       bIsVisible=false
       type {classStructOrEnum {refPath=&ENiagaraExecutionStateManagement}}
      }
      MaxDistance {
       bIsDynamic=false
       bIsEditable=false
       bIsStaticSwitch=false
       bIsVisible=false
       type {classStructOrEnum {refPath=&NiagaraFloat}}
      }
      Max Distance Response {
       bIsDynamic=false
       bIsEditable=false
       bIsStaticSwitch=false
       bIsVisible=false
       type {classStructOrEnum {refPath=&ENiagaraExecutionStateManagement}}
      }
      Scale Spawn Count {bIsDynamic=false, bIsEditable=false, bIsStaticSwitch=true, bIsVisible=false, type {classStructOrEnum {refPath=&NiagaraBool}}}
      Spawn Count Scale {
       bIsDynamic=false
       bIsEditable=false
       bIsStaticSwitch=false
       bIsVisible=false
       type {classStructOrEnum {refPath=&NiagaraFloat}}
      }
      Spawn Count Scale By Distance {
       bIsDynamic=false
       bIsEditable=false
       bIsStaticSwitch=false
       bIsVisible=false
       type {classStructOrEnum {refPath=&NiagaraDataInterfaceCurve}}
      }
      Enable Visibility Culling {bIsDynamic=false, bIsEditable=false, bIsStaticSwitch=true, bIsVisible=false, type {classStructOrEnum {refPath=&NiagaraBool}}}
      Visibility Cull Response {
       bIsDynamic=false
       bIsEditable=false
       bIsStaticSwitch=false
       bIsVisible=false
       type {classStructOrEnum {refPath=&ENiagaraExecutionStateManagement}}
      }
      Visibility Cull Delay {
       bIsDynamic=false
       bIsEditable=false
       bIsStaticSwitch=false
       bIsVisible=false
       type {classStructOrEnum {refPath=&NiagaraFloat}}
      }
      Reset Age On Awaken {bIsDynamic=false, bIsEditable=false, bIsStaticSwitch=false, bIsVisible=false, type {classStructOrEnum {refPath=&NiagaraBool}}}
     ]
     moduleScript {refPath=&EmitterState}
    }
    {
     bIsSetParametersModule=false
     enabled=true
     moduleName=SpawnBurst_Instantaneous
     inputs [
      Spawn Count {bIsDynamic=false, bIsEditable=true, bIsStaticSwitch=false, bIsVisible=true, type {classStructOrEnum {refPath=&NiagaraInt32}}}
      Spawn Time {bIsDynamic=false, bIsEditable=true, bIsStaticSwitch=false, bIsVisible=true, type {classStructOrEnum {refPath=&NiagaraFloat}}}
      Use Spawn Probability {bIsDynamic=false, bIsEditable=true, bIsStaticSwitch=false, bIsVisible=true, type {classStructOrEnum {refPath=&NiagaraBool}}}
      Spawn Probability {bIsDynamic=false, bIsEditable=false, bIsStaticSwitch=false, bIsVisible=true, type {classStructOrEnum {refPath=&NiagaraFloat}}}
      UseLoopCountLimit {bIsDynamic=false, bIsEditable=true, bIsStaticSwitch=false, bIsVisible=true, type {classStructOrEnum {refPath=&NiagaraBool}}}
      Loop Count Limit {bIsDynamic=false, bIsEditable=false, bIsStaticSwitch=false, bIsVisible=true, type {classStructOrEnum {refPath=&NiagaraInt32}}}
      Spawn Group {bIsDynamic=false, bIsEditable=true, bIsStaticSwitch=false, bIsVisible=true, type {classStructOrEnum {refPath=&NiagaraInt32}}}
      Age {bIsDynamic=false, bIsEditable=true, bIsStaticSwitch=false, bIsVisible=true, type {classStructOrEnum {refPath=&NiagaraFloat}}}
     ]
     moduleScript {refPath=&SpawnBurst_Instantaneous}
    }
   ]
  }
  particleSpawnScript {
   scriptName=ParticleSpawnScript
   modules [
    {
     bIsSetParametersModule=false
     enabled=true
     moduleName=InitializeParticle
     inputs [
      Lifetime Mode {
       bIsDynamic=false
       bIsEditable=true
       bIsStaticSwitch=true
       bIsVisible=true
       type {classStructOrEnum {refPath=&ENiagara_LifetimeMode}}
      }
      Lifetime Min {bIsDynamic=false, bIsEditable=true, bIsStaticSwitch=false, bIsVisible=true, type {classStructOrEnum {refPath=&NiagaraFloat}}}
      Lifetime {
       bIsDynamic=false
       bIsEditable=false
       bIsStaticSwitch=false
       bIsVisible=false
       type {classStructOrEnum {refPath=&NiagaraFloat}}
      }
      Lifetime Max {bIsDynamic=false, bIsEditable=true, bIsStaticSwitch=false, bIsVisible=true, type {classStructOrEnum {refPath=&NiagaraFloat}}}
      Lifetime Randomness Mode {
       bIsDynamic=false
       bIsEditable=true
       bIsStaticSwitch=true
       bIsVisible=true
       type {classStructOrEnum {refPath=&ENiagaraRandomnessMode}}
      }
      Color Mode {
       bIsDynamic=false
       bIsEditable=true
       bIsStaticSwitch=true
       bIsVisible=true
       type {classStructOrEnum {refPath=&ENiagara_ColorInitializationMode}}
      }
      Color Minimum {
       bIsDynamic=false
       bIsEditable=false
       bIsStaticSwitch=false
       bIsVisible=false
       type {classStructOrEnum {refPath=&LinearColor}}
      }
      Color {
       bIsDynamic=true
       bIsEditable=true
       bIsStaticSwitch=false
       bIsVisible=true
       type {classStructOrEnum {refPath=&LinearColor}}
      }
      AdjustHue {bIsDynamic=false, bIsEditable=false, bIsStaticSwitch=false, bIsVisible=false, type {classStructOrEnum {refPath=&NiagaraBool}}}
      Hue Shift Range {
       bIsDynamic=false
       bIsEditable=false
       bIsStaticSwitch=false
       bIsVisible=false
       type {classStructOrEnum {refPath=&Vector2f}}
      }
      Color Maximum {
       bIsDynamic=false
       bIsEditable=false
       bIsStaticSwitch=false
       bIsVisible=false
       type {classStructOrEnum {refPath=&LinearColor}}
      }
      AdjustSaturation {bIsDynamic=false, bIsEditable=false, bIsStaticSwitch=false, bIsVisible=false, type {classStructOrEnum {refPath=&NiagaraBool}}}
      Saturation Range {
       bIsDynamic=false
       bIsEditable=false
       bIsStaticSwitch=false
       bIsVisible=false
       type {classStructOrEnum {refPath=&Vector2f}}
      }
      AdjustValue {bIsDynamic=false, bIsEditable=false, bIsStaticSwitch=false, bIsVisible=false, type {classStructOrEnum {refPath=&NiagaraBool}}}
      Value Range {
       bIsDynamic=false
       bIsEditable=false
       bIsStaticSwitch=false
       bIsVisible=false
       type {classStructOrEnum {refPath=&Vector2f}}
      }
      AdjustAlpha {bIsDynamic=false, bIsEditable=false, bIsStaticSwitch=false, bIsVisible=false, type {classStructOrEnum {refPath=&NiagaraBool}}}
      Alpha Scale Range {
       bIsDynamic=false
       bIsEditable=false
       bIsStaticSwitch=false
       bIsVisible=false
       type {classStructOrEnum {refPath=&Vector2f}}
      }
      Color Channel Mode {
       bIsDynamic=false
       bIsEditable=false
       bIsStaticSwitch=true
       bIsVisible=false
       type {classStructOrEnum {refPath=&ENiagaraChannelCorrelation}}
      }
      Color Randomness Mode {
       bIsDynamic=false
       bIsEditable=false
       bIsStaticSwitch=true
       bIsVisible=false
       type {classStructOrEnum {refPath=&ENiagaraRandomnessMode}}
      }
      Position Mode {
       bIsDynamic=false
       bIsEditable=true
       bIsStaticSwitch=true
       bIsVisible=true
       type {classStructOrEnum {refPath=&ENiagara_PositionInitializationMode}}
      }
      Position {
       bIsDynamic=false
       bIsEditable=false
       bIsStaticSwitch=false
       bIsVisible=false
       type {classStructOrEnum {refPath=&NiagaraPosition}}
      }
      UsePositionOffset {bIsDynamic=false, bIsEditable=true, bIsStaticSwitch=false, bIsVisible=true, type {classStructOrEnum {refPath=&NiagaraBool}}}
      Position Offset {bIsDynamic=false, bIsEditable=true, bIsStaticSwitch=false, bIsVisible=true, type {classStructOrEnum {refPath=&Vector3f}}}
      Position Offset Coordinate Space {
       bIsDynamic=false
       bIsEditable=true
       bIsStaticSwitch=false
       bIsVisible=true
       type {classStructOrEnum {refPath=&ENiagaraCoordinateSpace}}
      }
      Mass Mode {
       bIsDynamic=false
       bIsEditable=true
       bIsStaticSwitch=true
       bIsVisible=true
       type {classStructOrEnum {refPath=&ENiagara_MassInitializationMode}}
      }
      Mass Min {
       bIsDynamic=false
       bIsEditable=false
       bIsStaticSwitch=false
       bIsVisible=false
       type {classStructOrEnum {refPath=&NiagaraFloat}}
      }
      Mass {
       bIsDynamic=false
       bIsEditable=false
       bIsStaticSwitch=false
       bIsVisible=false
       type {classStructOrEnum {refPath=&NiagaraFloat}}
      }
      Mass Max {
       bIsDynamic=false
       bIsEditable=false
       bIsStaticSwitch=false
       bIsVisible=false
       type {classStructOrEnum {refPath=&NiagaraFloat}}
      }
      Mass Randomness Mode {
       bIsDynamic=false
       bIsEditable=false
       bIsStaticSwitch=true
       bIsVisible=false
       type {classStructOrEnum {refPath=&ENiagaraRandomnessMode}}
      }
      Write Material Random {bIsDynamic=false, bIsEditable=true, bIsStaticSwitch=false, bIsVisible=true, type {classStructOrEnum {refPath=&NiagaraBool}}}
      Material Random {bIsDynamic=false, bIsEditable=true, bIsStaticSwitch=false, bIsVisible=true, type {classStructOrEnum {refPath=&NiagaraFloat}}}
      Sprite Size Mode {
       bIsDynamic=false
       bIsEditable=true
       bIsStaticSwitch=true
       bIsVisible=true
       type {classStructOrEnum {refPath=&ENiagara_SizeScaleMode}}
      }
      Sprite Size {
       bIsDynamic=false
       bIsEditable=false
       bIsStaticSwitch=false
       bIsVisible=false
       type {classStructOrEnum {refPath=&Vector2f}}
      }
      Uniform Sprite Size {
       bIsDynamic=false
       bIsEditable=false
       bIsStaticSwitch=false
       bIsVisible=false
       type {classStructOrEnum {refPath=&NiagaraFloat}}
      }
      Uniform Sprite Size Min {
       bIsDynamic=false
       bIsEditable=false
       bIsStaticSwitch=false
       bIsVisible=false
       type {classStructOrEnum {refPath=&NiagaraFloat}}
      }
      Sprite Size Min {
       bIsDynamic=false
       bIsEditable=false
       bIsStaticSwitch=false
       bIsVisible=false
       type {classStructOrEnum {refPath=&Vector2f}}
      }
      Sprite Size Max {
       bIsDynamic=false
       bIsEditable=false
       bIsStaticSwitch=false
       bIsVisible=false
       type {classStructOrEnum {refPath=&Vector2f}}
      }
      Uniform Sprite Size Max {
       bIsDynamic=false
       bIsEditable=false
       bIsStaticSwitch=false
       bIsVisible=false
       type {classStructOrEnum {refPath=&NiagaraFloat}}
      }
      Sprite Size Randomness Mode {
       bIsDynamic=false
       bIsEditable=false
       bIsStaticSwitch=true
       bIsVisible=false
       type {classStructOrEnum {refPath=&ENiagaraRandomnessMode}}
      }
      Sprite Rotation Mode {
       bIsDynamic=false
       bIsEditable=true
       bIsStaticSwitch=true
       bIsVisible=true
       type {classStructOrEnum {refPath=&ENiagara_SpriteRotationMode}}
      }
      Sprite Rotation Angle Min {
       bIsDynamic=false
       bIsEditable=false
       bIsStaticSwitch=false
       bIsVisible=false
       type {classStructOrEnum {refPath=&NiagaraFloat}}
      }
      Sprite Rotation Angle {
       bIsDynamic=false
       bIsEditable=false
       bIsStaticSwitch=false
       bIsVisible=false
       type {classStructOrEnum {refPath=&NiagaraFloat}}
      }
      Sprite Rotation Angle Max {
       bIsDynamic=false
       bIsEditable=false
       bIsStaticSwitch=false
       bIsVisible=false
       type {classStructOrEnum {refPath=&NiagaraFloat}}
      }
      Sprite Rotation Randomness Mode {
       bIsDynamic=false
       bIsEditable=false
       bIsStaticSwitch=true
       bIsVisible=false
       type {classStructOrEnum {refPath=&ENiagaraRandomnessMode}}
      }
      Sprite UV Mode {
       bIsDynamic=false
       bIsEditable=true
       bIsStaticSwitch=true
       bIsVisible=true
       type {classStructOrEnum {refPath=&ENiagara_UVFlippingMode}}
      }
      Sprite UV Randomness Mode {
       bIsDynamic=false
       bIsEditable=false
       bIsStaticSwitch=true
       bIsVisible=false
       type {classStructOrEnum {refPath=&ENiagaraRandomnessMode}}
      }
      Sprite UV Scale {
       bIsDynamic=false
       bIsEditable=false
       bIsStaticSwitch=false
       bIsVisible=false
       type {classStructOrEnum {refPath=&Vector2f}}
      }
      Mesh Scale Mode {
       bIsDynamic=false
       bIsEditable=true
       bIsStaticSwitch=true
       bIsVisible=true
       type {classStructOrEnum {refPath=&ENiagara_SizeScaleMode}}
      }
      Mesh Uniform Scale Min {bIsDynamic=false, bIsEditable=true, bIsStaticSwitch=false, bIsVisible=true, type {classStructOrEnum {refPath=&NiagaraFloat}}}
      Mesh Uniform Scale {bIsDynamic=true, bIsEditable=false, bIsStaticSwitch=false, bIsVisible=false, type {classStructOrEnum {refPath=&NiagaraFloat}}}
      Mesh Scale Min {
       bIsDynamic=false
       bIsEditable=false
       bIsStaticSwitch=false
       bIsVisible=false
       type {classStructOrEnum {refPath=&Vector3f}}
      }
      Mesh Scale {bIsDynamic=true, bIsEditable=false, bIsStaticSwitch=false, bIsVisible=false, type {classStructOrEnum {refPath=&Vector3f}}}
      Mesh Scale Max {
       bIsDynamic=false
       bIsEditable=false
       bIsStaticSwitch=false
       bIsVisible=false
       type {classStructOrEnum {refPath=&Vector3f}}
      }
      Mesh Uniform Scale Max {bIsDynamic=false, bIsEditable=true, bIsStaticSwitch=false, bIsVisible=true, type {classStructOrEnum {refPath=&NiagaraFloat}}}
      Mesh Scale Randomness Mode {
       bIsDynamic=false
       bIsEditable=true
       bIsStaticSwitch=true
       bIsVisible=true
       type {classStructOrEnum {refPath=&ENiagaraRandomnessMode}}
      }
      Mesh Renderer Array Visibility Mode {
       bIsDynamic=false
       bIsEditable=true
       bIsStaticSwitch=true
       bIsVisible=true
       type {classStructOrEnum {refPath=&ENiagara_UnsetDirectSetRandom}}
      }
      Mesh Renderer Mesh Index {
       bIsDynamic=false
       bIsEditable=false
       bIsStaticSwitch=false
       bIsVisible=false
       type {classStructOrEnum {refPath=&NiagaraInt32}}
      }
      Wrap Index By Mesh Count {bIsDynamic=false, bIsEditable=false, bIsStaticSwitch=true, bIsVisible=false, type {classStructOrEnum {refPath=&NiagaraBool}}}
      Mesh Renderer Info {
       bIsDynamic=false
       bIsEditable=false
       bIsStaticSwitch=false
       bIsVisible=false
       type {classStructOrEnum {refPath=&NiagaraDataInterfaceMeshRendererInfo}}
      }
      Mesh Renderer Random Seed {
       bIsDynamic=false
       bIsEditable=false
       bIsStaticSwitch=false
       bIsVisible=false
       type {classStructOrEnum {refPath=&NiagaraInt32}}
      }
      Ribbon Width Mode {
       bIsDynamic=false
       bIsEditable=true
       bIsStaticSwitch=true
       bIsVisible=true
       type {classStructOrEnum {refPath=&ENiagara_UnsetDirectSet}}
      }
      Ribbon Width {
       bIsDynamic=false
       bIsEditable=false
       bIsStaticSwitch=false
       bIsVisible=false
       type {classStructOrEnum {refPath=&NiagaraFloat}}
      }
      Ribbon Facing Mode {
       bIsDynamic=false
       bIsEditable=true
       bIsStaticSwitch=true
       bIsVisible=true
       type {classStructOrEnum {refPath=&ENiagara_UnsetDirectSet}}
      }
      Ribbon Facing Vector {
       bIsDynamic=false
       bIsEditable=false
       bIsStaticSwitch=false
       bIsVisible=false
       type {classStructOrEnum {refPath=&Vector3f}}
      }
      Ribbon Twist Mode {
       bIsDynamic=false
       bIsEditable=true
       bIsStaticSwitch=true
       bIsVisible=true
       type {classStructOrEnum {refPath=&ENiagara_UnsetDirectSet}}
      }
      "Ribbon Twist " {
       bIsDynamic=false
       bIsEditable=false
       bIsStaticSwitch=false
       bIsVisible=false
       type {classStructOrEnum {refPath=&NiagaraFloat}}
      }
     ]
     moduleScript {refPath=&InitializeParticle}
    }
    {
     bIsSetParametersModule=false
     enabled=true
     moduleName=ShapeLocation
     inputs [
      Shape Primitive {
       bIsDynamic=false
       bIsEditable=true
       bIsStaticSwitch=true
       bIsVisible=true
       type {classStructOrEnum {refPath=&ENiagara_LocationShapes}}
      }
      Box / Plane Mode {
       bIsDynamic=false
       bIsEditable=false
       bIsStaticSwitch=true
       bIsVisible=false
       type {classStructOrEnum {refPath=&ENiagara_BoxPlaneMode}}
      }
      Cone Mode {
       bIsDynamic=false
       bIsEditable=false
       bIsStaticSwitch=true
       bIsVisible=false
       type {classStructOrEnum {refPath=&ENiagara_ConeMode}}
      }
      Cylinder Height {
       bIsDynamic=false
       bIsEditable=false
       bIsStaticSwitch=false
       bIsVisible=false
       type {classStructOrEnum {refPath=&NiagaraFloat}}
      }
      Sphere Radius {
       bIsDynamic=false
       bIsEditable=false
       bIsStaticSwitch=false
       bIsVisible=false
       type {classStructOrEnum {refPath=&NiagaraFloat}}
      }
      Ring / Disc Mode {
       bIsDynamic=false
       bIsEditable=true
       bIsStaticSwitch=true
       bIsVisible=true
       type {classStructOrEnum {refPath=&ENiagaraRingDiscMode}}
      }
      Torus Mode {
       bIsDynamic=false
       bIsEditable=false
       bIsStaticSwitch=true
       bIsVisible=false
       type {classStructOrEnum {refPath=&ENiagaraShapeTorusMode}}
      }
      Knot Ring Loops {
       bIsDynamic=false
       bIsEditable=false
       bIsStaticSwitch=false
       bIsVisible=false
       type {classStructOrEnum {refPath=&NiagaraInt32}}
      }
      Knot Helix Loops {
       bIsDynamic=false
       bIsEditable=false
       bIsStaticSwitch=false
       bIsVisible=false
       type {classStructOrEnum {refPath=&NiagaraInt32}}
      }
      Enable Torus Knot Radius {bIsDynamic=false, bIsEditable=false, bIsStaticSwitch=false, bIsVisible=false, type {classStructOrEnum {refPath=&NiagaraBool}}}
      Torus Knot Tube Radius {
       bIsDynamic=false
       bIsEditable=false
       bIsStaticSwitch=false
       bIsVisible=false
       type {classStructOrEnum {refPath=&NiagaraFloat}}
      }
      Cylinder Radius {
       bIsDynamic=false
       bIsEditable=false
       bIsStaticSwitch=false
       bIsVisible=false
       type {classStructOrEnum {refPath=&NiagaraFloat}}
      }
      Cylinder Height Midpoint {
       bIsDynamic=false
       bIsEditable=false
       bIsStaticSwitch=false
       bIsVisible=false
       type {classStructOrEnum {refPath=&NiagaraFloat}}
      }
      Ring Radius {bIsDynamic=false, bIsEditable=true, bIsStaticSwitch=false, bIsVisible=true, type {classStructOrEnum {refPath=&NiagaraFloat}}}
      Box Size {
       bIsDynamic=false
       bIsEditable=false
       bIsStaticSwitch=false
       bIsVisible=false
       type {classStructOrEnum {refPath=&Vector3f}}
      }
      Box Midpoint {
       bIsDynamic=false
       bIsEditable=false
       bIsStaticSwitch=false
       bIsVisible=false
       type {classStructOrEnum {refPath=&Vector3f}}
      }
      Cone Length {
       bIsDynamic=false
       bIsEditable=false
       bIsStaticSwitch=false
       bIsVisible=false
       type {classStructOrEnum {refPath=&NiagaraFloat}}
      }
      Plane Size {
       bIsDynamic=false
       bIsEditable=false
       bIsStaticSwitch=false
       bIsVisible=false
       type {classStructOrEnum {refPath=&Vector2f}}
      }
      Wedge Horizontal Angle {
       bIsDynamic=false
       bIsEditable=false
       bIsStaticSwitch=false
       bIsVisible=false
       type {classStructOrEnum {refPath=&NiagaraFloat}}
      }
      Large Radius {
       bIsDynamic=false
       bIsEditable=false
       bIsStaticSwitch=false
       bIsVisible=false
       type {classStructOrEnum {refPath=&NiagaraFloat}}
      }
      Plane Midpoint {
       bIsDynamic=false
       bIsEditable=false
       bIsStaticSwitch=false
       bIsVisible=false
       type {classStructOrEnum {refPath=&Vector2f}}
      }
      Wedge Vertical Angle {
       bIsDynamic=false
       bIsEditable=false
       bIsStaticSwitch=false
       bIsVisible=false
       type {classStructOrEnum {refPath=&NiagaraFloat}}
      }
      Cone Angle {
       bIsDynamic=false
       bIsEditable=false
       bIsStaticSwitch=false
       bIsVisible=false
       type {classStructOrEnum {refPath=&NiagaraFloat}}
      }
      Handle Radius {
       bIsDynamic=false
       bIsEditable=false
       bIsStaticSwitch=false
       bIsVisible=false
       type {classStructOrEnum {refPath=&NiagaraFloat}}
      }
      Cone Inner Angle {
       bIsDynamic=false
       bIsEditable=false
       bIsStaticSwitch=false
       bIsVisible=false
       type {classStructOrEnum {refPath=&NiagaraFloat}}
      }
      Cone Radial Angle {
       bIsDynamic=false
       bIsEditable=false
       bIsStaticSwitch=false
       bIsVisible=false
       type {classStructOrEnum {refPath=&NiagaraFloat}}
      }
      Cone Angle Type {
       bIsDynamic=false
       bIsEditable=false
       bIsStaticSwitch=true
       bIsVisible=false
       type {classStructOrEnum {refPath=&ENiagara_AngleInput}}
      }
      Flatten Endcaps {bIsDynamic=false, bIsEditable=false, bIsStaticSwitch=true, bIsVisible=false, type {classStructOrEnum {refPath=&NiagaraBool}}}
      Apply To Particle Position {
       bIsDynamic=false
       bIsEditable=true
       bIsStaticSwitch=true
       bIsVisible=true
       type {classStructOrEnum {refPath=&ENiagara_AttributeSamplingApplyOutput}}
      }
      Cone Surface Distribution {
       bIsDynamic=false
       bIsEditable=false
       bIsStaticSwitch=false
       bIsVisible=false
       type {classStructOrEnum {refPath=&NiagaraFloat}}
      }
      Ring / Disc Distribution Mode {
       bIsDynamic=false
       bIsEditable=true
       bIsStaticSwitch=true
       bIsVisible=true
       type {classStructOrEnum {refPath=&ENiagaraSphereDistributionMode}}
      }
      Sphere Distribution {
       bIsDynamic=false
       bIsEditable=false
       bIsStaticSwitch=true
       bIsVisible=false
       type {classStructOrEnum {refPath=&ENiagaraSphereDistributionMode}}
      }
      Torus Distribution Mode {
       bIsDynamic=false
       bIsEditable=false
       bIsStaticSwitch=true
       bIsVisible=false
       type {classStructOrEnum {refPath=&ENiagaraTorusDistributionMode}}
      }
      Sphere Surface Distribution {
       bIsDynamic=false
       bIsEditable=false
       bIsStaticSwitch=false
       bIsVisible=false
       type {classStructOrEnum {refPath=&NiagaraFloat}}
      }
      Hemisphere Distribution {
       bIsDynamic=false
       bIsEditable=false
       bIsStaticSwitch=false
       bIsVisible=false
       type {classStructOrEnum {refPath=&Vector2f}}
      }
      Cylinder Distribution {
       bIsDynamic=false
       bIsEditable=false
       bIsStaticSwitch=true
       bIsVisible=false
       type {classStructOrEnum {refPath=&ENiagara_CylinderMode}}
      }
      Surface Only {bIsDynamic=false, bIsEditable=false, bIsStaticSwitch=false, bIsVisible=false, type {classStructOrEnum {refPath=&NiagaraBool}}}
      Surface Only Band Thickness {
       bIsDynamic=false
       bIsEditable=false
       bIsStaticSwitch=false
       bIsVisible=false
       type {classStructOrEnum {refPath=&NiagaraFloat}}
      }
      Use Endcaps In Surface Only Mode {bIsDynamic=false, bIsEditable=false, bIsStaticSwitch=true, bIsVisible=false, type {classStructOrEnum {refPath=&NiagaraBool}}}
      Hemicircle X {bIsDynamic=false, bIsEditable=false, bIsStaticSwitch=true, bIsVisible=false, type {classStructOrEnum {refPath=&NiagaraBool}}}
      Hemicircle Y {bIsDynamic=false, bIsEditable=false, bIsStaticSwitch=true, bIsVisible=false, type {classStructOrEnum {refPath=&NiagaraBool}}}
      Disc Coverage {
       bIsDynamic=false
       bIsEditable=false
       bIsStaticSwitch=false
       bIsVisible=false
       type {classStructOrEnum {refPath=&NiagaraFloat}}
      }
      Hemicircle Internal Cap {bIsDynamic=false, bIsEditable=false, bIsStaticSwitch=true, bIsVisible=false, type {classStructOrEnum {refPath=&NiagaraBool}}}
      U Position {bIsDynamic=true, bIsEditable=true, bIsStaticSwitch=false, bIsVisible=true, type {classStructOrEnum {refPath=&NiagaraFloat}}}
      Enable Lathe Profile {bIsDynamic=false, bIsEditable=false, bIsStaticSwitch=true, bIsVisible=false, type {classStructOrEnum {refPath=&NiagaraBool}}}
      Lathe Profile {
       bIsDynamic=false
       bIsEditable=false
       bIsStaticSwitch=false
       bIsVisible=false
       type {classStructOrEnum {refPath=&NiagaraDataInterfaceCurve}}
      }
      Surface Distribution {
       bIsDynamic=false
       bIsEditable=false
       bIsStaticSwitch=false
       bIsVisible=false
       type {classStructOrEnum {refPath=&NiagaraFloat}}
      }
      Surface Expansion Mode {
       bIsDynamic=false
       bIsEditable=false
       bIsStaticSwitch=false
       bIsVisible=false
       type {classStructOrEnum {refPath=&ENiagaraExpansionMode}}
      }
      U Distribution {
       bIsDynamic=false
       bIsEditable=false
       bIsStaticSwitch=false
       bIsVisible=false
       type {classStructOrEnum {refPath=&NiagaraFloat}}
      }
      V Position {
       bIsDynamic=false
       bIsEditable=false
       bIsStaticSwitch=false
       bIsVisible=false
       type {classStructOrEnum {refPath=&NiagaraFloat}}
      }
      Radius Position {bIsDynamic=false, bIsEditable=true, bIsStaticSwitch=false, bIsVisible=true, type {classStructOrEnum {refPath=&NiagaraFloat}}}
      Sphere Radius Position {
       bIsDynamic=false
       bIsEditable=false
       bIsStaticSwitch=false
       bIsVisible=false
       type {classStructOrEnum {refPath=&NiagaraFloat}}
      }
      V Distribution {
       bIsDynamic=false
       bIsEditable=false
       bIsStaticSwitch=false
       bIsVisible=false
       type {classStructOrEnum {refPath=&NiagaraFloat}}
      }
      Evaluation Type {
       bIsDynamic=false
       bIsEditable=false
       bIsStaticSwitch=true
       bIsVisible=false
       type {classStructOrEnum {refPath=&ENiagaraRandomnessEvaluation}}
      }
      Uniform Distribution {
       bIsDynamic=false
       bIsEditable=false
       bIsStaticSwitch=false
       bIsVisible=false
       type {classStructOrEnum {refPath=&NiagaraFloat}}
      }
      Uniform Spiral Amount {
       bIsDynamic=false
       bIsEditable=false
       bIsStaticSwitch=false
       bIsVisible=false
       type {classStructOrEnum {refPath=&NiagaraFloat}}
      }
      Uniform Spiral Falloff {
       bIsDynamic=false
       bIsEditable=false
       bIsStaticSwitch=false
       bIsVisible=false
       type {classStructOrEnum {refPath=&NiagaraFloat}}
      }
      Uniform Count {
       bIsDynamic=false
       bIsEditable=false
       bIsStaticSwitch=false
       bIsVisible=false
       type {classStructOrEnum {refPath=&NiagaraInt32}}
      }
      Uniform Particle Index {
       bIsDynamic=false
       bIsEditable=false
       bIsStaticSwitch=false
       bIsVisible=false
       type {classStructOrEnum {refPath=&NiagaraInt32}}
      }
      Hemisphere Angle Type {
       bIsDynamic=false
       bIsEditable=false
       bIsStaticSwitch=true
       bIsVisible=false
       type {classStructOrEnum {refPath=&ENiagara_AngleInput}}
      }
      Uniform Endpoint Offset {
       bIsDynamic=false
       bIsEditable=false
       bIsStaticSwitch=false
       bIsVisible=false
       type {classStructOrEnum {refPath=&NiagaraFloat}}
      }
      MaskSpawnGroup {bIsDynamic=false, bIsEditable=true, bIsStaticSwitch=false, bIsVisible=true, type {classStructOrEnum {refPath=&NiagaraBool}}}
      Spawn Group Mask {bIsDynamic=false, bIsEditable=false, bIsStaticSwitch=false, bIsVisible=true, type {classStructOrEnum {refPath=&NiagaraInt32}}}
      Spawn Group {
       bIsDynamic=false
       bIsEditable=false
       bIsStaticSwitch=false
       bIsVisible=false
       type {classStructOrEnum {refPath=&NiagaraInt32}}
      }
      Randomness Mode {
       bIsDynamic=false
       bIsEditable=false
       bIsStaticSwitch=true
       bIsVisible=false
       type {classStructOrEnum {refPath=&ENiagaraRandomnessMode}}
      }
      Override Seed {bIsDynamic=false, bIsEditable=true, bIsStaticSwitch=false, bIsVisible=true, type {classStructOrEnum {refPath=&NiagaraBool}}}
      Random Seed {bIsDynamic=false, bIsEditable=false, bIsStaticSwitch=false, bIsVisible=true, type {classStructOrEnum {refPath=&NiagaraInt32}}}
      Fixed Random Seed {bIsDynamic=false, bIsEditable=false, bIsStaticSwitch=true, bIsVisible=false, type {classStructOrEnum {refPath=&NiagaraBool}}}
      Transform Order {
       bIsDynamic=false
       bIsEditable=true
       bIsStaticSwitch=true
       bIsVisible=true
       type {classStructOrEnum {refPath=&ENiagara_TransformOrder}}
      }
      Scale Mode {
       bIsDynamic=false
       bIsEditable=true
       bIsStaticSwitch=true
       bIsVisible=true
       type {classStructOrEnum {refPath=&ENiagara_ScaleMode}}
      }
      Non Uniform Scale {bIsDynamic=false, bIsEditable=true, bIsStaticSwitch=false, bIsVisible=true, type {classStructOrEnum {refPath=&Vector3f}}}
      UseOwnerScale {bIsDynamic=false, bIsEditable=true, bIsStaticSwitch=false, bIsVisible=true, type {classStructOrEnum {refPath=&NiagaraBool}}}
      Apply Owner Scale {bIsDynamic=false, bIsEditable=true, bIsStaticSwitch=false, bIsVisible=true, type {classStructOrEnum {refPath=&Vector3f}}}
      Rotation Mode {
       bIsDynamic=false
       bIsEditable=true
       bIsStaticSwitch=true
       bIsVisible=true
       type {classStructOrEnum {refPath=&ENiagara_RotationMode}}
      }
      Rotation Axis {
       bIsDynamic=false
       bIsEditable=false
       bIsStaticSwitch=false
       bIsVisible=false
       type {classStructOrEnum {refPath=&Vector3f}}
      }
      Rotation Matrix {
       bIsDynamic=false
       bIsEditable=false
       bIsStaticSwitch=false
       bIsVisible=false
       type {classStructOrEnum {refPath=&NiagaraMatrix}}
      }
      Rotation Quaternion {bIsDynamic=false, bIsEditable=false, bIsStaticSwitch=false, bIsVisible=false, type {classStructOrEnum {refPath=&Quat4f}}}
      Yaw / Pitch / Roll {
       bIsDynamic=false
       bIsEditable=false
       bIsStaticSwitch=false
       bIsVisible=false
       type {classStructOrEnum {refPath=&Vector3f}}
      }
      Rotation Angle {
       bIsDynamic=false
       bIsEditable=false
       bIsStaticSwitch=false
       bIsVisible=false
       type {classStructOrEnum {refPath=&NiagaraFloat}}
      }
      Invert Rotation Quaternion {bIsDynamic=false, bIsEditable=false, bIsStaticSwitch=false, bIsVisible=false, type {classStructOrEnum {refPath=&NiagaraBool}}}
      Additional Quaternion Rotation {bIsDynamic=false, bIsEditable=false, bIsStaticSwitch=false, bIsVisible=false, type {classStructOrEnum {refPath=&NiagaraBool}}}
      Additional Yaw / Pitch / Roll {
       bIsDynamic=false
       bIsEditable=false
       bIsStaticSwitch=false
       bIsVisible=false
       type {classStructOrEnum {refPath=&Vector3f}}
      }
      Rotation Angle Type {
       bIsDynamic=false
       bIsEditable=false
       bIsStaticSwitch=true
       bIsVisible=false
       type {classStructOrEnum {refPath=&ENiagara_AngleInput}}
      }
      Rotation Coordinate Space {
       bIsDynamic=false
       bIsEditable=true
       bIsStaticSwitch=false
       bIsVisible=true
       type {classStructOrEnum {refPath=&ENiagaraCoordinateSpace}}
      }
      Offset Mode {
       bIsDynamic=false
       bIsEditable=true
       bIsStaticSwitch=true
       bIsVisible=true
       type {classStructOrEnum {refPath=&ENiagara_OffsetMode}}
      }
      Shape Origin {
       bIsDynamic=false
       bIsEditable=true
       bIsStaticSwitch=false
       bIsVisible=true
       type {classStructOrEnum {refPath=&NiagaraPosition}}
      }
      Offset {
       bIsDynamic=false
       bIsEditable=false
       bIsStaticSwitch=false
       bIsVisible=false
       type {classStructOrEnum {refPath=&Vector3f}}
      }
      Offset Coordinate Space {
       bIsDynamic=false
       bIsEditable=false
       bIsStaticSwitch=false
       bIsVisible=false
       type {classStructOrEnum {refPath=&ENiagaraCoordinateSpace}}
      }
      Custom Transform Matrix {
       bIsDynamic=false
       bIsEditable=false
       bIsStaticSwitch=false
       bIsVisible=false
       type {classStructOrEnum {refPath=&NiagaraMatrix}}
      }
      Transform Method {
       bIsDynamic=false
       bIsEditable=true
       bIsStaticSwitch=true
       bIsVisible=true
       type {classStructOrEnum {refPath=&ENiagara_TransformType}}
      }
     ]
     moduleScript {refPath=&ShapeLocation}
    }
    {
     bIsSetParametersModule=false
     enabled=true
     moduleName=InitialMeshOrientation
     inputs [
      Mesh Orientation Mode {
       bIsDynamic=false
       bIsEditable=true
       bIsStaticSwitch=true
       bIsVisible=true
       type {classStructOrEnum {refPath=&ENiagara_MeshOrientationOptions}}
      }
      Orientation Quaternion {bIsDynamic=false, bIsEditable=false, bIsStaticSwitch=false, bIsVisible=false, type {classStructOrEnum {refPath=&Quat4f}}}
      Orientation Matrix {
       bIsDynamic=false
       bIsEditable=false
       bIsStaticSwitch=false
       bIsVisible=false
       type {classStructOrEnum {refPath=&NiagaraMatrix}}
      }
      Orientation Vector {bIsDynamic=false, bIsEditable=true, bIsStaticSwitch=false, bIsVisible=true, type {classStructOrEnum {refPath=&Vector3f}}}
      Orientation Coordinate Space {
       bIsDynamic=false
       bIsEditable=true
       bIsStaticSwitch=true
       bIsVisible=true
       type {classStructOrEnum {refPath=&ENiagaraCoordinateSpace}}
      }
      Mesh Axis To Orient {bIsDynamic=false, bIsEditable=true, bIsStaticSwitch=false, bIsVisible=true, type {classStructOrEnum {refPath=&Vector3f}}}
      Random Seed {
       bIsDynamic=false
       bIsEditable=false
       bIsStaticSwitch=false
       bIsVisible=false
       type {classStructOrEnum {refPath=&NiagaraInt32}}
      }
      Use Rotation Vector {bIsDynamic=false, bIsEditable=true, bIsStaticSwitch=false, bIsVisible=true, type {classStructOrEnum {refPath=&NiagaraBool}}}
      Rotation {bIsDynamic=false, bIsEditable=true, bIsStaticSwitch=false, bIsVisible=true, type {classStructOrEnum {refPath=&Vector3f}}}
      Rotation Coordinate Space {
       bIsDynamic=false
       bIsEditable=true
       bIsStaticSwitch=true
       bIsVisible=true
       type {classStructOrEnum {refPath=&ENiagaraMeshTransforms}}
      }
     ]
     moduleScript {refPath=&InitialMeshOrientation}
    }
    {
     bIsSetParametersModule=false
     enabled=true
     moduleName=DynamicMaterialParameters
     inputs [
      Write Parameter Index 0 {bIsDynamic=false, bIsEditable=true, bIsStaticSwitch=true, bIsVisible=true, type {classStructOrEnum {refPath=&NiagaraBool}}}
      Module.Index0Param0Write {bIsDynamic=false, bIsEditable=true, bIsStaticSwitch=true, bIsVisible=true, type {classStructOrEnum {refPath=&NiagaraBool}}}
      Index 0 Param 1 {bIsDynamic=false, bIsEditable=true, bIsStaticSwitch=false, bIsVisible=true, type {classStructOrEnum {refPath=&NiagaraFloat}}}
      Module.Index0Param1Write {bIsDynamic=false, bIsEditable=true, bIsStaticSwitch=true, bIsVisible=true, type {classStructOrEnum {refPath=&NiagaraBool}}}
      Index 0 Param 2 {bIsDynamic=false, bIsEditable=true, bIsStaticSwitch=false, bIsVisible=true, type {classStructOrEnum {refPath=&NiagaraFloat}}}
      Module.Index0Param2Write {bIsDynamic=false, bIsEditable=true, bIsStaticSwitch=true, bIsVisible=true, type {classStructOrEnum {refPath=&NiagaraBool}}}
      Index 0 Param 3 {bIsDynamic=true, bIsEditable=true, bIsStaticSwitch=false, bIsVisible=true, type {classStructOrEnum {refPath=&NiagaraFloat}}}
      Module.Index0Param3Write {bIsDynamic=false, bIsEditable=true, bIsStaticSwitch=true, bIsVisible=true, type {classStructOrEnum {refPath=&NiagaraBool}}}
      Index 0 Param 4 {bIsDynamic=false, bIsEditable=false, bIsStaticSwitch=false, bIsVisible=true, type {classStructOrEnum {refPath=&NiagaraFloat}}}
      Write Parameter Index 1 {bIsDynamic=false, bIsEditable=true, bIsStaticSwitch=true, bIsVisible=true, type {classStructOrEnum {refPath=&NiagaraBool}}}
      Index 1 Param 1 {
       bIsDynamic=false
       bIsEditable=false
       bIsStaticSwitch=false
       bIsVisible=false
       type {classStructOrEnum {refPath=&NiagaraFloat}}
      }
      Index 1 Param 2 {
       bIsDynamic=false
       bIsEditable=false
       bIsStaticSwitch=false
       bIsVisible=false
       type {classStructOrEnum {refPath=&NiagaraFloat}}
      }
      Index 1 Param 3 {
       bIsDynamic=false
       bIsEditable=false
       bIsStaticSwitch=false
       bIsVisible=false
       type {classStructOrEnum {refPath=&NiagaraFloat}}
      }
      Index 1 Param 4 {
       bIsDynamic=false
       bIsEditable=false
       bIsStaticSwitch=false
       bIsVisible=false
       type {classStructOrEnum {refPath=&NiagaraFloat}}
      }
      Write Parameter Index 2 {bIsDynamic=false, bIsEditable=true, bIsStaticSwitch=true, bIsVisible=true, type {classStructOrEnum {refPath=&NiagaraBool}}}
      Index 2 Param 1 {
       bIsDynamic=false
       bIsEditable=false
       bIsStaticSwitch=false
       bIsVisible=false
       type {classStructOrEnum {refPath=&NiagaraFloat}}
      }
      Index 2 Param 2 {
       bIsDynamic=false
       bIsEditable=false
       bIsStaticSwitch=false
       bIsVisible=false
       type {classStructOrEnum {refPath=&NiagaraFloat}}
      }
      Index 2 Param 3 {
       bIsDynamic=false
       bIsEditable=false
       bIsStaticSwitch=false
       bIsVisible=false
       type {classStructOrEnum {refPath=&NiagaraFloat}}
      }
      Index 2 Param 4 {
       bIsDynamic=false
       bIsEditable=false
       bIsStaticSwitch=false
       bIsVisible=false
       type {classStructOrEnum {refPath=&NiagaraFloat}}
      }
      Write Parameter Index 3 {bIsDynamic=false, bIsEditable=true, bIsStaticSwitch=true, bIsVisible=true, type {classStructOrEnum {refPath=&NiagaraBool}}}
      Index 3 Param 1 {
       bIsDynamic=false
       bIsEditable=false
       bIsStaticSwitch=false
       bIsVisible=false
       type {classStructOrEnum {refPath=&NiagaraFloat}}
      }
      Index 3 Param 2 {
       bIsDynamic=false
       bIsEditable=false
       bIsStaticSwitch=false
       bIsVisible=false
       type {classStructOrEnum {refPath=&NiagaraFloat}}
      }
      Index 3 Param 3 {
       bIsDynamic=false
       bIsEditable=false
       bIsStaticSwitch=false
       bIsVisible=false
       type {classStructOrEnum {refPath=&NiagaraFloat}}
      }
      Index 3 Param 4 {
       bIsDynamic=false
       bIsEditable=false
       bIsStaticSwitch=false
       bIsVisible=false
       type {classStructOrEnum {refPath=&NiagaraFloat}}
      }
     ]
     moduleScript {refPath=&DynamicMaterialParameters}
    }
    {
     bIsSetParametersModule=false
     enabled=true
     moduleName=AddVelocity
     inputs [
      Velocity Mode {
       bIsDynamic=false
       bIsEditable=true
       bIsStaticSwitch=true
       bIsVisible=true
       type {classStructOrEnum {refPath=&ENiagara_VelocityMode}}
      }
      Velocity {bIsDynamic=true, bIsEditable=false, bIsStaticSwitch=false, bIsVisible=false, type {classStructOrEnum {refPath=&Vector3f}}}
      Velocity Speed {bIsDynamic=true, bIsEditable=true, bIsStaticSwitch=false, bIsVisible=true, type {classStructOrEnum {refPath=&NiagaraFloat}}}
      Velocity Speed Scale {bIsDynamic=true, bIsEditable=false, bIsStaticSwitch=false, bIsVisible=false, type {classStructOrEnum {refPath=&NiagaraFloat}}}
      Module.Use Falloff Distance {bIsDynamic=false, bIsEditable=true, bIsStaticSwitch=false, bIsVisible=true, type {classStructOrEnum {refPath=&NiagaraBool}}}
      Constrain To Radius {bIsDynamic=false, bIsEditable=false, bIsStaticSwitch=false, bIsVisible=true, type {classStructOrEnum {refPath=&NiagaraFloat}}}
      Distribution Along Cone Axis {
       bIsDynamic=false
       bIsEditable=false
       bIsStaticSwitch=false
       bIsVisible=false
       type {classStructOrEnum {refPath=&NiagaraFloat}}
      }
      Radius Falloff Near / Far {
       bIsDynamic=false
       bIsEditable=false
       bIsStaticSwitch=false
       bIsVisible=false
       type {classStructOrEnum {refPath=&Vector2f}}
      }
      Radius Falloff Exponent {
       bIsDynamic=false
       bIsEditable=false
       bIsStaticSwitch=false
       bIsVisible=false
       type {classStructOrEnum {refPath=&NiagaraFloat}}
      }
      Speed Falloff From Cone Axis {
       bIsDynamic=false
       bIsEditable=false
       bIsStaticSwitch=false
       bIsVisible=false
       type {classStructOrEnum {refPath=&NiagaraFloat}}
      }
      Invert Velocity Falloff {bIsDynamic=false, bIsEditable=false, bIsStaticSwitch=false, bIsVisible=false, type {classStructOrEnum {refPath=&NiagaraBool}}}
      Override Seed {bIsDynamic=false, bIsEditable=true, bIsStaticSwitch=false, bIsVisible=true, type {classStructOrEnum {refPath=&NiagaraBool}}}
      Random Seed {bIsDynamic=false, bIsEditable=false, bIsStaticSwitch=false, bIsVisible=true, type {classStructOrEnum {refPath=&NiagaraInt32}}}
      Fixed Random Seed {bIsDynamic=false, bIsEditable=false, bIsStaticSwitch=true, bIsVisible=false, type {classStructOrEnum {refPath=&NiagaraBool}}}
      Randomness Mode {
       bIsDynamic=false
       bIsEditable=true
       bIsStaticSwitch=true
       bIsVisible=true
       type {classStructOrEnum {refPath=&ENiagaraRandomnessMode}}
      }
      Cone Axis {
       bIsDynamic=false
       bIsEditable=false
       bIsStaticSwitch=false
       bIsVisible=false
       type {classStructOrEnum {refPath=&Vector3f}}
      }
      Cone Angle {
       bIsDynamic=false
       bIsEditable=false
       bIsStaticSwitch=false
       bIsVisible=false
       type {classStructOrEnum {refPath=&NiagaraFloat}}
      }
      Inner Cone Angle {
       bIsDynamic=false
       bIsEditable=false
       bIsStaticSwitch=false
       bIsVisible=false
       type {classStructOrEnum {refPath=&NiagaraFloat}}
      }
      Cone Angle Mode {
       bIsDynamic=false
       bIsEditable=false
       bIsStaticSwitch=true
       bIsVisible=false
       type {classStructOrEnum {refPath=&ENiagara_AngleInput}}
      }
      Rotation Mode {
       bIsDynamic=false
       bIsEditable=false
       bIsStaticSwitch=true
       bIsVisible=false
       type {classStructOrEnum {refPath=&ENiagara_RotationMode}}
      }
      Rotation Axis {
       bIsDynamic=false
       bIsEditable=false
       bIsStaticSwitch=false
       bIsVisible=false
       type {classStructOrEnum {refPath=&Vector3f}}
      }
      Yaw / Pitch / Roll {
       bIsDynamic=false
       bIsEditable=false
       bIsStaticSwitch=false
       bIsVisible=false
       type {classStructOrEnum {refPath=&Vector3f}}
      }
      Rotation Quaternion {bIsDynamic=false, bIsEditable=false, bIsStaticSwitch=false, bIsVisible=false, type {classStructOrEnum {refPath=&Quat4f}}}
      Rotation Matrix {
       bIsDynamic=false
       bIsEditable=false
       bIsStaticSwitch=false
       bIsVisible=false
       type {classStructOrEnum {refPath=&NiagaraMatrix}}
      }
      Rotation Angle {
       bIsDynamic=false
       bIsEditable=false
       bIsStaticSwitch=false
       bIsVisible=false
       type {classStructOrEnum {refPath=&NiagaraFloat}}
      }
      Invert Rotation Quaternion {bIsDynamic=false, bIsEditable=false, bIsStaticSwitch=false, bIsVisible=false, type {classStructOrEnum {refPath=&NiagaraBool}}}
      Additional Quaternion Rotation {bIsDynamic=false, bIsEditable=false, bIsStaticSwitch=false, bIsVisible=false, type {classStructOrEnum {refPath=&NiagaraBool}}}
      Additional Yaw / Pitch / Roll {
       bIsDynamic=false
       bIsEditable=false
       bIsStaticSwitch=false
       bIsVisible=false
       type {classStructOrEnum {refPath=&Vector3f}}
      }
      Rotation Angle Type {
       bIsDynamic=false
       bIsEditable=false
       bIsStaticSwitch=true
       bIsVisible=false
       type {classStructOrEnum {refPath=&ENiagara_AngleInput}}
      }
      Rotation Coordinate Space {
       bIsDynamic=false
       bIsEditable=false
       bIsStaticSwitch=false
       bIsVisible=false
       type {classStructOrEnum {refPath=&ENiagaraCoordinateSpace}}
      }
      Velocity Origin {
       bIsDynamic=false
       bIsEditable=true
       bIsStaticSwitch=false
       bIsVisible=true
       type {classStructOrEnum {refPath=&NiagaraPosition}}
      }
      Origin Offset {bIsDynamic=false, bIsEditable=true, bIsStaticSwitch=false, bIsVisible=true, type {classStructOrEnum {refPath=&Vector3f}}}
      Origin Offset Coordinate Space {
       bIsDynamic=false
       bIsEditable=true
       bIsStaticSwitch=false
       bIsVisible=true
       type {classStructOrEnum {refPath=&ENiagaraCoordinateSpace}}
      }
      Default Position {
       bIsDynamic=false
       bIsEditable=true
       bIsStaticSwitch=false
       bIsVisible=true
       type {classStructOrEnum {refPath=&NiagaraPosition}}
      }
     ]
     moduleScript {refPath=&AddVelocity}
    }
   ]
  }
  particleUpdateScript {
   scriptName=ParticleUpdateScript
   modules [
    {
     bIsSetParametersModule=false
     enabled=true
     moduleName=ParticleState
     inputs [
      Kill Particles When Lifetime Has Elapsed {bIsDynamic=false, bIsEditable=true, bIsStaticSwitch=true, bIsVisible=true, type {classStructOrEnum {refPath=&NiagaraBool}}}
      Loop Particles Lifetime {bIsDynamic=false, bIsEditable=false, bIsStaticSwitch=false, bIsVisible=false, type {classStructOrEnum {refPath=&NiagaraBool}}}
      Let Infinitely Lived Particles Die When Emitter Deactivates {bIsDynamic=false, bIsEditable=false, bIsStaticSwitch=false, bIsVisible=false, type {classStructOrEnum {refPath=&NiagaraBool}}}
      Lifetime {bIsDynamic=false, bIsEditable=true, bIsStaticSwitch=false, bIsVisible=true, type {classStructOrEnum {refPath=&NiagaraFloat}}}
      DeltaTime {bIsDynamic=false, bIsEditable=true, bIsStaticSwitch=false, bIsVisible=true, type {classStructOrEnum {refPath=&NiagaraFloat}}}
     ]
     moduleScript {refPath=&ParticleState}
    }
    {
     bIsSetParametersModule=false
     enabled=true
     moduleName=Drag
     inputs [
      Use Linear Drag {bIsDynamic=false, bIsEditable=true, bIsStaticSwitch=false, bIsVisible=true, type {classStructOrEnum {refPath=&NiagaraBool}}}
      Drag {bIsDynamic=true, bIsEditable=true, bIsStaticSwitch=false, bIsVisible=true, type {classStructOrEnum {refPath=&NiagaraFloat}}}
      Use Rotational Drag {bIsDynamic=false, bIsEditable=true, bIsStaticSwitch=false, bIsVisible=true, type {classStructOrEnum {refPath=&NiagaraBool}}}
      Rotational Drag {bIsDynamic=false, bIsEditable=true, bIsStaticSwitch=false, bIsVisible=true, type {classStructOrEnum {refPath=&NiagaraFloat}}}
      Ignore Mass {bIsDynamic=false, bIsEditable=true, bIsStaticSwitch=false, bIsVisible=true, type {classStructOrEnum {refPath=&NiagaraBool}}}
     ]
     moduleScript {refPath=&Drag}
    }
    {
     bIsSetParametersModule=false
     enabled=true
     moduleName=SolveForcesAndVelocity
     inputs [
      Clamp Velocity {bIsDynamic=false, bIsEditable=true, bIsStaticSwitch=false, bIsVisible=true, type {classStructOrEnum {refPath=&NiagaraBool}}}
      Speed Limit {bIsDynamic=false, bIsEditable=false, bIsStaticSwitch=false, bIsVisible=true, type {classStructOrEnum {refPath=&NiagaraFloat}}}
      Limit Acceleration {bIsDynamic=false, bIsEditable=true, bIsStaticSwitch=false, bIsVisible=true, type {classStructOrEnum {refPath=&NiagaraBool}}}
      Acceleration Limit {bIsDynamic=false, bIsEditable=false, bIsStaticSwitch=false, bIsVisible=true, type {classStructOrEnum {refPath=&NiagaraFloat}}}
      Manually Enable Rotational Solver {bIsDynamic=false, bIsEditable=true, bIsStaticSwitch=false, bIsVisible=true, type {classStructOrEnum {refPath=&NiagaraBool}}}
      Write to Intrinsic Properties {bIsDynamic=false, bIsEditable=true, bIsStaticSwitch=false, bIsVisible=true, type {classStructOrEnum {refPath=&NiagaraBool}}}
      Write to Presolve Properties {bIsDynamic=false, bIsEditable=true, bIsStaticSwitch=true, bIsVisible=true, type {classStructOrEnum {refPath=&NiagaraBool}}}
      DeltaTime {bIsDynamic=false, bIsEditable=true, bIsStaticSwitch=false, bIsVisible=true, type {classStructOrEnum {refPath=&NiagaraFloat}}}
      Mass {bIsDynamic=false, bIsEditable=true, bIsStaticSwitch=false, bIsVisible=true, type {classStructOrEnum {refPath=&NiagaraFloat}}}
      Position {
       bIsDynamic=false
       bIsEditable=true
       bIsStaticSwitch=false
       bIsVisible=true
       type {classStructOrEnum {refPath=&NiagaraPosition}}
      }
      Velocity {bIsDynamic=false, bIsEditable=true, bIsStaticSwitch=false, bIsVisible=true, type {classStructOrEnum {refPath=&Vector3f}}}
      PreviousVelocity {bIsDynamic=false, bIsEditable=true, bIsStaticSwitch=false, bIsVisible=true, type {classStructOrEnum {refPath=&Vector3f}}}
      Force {bIsDynamic=false, bIsEditable=true, bIsStaticSwitch=false, bIsVisible=true, type {classStructOrEnum {refPath=&Vector3f}}}
      Mesh Orientation {bIsDynamic=false, bIsEditable=false, bIsStaticSwitch=false, bIsVisible=false, type {classStructOrEnum {refPath=&Quat4f}}}
      Rotational Velocity {
       bIsDynamic=false
       bIsEditable=false
       bIsStaticSwitch=false
       bIsVisible=false
       type {classStructOrEnum {refPath=&Vector3f}}
      }
      Rotational Force {
       bIsDynamic=false
       bIsEditable=false
       bIsStaticSwitch=false
       bIsVisible=false
       type {classStructOrEnum {refPath=&Vector3f}}
      }
      Rotational Inertia {
       bIsDynamic=false
       bIsEditable=false
       bIsStaticSwitch=false
       bIsVisible=false
       type {classStructOrEnum {refPath=&NiagaraFloat}}
      }
     ]
     moduleScript {refPath=&SolveForcesAndVelocity}
    }
    {
     bIsSetParametersModule=false
     enabled=true
     moduleName=ScaleMeshSize
     inputs [
      Scale Factor {bIsDynamic=true, bIsEditable=true, bIsStaticSwitch=false, bIsVisible=true, type {classStructOrEnum {refPath=&Vector3f}}}
      Initial Mesh Scale {bIsDynamic=false, bIsEditable=true, bIsStaticSwitch=false, bIsVisible=true, type {classStructOrEnum {refPath=&Vector3f}}}
     ]
     moduleScript {refPath=&ScaleMeshSize}
    }
    {
     bIsSetParametersModule=false
     enabled=true
     moduleName=ScaleColor
     inputs [
      Scale Mode {
       bIsDynamic=false
       bIsEditable=true
       bIsStaticSwitch=true
       bIsVisible=true
       type {classStructOrEnum {refPath=&ENiagaraScaleColorMode}}
      }
      Linear Color Curve {
       bIsDynamic=false
       bIsEditable=false
       bIsStaticSwitch=false
       bIsVisible=false
       type {classStructOrEnum {refPath=&NiagaraDataInterfaceColorCurve}}
      }
      ScaleRGB {bIsDynamic=false, bIsEditable=true, bIsStaticSwitch=false, bIsVisible=true, type {classStructOrEnum {refPath=&NiagaraBool}}}
      Scale RGB {bIsDynamic=false, bIsEditable=true, bIsStaticSwitch=false, bIsVisible=true, type {classStructOrEnum {refPath=&Vector3f}}}
      ScaleRGBA {bIsDynamic=false, bIsEditable=false, bIsStaticSwitch=false, bIsVisible=false, type {classStructOrEnum {refPath=&NiagaraBool}}}
      Scale RGBA {
       bIsDynamic=false
       bIsEditable=false
       bIsStaticSwitch=false
       bIsVisible=false
       type {classStructOrEnum {refPath=&Vector4f}}
      }
      Curve Index {
       bIsDynamic=false
       bIsEditable=false
       bIsStaticSwitch=false
       bIsVisible=false
       type {classStructOrEnum {refPath=&NiagaraFloat}}
      }
      ScaleA {bIsDynamic=false, bIsEditable=true, bIsStaticSwitch=false, bIsVisible=true, type {classStructOrEnum {refPath=&NiagaraBool}}}
      Scale Alpha {bIsDynamic=true, bIsEditable=true, bIsStaticSwitch=false, bIsVisible=true, type {classStructOrEnum {refPath=&NiagaraFloat}}}
      Color Value To Scale {
       bIsDynamic=false
       bIsEditable=true
       bIsStaticSwitch=false
       bIsVisible=true
       type {classStructOrEnum {refPath=&LinearColor}}
      }
     ]
     moduleScript {refPath=&ScaleColor}
    }
   ]
  }
  rendererClasses [{refPath=&NiagaraMeshRendererProperties}]
  renderers [
   {rendererIndex=0, rendererClass {refPath=&NiagaraMeshRendererProperties}}
   {rendererIndex=1, rendererClass {refPath=&NiagaraMeshRendererProperties}}
  ]
 }
 ToonImpactSmoke {
  emitterSpawnScript {scriptName=None, modules []}
  emitterUpdateScript {scriptName=None, modules []}
  particleSpawnScript {scriptName=None, modules []}
  particleUpdateScript {scriptName=None, modules []}
  rendererClasses []
  renderers []
 }
}
summary {
 systemName=NS_Emote_MatchaSpare_AnvilGround
 emitters [
  {emitterName=Dust, rendererClasses [{refPath=&NiagaraSpriteRendererProperties}]}
  {emitterName=Swoosh, rendererClasses [{refPath=&NiagaraMeshRendererProperties}]}
  {emitterName=Puffs, rendererClasses [{refPath=&NiagaraMeshRendererProperties}]}
  {emitterName=Debris, rendererClasses [{refPath=&NiagaraMeshRendererProperties}]}
  {emitterName=ToonImpactSmoke, rendererClasses []}
  {emitterName=Ground_Cracks001, rendererClasses [{refPath=&NiagaraDecalRendererProperties}]}
  {emitterName=Ground_Crater, rendererClasses [{refPath=&NiagaraDecalRendererProperties}]}
 ]
 userVariables [
  User.bIsFrontEnd {
   description=""
   defaultValue {struct {refPath=&NiagaraBool}, value {value=0}}
   type {classStructOrEnum {refPath=&NiagaraBool}}
  }
  User.bIsFrontEndPreview {
   description=""
   defaultValue {struct {refPath=&NiagaraBool}, value {value=0}}
   type {classStructOrEnum {refPath=&NiagaraBool}}
  }
 ]
}
