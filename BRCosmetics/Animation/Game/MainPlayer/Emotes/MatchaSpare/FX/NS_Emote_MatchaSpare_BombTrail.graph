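# Niagara ribbon trail system for the bomb in the MatchaSpare emote, with two active emitters and a disabled original.
# FN AnimTrail modules spawn ribbons between bone or socket pairs, then fade alpha and taper width over short lifetimes.
# Bomb001 switches color, width and lifetime by spawn group for translucent green and grey trails.
# Bomb002 adds translucent black ribbons; both active emitters drive material parameters through dynamic parameters.
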NiagaraSystem /BRCosmetics/Animation/Game/MainPlayer/Emotes/MatchaSpare/FX/NS_Emote_MatchaSpare_BombTrail
dependencies {
 usedDataInterfaces [&NiagaraDataInterfaceColorCurve, &NiagaraDataInterfaceCurve, &NiagaraDataInterfaceSocketReader]
 usedDynamicInputs []
 usedModules [
  &self:SystemSpawnScript.SystemScriptSource.SystemScriptGraph.NiagaraNodeAssignment_1.SetVariables_#1
  &SystemState
 ]
}
emitters {
 Bomb {
  bEnabled=false
  emitterSpawnScript [&FN_RibbonIDOffsetUtility()]
  emitterUpdateScript [
   &EmitterState()
   &FN_SpawnAnimTrail()
  ]
  particleSpawnScript [
   &FN_InitializeAnimTrailBoneIndecies()
   &FN_InitializeAnimTrailBoneIndecies(AnimTrail Selection=1, First Bone or Socket Index=2, Second Bone or Socket Index=3) {enabled=false, module=FN_InitializeAnimTrailBoneIndecies001}
   &FN_InitializeAnimTrailBoneIndecies(AnimTrail Selection=2, First Bone or Socket Index=4, Second Bone or Socket Index=5) {enabled=false, module=FN_InitializeAnimTrailBoneIndecies002}
   &FN_InitializeAnimTrailBoneIndecies(First Bone or Socket Index=6, Second Bone or Socket Index=7) {enabled=false, module=FN_InitializeAnimTrailBoneIndecies003}
   &FN_InitializeAnimTrail(Color=<rgba 1 0.1411 0.0905 1>, Lifetime=0.225)
   &AddVelocity(Velocity=<0 400 0>) {enabled=false}
  ]
  particleUpdateScript [
   &SolveForcesAndVelocity(Clamp Velocity=false, Limit Acceleration=false)
   &FN_UpdateAnimTrail() {module=FN_UpdateAnimTrail001}
   &ScaleColor(
    Linear Color Curve=DataInterface() {propertyValues {
      AlphaCurve {keys [<at 0 0.45 RCIM_Cubic>,<at 1 0 RCIM_Cubic>]}
      BlueCurve {keys [<at 0 1>,<at 0.2 0.4271>,<at 1 0.1042>]}
      GreenCurve {keys [<at 0 1>,<at 0.2 0.4271>,<at 1 0.1042>]}
      RedCurve {keys [<at 0 1>,<at 0.2 0.4271>,<at 1 0.1042>]}
     }}
    Scale Mode=::NewEnumerator2
   )
   &ScaleRibbonWidth(Ribbon Width Scale=&FloatFromCurve(FloatCurve=DataInterface() {propertyValues {Curve {
        defaultValue=1
        keys [
         {tangentMode=RCTM_User, time=0, value=1}
         {arriveTangent=-0.0405, interpMode=RCIM_Cubic, leaveTangent=-0.0405, tangentMode=RCTM_User, time=0.7009, value=1}
         {arriveTangent=-8.0089, interpMode=RCIM_Cubic, leaveTangent=-8.0089, tangentMode=RCTM_User, time=1, value=0}
        ]
       }}}))
  ]
  rendererClasses [&NiagaraRibbonRendererProperties]
  renderers [
   {rendererClass=&NiagaraRibbonRendererProperties}
   {rendererClass=&NiagaraRibbonRendererProperties, rendererIndex=1}
  ]
 }
 Bomb001 {
  emitterSpawnScript [&FN_RibbonIDOffsetUtility()]
  emitterUpdateScript [
   &EmitterState()
   &FN_SpawnAnimTrail(Number of Anim Trails=2)
  ]
  particleSpawnScript [
   &FN_InitializeAnimTrailBoneIndecies()
   &FN_InitializeAnimTrailBoneIndecies(AnimTrail Selection=1, First Bone or Socket Index=2, Second Bone or Socket Index=3) {module=FN_InitializeAnimTrailBoneIndecies001}
   &FN_InitializeAnimTrailBoneIndecies(AnimTrail Selection=2, First Bone or Socket Index=4, Second Bone or Socket Index=5) {enabled=false, module=FN_InitializeAnimTrailBoneIndecies002}
   &FN_InitializeAnimTrailBoneIndecies(First Bone or Socket Index=6, Second Bone or Socket Index=7) {enabled=false, module=FN_InitializeAnimTrailBoneIndecies003}
   &FN_InitializeAnimTrail(
    AnimTrailWidth=&MergeHelper_FloatSwitch(0=2, 1=3, 2=?, 3=?, Index=&SpawnGroup(), Num=::NewEnumerator0)
    Color=&MergeHelper_ColorSwitch(
     0=<rgba 0.2195 0.5841 0.1221 0.5>
     1=<rgba 0.1198 0.1198 0.1198 0.7>
     2=?
     3=?
     Index=&SpawnGroup()
     Num=::NewEnumerator0
    )
    Lifetime=&MergeHelper_FloatSwitch(0=0.2, 1=0.12, 2=?, 3=?, Index=&SpawnGroup(), Num=::NewEnumerator0)
   )
   &AddVelocity(Velocity=<0 400 0>) {enabled=false}
  ]
  particleUpdateScript [
   &InheritSourceMovement(ApplyForce=false, ApplyPosition=true, ApplyPositionFalloff=false) {enabled=false}
   &SolveForcesAndVelocity(Clamp Velocity=false, Limit Acceleration=false)
   &FN_UpdateAnimTrail() {module=FN_UpdateAnimTrail001}
   &ScaleColor(
    Scale Alpha=&FloatFromCurve(
     FloatCurve=DataInterface() {propertyValues {Curve {
        defaultValue=1
        keys [{interpMode=RCIM_Cubic, tangentMode=RCTM_User, time=0, value=1}, {interpMode=RCIM_Cubic, tangentMode=RCTM_User, time=1, value=0}]
       }}}
     Scale Curve=0.7
    )
    ScaleA=true
    ScaleRGB=true
   )
   &ScaleRibbonWidth(Ribbon Width Scale=&FloatFromCurve(FloatCurve=DataInterface() {propertyValues {Curve {
        defaultValue=1
        keys [
         {tangentMode=RCTM_User, time=0, value=1}
         {arriveTangent=-0.0405, interpMode=RCIM_Cubic, leaveTangent=-0.0405, tangentMode=RCTM_User, time=0.7009, value=1}
         {arriveTangent=-8.0089, interpMode=RCIM_Cubic, leaveTangent=-8.0089, tangentMode=RCTM_User, time=1, value=0}
        ]
       }}}))
   &DynamicMaterialParameters(
    Index 0 Param 1=&FloatFromCurve(FloatCurve=DataInterface() {propertyValues {Curve {
        defaultValue=1
        keys [{interpMode=RCIM_Cubic, tangentMode=RCTM_User, time=0, value=1}, {interpMode=RCIM_Cubic, tangentMode=RCTM_User, time=1, value=0}]
       }}})
    Index 0 Param 2=1
    Index 0 Param 3=0.2
    Index 0 Param 4=1
    Module.Index0Float0Write=true
    Module.Index0Float1Write=true
    Module.Index0Float2Write=true
    Module.Index0Float3Write=true
   )
  ]
  rendererClasses [&NiagaraRibbonRendererProperties]
  renderers [
   {rendererClass=&NiagaraRibbonRendererProperties}
   {rendererClass=&NiagaraRibbonRendererProperties, rendererIndex=1}
  ]
 }
 Bomb002 {
  emitterSpawnScript [&FN_RibbonIDOffsetUtility()]
  emitterUpdateScript [
   &EmitterState()
   &FN_SpawnAnimTrail(Number of Anim Trails=2)
  ]
  particleSpawnScript [
   &FN_InitializeAnimTrailBoneIndecies()
   &FN_InitializeAnimTrailBoneIndecies(AnimTrail Selection=1, First Bone or Socket Index=2, Second Bone or Socket Index=3) {enabled=false, module=FN_InitializeAnimTrailBoneIndecies001}
   &FN_InitializeAnimTrailBoneIndecies(AnimTrail Selection=2, First Bone or Socket Index=4, Second Bone or Socket Index=5) {enabled=false, module=FN_InitializeAnimTrailBoneIndecies002}
   &FN_InitializeAnimTrailBoneIndecies(First Bone or Socket Index=6, Second Bone or Socket Index=7) {enabled=false, module=FN_InitializeAnimTrailBoneIndecies003}
   &FN_InitializeAnimTrail(AnimTrailWidth=2, Color=<rgba 0 0 0 0.4>, Lifetime=0.2)
   &AddVelocity(Velocity=<0 400 0>) {enabled=false}
  ]
  particleUpdateScript [
   &InheritSourceMovement(ApplyForce=false, ApplyPosition=true, ApplyPositionFalloff=false) {enabled=false}
   &SolveForcesAndVelocity(Clamp Velocity=false, Limit Acceleration=false)
   &FN_UpdateAnimTrail() {module=FN_UpdateAnimTrail001}
   &ScaleColor(
    Scale Alpha=&FloatFromCurve(FloatCurve=DataInterface() {propertyValues {Curve {
        defaultValue=1
        keys [{interpMode=RCIM_Cubic, tangentMode=RCTM_User, time=0, value=1}, {interpMode=RCIM_Cubic, tangentMode=RCTM_User, time=1, value=0}]
       }}})
    ScaleA=true
    ScaleRGB=true
   )
   &ScaleRibbonWidth(Ribbon Width Scale=&FloatFromCurve(FloatCurve=DataInterface() {propertyValues {Curve {
        defaultValue=1
        keys [
         {tangentMode=RCTM_User, time=0, value=1}
         {arriveTangent=-0.0405, interpMode=RCIM_Cubic, leaveTangent=-0.0405, tangentMode=RCTM_User, time=0.7009, value=1}
         {arriveTangent=-8.0089, interpMode=RCIM_Cubic, leaveTangent=-8.0089, tangentMode=RCTM_User, time=1, value=0}
        ]
       }}}))
   &DynamicMaterialParameters(
    Index 0 Param 1=&FloatFromCurve(FloatCurve=DataInterface() {propertyValues {Curve {
        defaultValue=1
        keys [{interpMode=RCIM_Cubic, tangentMode=RCTM_User, time=0, value=1}, {interpMode=RCIM_Cubic, tangentMode=RCTM_User, time=1, value=0}]
       }}})
    Index 0 Param 2=1
    Index 0 Param 3=0.2
    Index 0 Param 4=1
    Module.Index0Float0Write=true
    Module.Index0Float1Write=true
    Module.Index0Float2Write=true
    Module.Index0Float3Write=true
   )
  ]
  rendererClasses [&NiagaraRibbonRendererProperties]
  renderers [
   {rendererClass=&NiagaraRibbonRendererProperties}
   {rendererClass=&NiagaraRibbonRendererProperties, rendererIndex=1}
  ]
 }
}
summary {
 systemName=NS_Emote_MatchaSpare_BombTrail
 emitters [
  {bEnabled=false, emitterName=Bomb, rendererClasses [&NiagaraRibbonRendererProperties]}
  {emitterName=Bomb001, rendererClasses [&NiagaraRibbonRendererProperties]}
  {emitterName=Bomb002, rendererClasses [&NiagaraRibbonRendererProperties]}
 ]
 userVariables [
  User.bIsFrontEnd {type=NiagaraBool}
  User.bIsFrontEndPreview {type=NiagaraBool}
  User.PhysicalSurface {type=NiagaraInt32}
  User.RibbonIDOffset {type=NiagaraInt32}
 ]
}
