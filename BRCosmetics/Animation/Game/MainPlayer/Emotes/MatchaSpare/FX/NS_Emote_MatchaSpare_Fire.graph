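# Niagara system for the fire effect of the MatchaSpare emote, attached to bones or sockets of a skeletal mesh.
# A burst emitter draws the main flame as sprites sized by two User floats, plus mesh and flickering light renderers.
# FlameSprites spawn in three groups indexed to separate bones, joined by curl-noise FireBits, a pulsing Glow and a fading ParticleDecal sprite.
# Several other impact and glow emitters have no renderers.
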NiagaraSystem /BRCosmetics/Animation/Game/MainPlayer/Emotes/MatchaSpare/FX/NS_Emote_MatchaSpare_Fire
dependencies {
 usedDataInterfaces [
  &NiagaraDataInterfaceCurve
  &NiagaraDataInterfaceMeshRendererInfo
  &NiagaraDataInterfaceVector2DCurve
  &NiagaraDataInterfaceColorCurve
  &NiagaraDataInterfaceArrayFloat
  &NiagaraDataInterfaceSkeletalMesh
 ]
 usedDynamicInputs [
  &SelectFloatFromArray
  &MakeFloatFromVector
  &FNTimeOfDayColor
 ]
 usedModules [
  &self:SystemSpawnScript.SystemScriptSource.SystemScriptGraph.NiagaraNodeAssignment_0.SetVariables_#1
  &SystemState
  &self:SystemSpawnScript.SystemScriptSource.SystemScriptGraph.NiagaraNodeAssignment_1.SetVariables_#2
 ]
}
emitters {
 DarkGlow {rendererClasses [], renderers []}
 Debris {rendererClasses [], renderers []}
 ExplosionStreaks {rendererClasses [], renderers []}
 FireBits {
  emitterUpdateScript [
   &EmitterState()
   &SpawnRate(
    SpawnRate=&FloatFromCurve(
     FloatCurve=DataInterface() {propertyValues {Curve {keys [<at 0 1 RCIM_Constant>,<at 0.8 0 RCIM_Constant>]}}}
     CurveIndex=$Emitter.NormalizedLoopAge
     Scale Curve=4
    )
    Use Spawn Probability=false
   )
  ]
  particleSpawnScript [
   &FN_SetSpawnGroup(Number of Spawn Groups=1)
   &InitializeParticle(
    Color=<rgba 1 0.6424 0.2147 1>
    Lifetime Max=1.2
    Lifetime Min=0.4
    Lifetime Mode=::NewEnumerator1
    Mesh Scale Mode=::NewEnumerator1
    Mesh Uniform Scale Max=1.5
    Mesh Uniform Scale Min=0.5
    Sprite Size Max=<4 8>
    Sprite Size Min=<1 5>
    Sprite Size Mode=::NewEnumerator2
    UsePositionOffset=false
    Write Material Random=true
   )
   &ShapeLocation(
    Box / Plane Mode:ENiagara_BoxPlaneMode=::NewEnumerator1
    MaskSpawnGroup=false
    Override Seed=false
    Plane Size:Vector2f=<45 0>
    Shape Primitive=::NewEnumerator2
    UseOwnerScale=true
   ) {module=ShapeLocation001}
   &AddVelocity(Velocity Speed Scale=2.5)
  ]
  particleUpdateScript [
   &ParticleState()
   &SkeletalMeshLocation(
    Bone / Socket Index=3
    Bone Sampling Mode=::NewEnumerator4
    MaskSpawnGroup=false
    Position Sampling=::NewEnumerator1
    Sampled Position Offset=&AddVectorToPosition(Position=$Particles.ShapeLocation001.ShapePosition, Vector=<0 0 -15>)
    Sampled Position Offset Coordinate Space:ENiagaraSkelMeshTransforms=::NewEnumerator2
    Skeletal Mesh=$System.Skeletal Mesh
   )
   &ScaleMeshSize(Scale Factor=&VectorFromFloat(Value=&FloatFromCurve(FloatCurve=DataInterface() {propertyValues {Curve {keys [
          {arriveTangent=4.1106, interpMode=RCIM_Cubic, leaveTangent=4.1106, tangentMode=RCTM_User, time=0, value=0}
          {interpMode=RCIM_Cubic, tangentMode=RCTM_User, time=0.35, value=1}
          {tangentMode=RCTM_User, time=1, value=0}
         ]}}})))
   &ScaleSpriteSize(Uniform Curve Sprite Scale=DataInterface() {propertyValues {Curve {keys [
        {arriveTangent=4.677, interpMode=RCIM_Cubic, leaveTangent=4.677, tangentMode=RCTM_User, time=0, value=0}
        {interpMode=RCIM_Cubic, tangentMode=RCTM_User, time=0.3, value=1}
        {tangentMode=RCTM_User, time=1, value=0}
       ]}}})
   &DynamicMaterialParameters(
    Index 0 Param 1=&RandomRangeFloat(Minimum=0, Maximum=1)
    Module.Param0WriteEnabled=true
    Module.Param1WriteEnabled=true
    Module.Param2WriteEnabled=true
    Module.Param3WriteEnabled=true
   ) {module=DynamicMaterialParameters001}
   &GravityForce(Gravity=<0 0 200>)
   &Drag(Drag=&RandomRangeFloat(Minimum=2, Maximum=3.2), Use Linear Drag=true, Use Rotational Drag=true)
   &CurlNoiseForce(Noise Frequency=10, Noise Strength=150, Pan Noise=true, Pan Noise Field=<1 0 0>)
   &SolveForcesAndVelocity(Clamp Velocity=false, Limit Acceleration=false)
   &FN_Sprite_Attributes(
    Color=<rgba 0.2691 0.2258 0.2031 1>
    Create Color Attribute=true
    Create SpriteSize Attribute=true
    SpriteSize=&Multiply_Vector2D_ByFloat(Float=1.15, Vector2D=$Particles.SpriteSize)
    Write to Module Namespace=true
   )
   /SetVariables(Particles.R2_CamOffset=-5) {isSetParameters=true}
  ]
  rendererClasses [&NiagaraSpriteRendererProperties]
  renderers [
   {rendererClass=&NiagaraSpriteRendererProperties}
   {rendererClass=&NiagaraSpriteRendererProperties, rendererIndex=1}
  ]
 }
 FlameSprites {
  emitterUpdateScript [
   &EmitterState()
   &SpawnRate(
    SpawnRate=&FloatFromCurve(
     FloatCurve=DataInterface() {propertyValues {Curve {keys [<at 0 1>,<at 0.5 1>,<at 0.8 0>]}}}
     CurveIndex=$Emitter.NormalizedLoopAge
     Scale Curve=4
    )
    Use Spawn Probability=false
   )
   &SpawnRate(
    SpawnGroup=1
    SpawnRate=&FloatFromCurve(
     FloatCurve=DataInterface() {propertyValues {Curve {keys [<at 0 1>,<at 0.5 1>,<at 0.8 0>]}}}
     CurveIndex=$Emitter.NormalizedLoopAge
     Scale Curve=4
    )
    Use Spawn Probability=false
   ) {module=SpawnRate001}
   &SpawnRate(
    SpawnGroup=2
    SpawnRate=&FloatFromCurve(
     FloatCurve=DataInterface() {propertyValues {Curve {keys [<at 0 1>,<at 0.5 1>,<at 0.8 0>]}}}
     CurveIndex=$Emitter.NormalizedLoopAge
     Scale Curve=4
    )
    Use Spawn Probability=false
   ) {module=SpawnRate002}
  ]
  particleSpawnScript [
   &InitializeParticle(
    Color=<rgba 1 0.091 0 1>
    Lifetime Max=0.4
    Lifetime Min=0.2
    Lifetime Mode=::NewEnumerator1
    Position Offset=<-50 0 0>
    Sprite Rotation Mode=::NewEnumerator2
    Sprite Size Mode=::NewEnumerator1
    Uniform Sprite Size Max=30
    Uniform Sprite Size Min=20
    UsePositionOffset=false
    Write Material Random=true
   ) {module=InitializeParticle001}
   &ShapeLocation(
    Box / Plane Mode:ENiagara_BoxPlaneMode=::NewEnumerator1
    MaskSpawnGroup=false
    Override Seed=false
    Plane Size:Vector2f=&MergeHelper_Vector2Switch(0=<17.5 0>, 1=<10 0>, 2=<20 0>, 3=?, 4=?, Index=&SpawnGroup(), Num=::NewEnumerator1)
    Shape Primitive=::NewEnumerator2
    UseOwnerScale=true
   ) {module=ShapeLocation001}
   &AddVelocity(
    Scale Added Velocity=<0.8 0.8 0.8>
    Velocity=&MakeVector(
     X=0
     Y=0
     Z=&MergeHelper_FloatSwitch(
      0=&RandomRangeFloat(Minimum=125, Maximum=250)
      1=&RandomRangeFloat(Minimum=75, Maximum=125)
      2=&RandomRangeFloat(Minimum=75, Maximum=125)
      3=?
      Index=&SpawnGroup()
      Num=::NewEnumerator1
     )
    )
   )
   &SpriteFacingAndAlignment(Sprite Facing=<0 0 1>, WriteAlignment=false, WriteFacing=true) {enabled=false}
  ]
  particleUpdateScript [
   /SetVariables(Particles.SubImageIndex=&FloatFromCurve(FloatCurve=DataInterface() {propertyValues {Curve {keys [<at 0 5>,<at 1 15>]}}})) {isSetParameters=true}
   &ParticleState()
   &SkeletalMeshLocation(
    Bone / Socket Index=&SpawnGroup()
    Bone Sampling Mode=::NewEnumerator4
    MaskSpawnGroup=false
    Position Sampling=::NewEnumerator1
    Sampled Position Offset=&AddVectorToPosition(Position=$Particles.ShapeLocation001.ShapePosition, Vector=<0 0 5>)
    Sampled Position Offset Coordinate Space:ENiagaraSkelMeshTransforms=::NewEnumerator2
    Skeletal Mesh=$System.Skeletal Mesh
   )
   &Drag(Drag=4, Use Linear Drag=true, Use Rotational Drag=false)
   &SolveForcesAndVelocity(Clamp Velocity=false, Limit Acceleration=false)
   &ScaleSpriteSize(Scale Factor=&Vector2DFromFloat(Value=&FloatFromCurve(FloatCurve=DataInterface() {propertyValues {Curve {keys [
          {interpMode=RCIM_Cubic, leaveTangent=5.3269, tangentMode=RCTM_Break, time=0, value=0}
          {arriveTangent=0.6, interpMode=RCIM_Cubic, leaveTangent=0.6, time=0.2, value=0.54}
          {arriveTangent=0.0611, interpMode=RCIM_Cubic, tangentMode=RCTM_Break, time=1, value=0.6}
         ]}}})))
   &ScaleColor(
    Scale Alpha=&Multiply_Float(
     A=&FloatFromCurve(FloatCurve=DataInterface() {propertyValues {Curve {keys [<at 0 0>,<at 0.5 1>,<at 1 0>]}}})
     B=&RandomRangeFloat(Minimum=1, Maximum=2)
    )
    Scale RGB=&Multiply_Vector(
     A=<2 2 2>
     B=&MakeVectorFromLinearColorRGB(Color=&ColorFromCurve(ColorCurve=DataInterface() {propertyValues {
         AlphaCurve {keys [<at 0 1 RCIM_Cubic>]}
         BlueCurve {keys [<at 0 0 RCIM_Cubic>,<at 1 1 RCIM_Cubic>]}
         GreenCurve {keys [<at 0 0.0601 RCIM_Cubic>,<at 1 1 RCIM_Cubic>]}
         RedCurve {keys [<at 0 1 RCIM_Cubic>,<at 1 1 RCIM_Cubic>]}
        }}))
    )
    ScaleA=true
    ScaleRGB=true
   ) {module=ScaleColor001}
  ]
  rendererClasses [&NiagaraSpriteRendererProperties]
  renderers [{rendererClass=&NiagaraSpriteRendererProperties}]
 }
 Glow {
  emitterUpdateScript [
   &EmitterState()
   &SpawnBurst_Instantaneous(Use Spawn Probability=false, UseLoopCountLimit=false)
  ]
  particleSpawnScript [
   &InitializeParticle(
    Color=<rgba 5 1.0882 0 0.02>
    Lifetime=4.62
    Sprite Size=<80 100>
    Sprite Size Mode=::NewEnumerator4
    UsePositionOffset=false
    Write Material Random=true
   )
   &CameraOffset(Camera Offset Amount=65)
   /SetVariables(Particles.SpriteAlignment=<0 0 1>) {isSetParameters=true}
   /SetVariables(
    Particles.CameraOffset2=45
    Particles.Position2:NiagaraPosition=&AddVectorToPosition(Position=$Engine.Emitter.SimulationPosition, Vector=&MakeVector(X=0, Y=0, Z=-47.5))
    Particles.SpriteAlignment2:Vector3f=<0 1 0>
    Particles.SpriteFacing2:Vector3f=<0 0 1>
    Particles.SpriteRotation2=55
    Particles.SpriteSize2=<45 60>
   ) {isSetParameters=true}
  ]
  particleUpdateScript [
   &ParticleState()
   &DynamicMaterialParameters(
    Index 0 Param 1=2
    Module.Index0Float0Write=true
    Module.Index0Float1Write=true
    Module.Index0Float2Write=true
    Module.Index0Float3Write=true
   )
   &ScaleColor(
    Scale Alpha=&Multiply_Float(
     A=&FloatFromCurve(FloatCurve=DataInterface() {propertyValues {Curve {
         defaultValue=1
         keys [
          {interpMode=RCIM_Cubic, tangentMode=RCTM_User, time=0, value=0}
          {interpMode=RCIM_Cubic, tangentMode=RCTM_User, time=0.2, value=1}
          {arriveTangent=0.1259, interpMode=RCIM_Cubic, leaveTangent=0.1259, tangentMode=RCTM_User, time=0.8, value=1}
          {tangentMode=RCTM_User, time=1, value=0}
         ]
        }}})
     B=&Waveform("Waveform [1]"=::NewEnumerator2, "[1] Frequency"=6, "[1] Amplitude Min/Max"=<0.8 1>)
    )
    ScaleA=true
    ScaleRGB=true
   )
   /SetVariables(Particles.Position=&AddVectorToPosition(
     Position=$Engine.Emitter.SimulationPosition
     Vector=&MakeVector(
      X=0
      Y=0
      Z=&FloatFromCurve(
       FloatCurve=DataInterface() {propertyValues {Curve {keys [<at 0 -1 RCIM_Cubic>, <at 0.075 -0.5 RCIM_Cubic>, {arriveTangent=-0.5405, leaveTangent=-0.5405, time=0.15, value=-1}, <at 1 -1 RCIM_Cubic>]}}}
       CurveIndex=$System.NormalizedLoopAge
       Scale Curve=40
      )
     )
    )) {isSetParameters=true}
  ]
  rendererClasses [&NiagaraSpriteRendererProperties]
  renderers [
   {rendererClass=&NiagaraSpriteRendererProperties}
   {rendererClass=&NiagaraSpriteRendererProperties, rendererIndex=1}
  ]
 }
 OrangeGlow {rendererClasses [], renderers []}
 ParticleDecal {
  emitterUpdateScript [
   &EmitterState()
   &SpawnBurst_Instantaneous(Use Spawn Probability=false, UseLoopCountLimit=false)
  ]
  particleSpawnScript [
   &InitializeParticle(
    Color=<rgba 6 1.4579 0 1>
    Lifetime=4.62
    Position Offset=&MakeVector(
     X=0
     Y=0
     Z=&FloatFromCurve(
      FloatCurve=DataInterface() {propertyValues {Curve {keys [<at 0 1 RCIM_Cubic>,<at 0.075 1.5 RCIM_Cubic>,<at 0.15 1>,<at 1 1>]}}}
      CurveIndex=$Emitter.NormalizedLoopAge
      Scale Curve=-35
     )
    )
    Sprite Size=<100 45>
    Sprite Size Mode=::NewEnumerator4
    UsePositionOffset=true
    Write Material Random=true
   )
   &CameraOffset(Camera Offset Amount=10)
   &DynamicMaterialParameters(
    Index 0 Param 1=1.7
    Index 0 Param 2=1.25
    Index 0 Param 3=1
    Module.Index0Float0Write=true
    Module.Index0Float1Write=true
    Module.Index0Float2Write=true
    Module.Index0Float3Write=true
   )
  ]
  particleUpdateScript [
   &ParticleState()
   &ScaleColor(
    Scale Alpha=&FloatFromCurve(FloatCurve=DataInterface() {propertyValues {Curve {keys [<at 0 1>, {arriveTangent=-0.1053, interpMode=RCIM_Cubic, leaveTangent=-0.1053, tangentMode=RCTM_User, time=0.5, value=1}, <at 1 0>]}}})
    ScaleA=true
    ScaleRGB=true
   )
  ]
  rendererClasses [&NiagaraSpriteRendererProperties]
  renderers [{rendererClass=&NiagaraSpriteRendererProperties}]
 }
 SpikyImpact {rendererClasses [], renderers []}
 StylizedFire_Main001 {
  emitterUpdateScript [
   &EmitterState()
   &SpawnBurst_Instantaneous(Spawn Count=2, Use Spawn Probability=false, UseLoopCountLimit=false)
   &SpawnBurst_Instantaneous(Spawn Group=1, Use Spawn Probability=false, UseLoopCountLimit=false) {module=SpawnBurst_Instantaneous001}
  ]
  particleSpawnScript [
   /SetVariables(Particles.VisibilityTag=&SpawnGroup()) {isSetParameters=true}
   &FN_SetSpawnGroup(Number of Spawn Groups=2)
   &InitializeParticle(
    Lifetime=4.62
    Mesh Scale=<0.65 0.6 0.45>
    Mesh Scale Mode=::NewEnumerator4
    Sprite Size=&MakeVector2D(X=$User.Emote_MatchaSpare_SizeX, Y=$User.Emote_MatchaSpare_SizeY)
    Sprite Size Mode=::NewEnumerator4
    UsePositionOffset=false
    Write Material Random=true
   )
   &CameraOffset(Camera Offset Amount=-15)
   /SetVariables() {isSetParameters=true}
   &SpriteFacingAndAlignment(Sprite Alignment=<0 0 1>, WriteAlignment=true, WriteFacing=false)
   &InitialMeshOrientation(Mesh Orientation Mode=::NewEnumerator2, Rotation=<0 0.25 0>, Use Rotation Vector=true)
  ]
  particleUpdateScript [
   &ParticleState()
   &SkeletalMeshLocation(
    Bone / Socket Index=3
    Bone Sampling Mode=::NewEnumerator4
    MaskSpawnGroup=false
    Sampled Position Offset=&MergeHelper_VectorSwitch(
     0=&MakeVector(X=-8, Y=15, Z=-2)
     1=&MakeVector(X=8, Y=15, Z=-2)
     2=?
     3=?
     Index=&SpawnGroup()
     Num=::NewEnumerator0
    )
    Sampled Position Offset Coordinate Space:ENiagaraSkelMeshTransforms=::NewEnumerator2
    Skeletal Mesh=$System.Skeletal Mesh
   )
   &SkeletalMeshLocation(
    Bone / Socket Index=3
    Bone Sampling Mode=::NewEnumerator4
    MaskSpawnGroup=false
    Position Sampling=::NewEnumerator2
    Sampled Position Offset=<0 6 -8>
    Sampled Position Offset Coordinate Space:ENiagaraSkelMeshTransforms=::NewEnumerator2
    Skeletal Mesh=$System.Skeletal Mesh
   ) {module=SkeletalMeshLocation001}
   &ScaleSpriteSize(
    Non-Uniform Curve Sprite Scale=DataInterface() {propertyValues {
      XCurve {keys [
        {arriveTangent=8.1644, interpMode=RCIM_Cubic, leaveTangent=8.1644, leaveTangentWeight=0.3333, tangentMode=RCTM_User, time=0, value=0}
        {
         arriveTangent=-0.3076
         arriveTangentWeight=0.0673
         interpMode=RCIM_Cubic
         leaveTangent=-0.3076
         leaveTangentWeight=0.2917
         tangentMode=RCTM_User
         time=0.075
         value=1.25
        }
        {arriveTangent=-0.3448, interpMode=RCIM_Cubic, leaveTangent=-0.3448, time=0.6, value=1}
        {
         arriveTangent=-0.0864
         arriveTangentWeight=0.1333
         interpMode=RCIM_Cubic
         leaveTangent=-0.0864
         leaveTangentWeight=0.0667
         tangentMode=RCTM_User
         time=0.8
         value=1
        }
       ]}
      YCurve {keys [
        {arriveTangent=8.1644, interpMode=RCIM_Cubic, leaveTangent=8.1644, leaveTangentWeight=0.3333, tangentMode=RCTM_User, time=0, value=0}
        {arriveTangent=-0.3076, interpMode=RCIM_Cubic, leaveTangent=-0.3076, tangentMode=RCTM_User, time=0.075, value=1.4}
        {arriveTangent=-0.308, interpMode=RCIM_Cubic, leaveTangent=-0.308, tangentMode=RCTM_User, time=0.175, value=1}
        {arriveTangent=-0.2096, interpMode=RCIM_Cubic, leaveTangent=-0.2096, tangentMode=RCTM_User, time=0.6, value=1}
        {
         arriveTangent=-0.0864
         arriveTangentWeight=0.1333
         interpMode=RCIM_Cubic
         leaveTangent=-0.0864
         leaveTangentWeight=0.0667
         tangentMode=RCTM_User
         time=0.8
         value=1
        }
        <at 1 0.3 RCIM_Cubic>
       ]}
     }}
    Scale Sprite Size Mode=::NewEnumerator3
   )
   &ScaleColor(
    Linear Color Curve=DataInterface() {propertyValues {
      AlphaCurve {keys [
        {arriveTangent=5.1543, interpMode=RCIM_Cubic, leaveTangent=5.1543, tangentMode=RCTM_User, time=0, value=0}
        {arriveTangent=0.023, interpMode=RCIM_Cubic, leaveTangent=0.023, tangentMode=RCTM_User, time=0.125, value=1}
        {arriveTangent=0.0022, interpMode=RCIM_Cubic, leaveTangent=0.0022, tangentMode=RCTM_User, time=0.7, value=1}
        <at 1 0>
       ]}
      BlueCurve {keys [<at 0 1>,<at 0.1 2>,<at 0.7 2>,<at 1 1>]}
      GreenCurve {keys [<at 0 1>,<at 0.1 2>,<at 0.7 2>,<at 1 1>]}
      RedCurve {keys [<at 0 1>,<at 0.1 2>,<at 0.7 2>,<at 1 1>]}
     }}
    Scale Mode=::NewEnumerator2
   )
   &FN_Light_Attributes(
    Color=<rgba 1 0.4577 0 1>
    Create Color Attribute=true
    Create Exponent Attribute=true
    Create Position Attribute=true
    Create Radius Attribute=true
    Exponent=4
    Position=&MakeVector(
     X=0
     Y=0
     Z=&FloatFromCurve(
      FloatCurve=DataInterface() {propertyValues {Curve {keys [<at 0 1 RCIM_Cubic>, <at 0.075 1.5 RCIM_Cubic>, {arriveTangent=-0.2703, leaveTangent=-0.2703, time=0.15, value=1}, <at 1 1>]}}}
      Scale Curve=22.5
     )
    )
    Radius=120
    Scale Color=&Multiply_Float(
     A=&FloatFromCurve(
      FloatCurve=DataInterface() {propertyValues {Curve {keys [
          {arriveTangent=-12.2496, interpMode=RCIM_Cubic, leaveTangent=-12.2496, tangentMode=RCTM_User, time=0, value=2}
          {arriveTangent=-1.25, leaveTangent=-1.25, time=0.15, value=1}
          <at 0.8 1>
          <at 1 0>
         ]}}}
      Scale Curve=2.75
     )
     B=&Waveform("Waveform [1]"=::NewEnumerator2, "[1] Frequency"=6, "[1] Amplitude Min/Max"=<0.6 1.2>)
    )
   )
   /SetVariables(Particles.MeshPosition:NiagaraPosition=$Particles.SkeletalMeshLocation001.SampledPosition) {isSetParameters=true}
  ]
  rendererClasses [
   &NiagaraSpriteRendererProperties
   &NiagaraLightRendererProperties
   &NiagaraMeshRendererProperties
  ]
  renderers [
   {rendererClass=&NiagaraSpriteRendererProperties}
   {rendererClass=&NiagaraLightRendererProperties, rendererIndex=1}
   {rendererClass=&NiagaraMeshRendererProperties, rendererIndex=2}
  ]
 }
 ToonImpactSmoke {rendererClasses [], renderers []}
}
summary {
 systemName=NS_Emote_MatchaSpare_Fire
 emitters [
  {
   emitterName=StylizedFire_Main001
   rendererClasses [
    &NiagaraSpriteRendererProperties
    &NiagaraLightRendererProperties
    &NiagaraMeshRendererProperties
   ]
  }
  {emitterName=FireBits, rendererClasses [&NiagaraSpriteRendererProperties]}
  {emitterName=ToonImpactSmoke, rendererClasses []}
  {emitterName=OrangeGlow, rendererClasses []}
  {emitterName=DarkGlow, rendererClasses []}
  {emitterName=ExplosionStreaks, rendererClasses []}
  {emitterName=SpikyImpact, rendererClasses []}
  {emitterName=Debris, rendererClasses []}
  {emitterName=ParticleDecal, rendererClasses [&NiagaraSpriteRendererProperties]}
  {emitterName=FlameSprites, rendererClasses [&NiagaraSpriteRendererProperties]}
  {emitterName=Glow, rendererClasses [&NiagaraSpriteRendererProperties]}
 ]
 userVariables [
  User.Emote_MatchaSpare_SizeX {type=NiagaraFloat, defaultValue=0, description=""}
  User.Emote_MatchaSpare_SizeY {type=NiagaraFloat, defaultValue=0, description=""}
 ]
}
